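# Editor utility widget for render-settings art tools: a labelled checkbox row that toggles a console variable, with a reset button.
# On construct it reads the cvar as a bool, storing it as default and building the tooltip.
# Changes run console commands and show reset when off-default; without its own cvar, it enables a linked combo or spin box's cvar, or zeroes it.
# Helpers resync from the cvar, serialize "cvar,value" and clone the row.
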EditorUtilityWidgetBlueprint /Game/ArtTools/RenderSettings/EUW_cvar_checkBox
graphs [
 EventGraph {nodes [
   "On Clicked (resetButton)"() {is=title K2Node_ComponentBoundEvent}
   n1 = Get defaultValue() -> (defaultValue:bool)
   > setUI!(execute, self, value:bool=@n1) -> (then)
   > setCvar!(execute, self, value:bool=@n1) -> (then)
   > SetVisibility(self=Get resetButton() -> (resetButton:Button Object Reference), InVisibility=Hidden)
   n2 = "On Check State Changed (checkBox)"!() -> (OutputDelegate, then, bIsChecked:bool) {is=title K2Node_ComponentBoundEvent}
   > n3 = Branch(Condition="Equal (Boolean)"!(A:bool=@n2.bIsChecked, B:bool=Get defaultValue()) -> (ReturnValue:bool))
   n4 = Get resetButton()
   > n5 = SetVisibility(self=@n4, InVisibility=Hidden)
   n6 = Branch(execute=[@n11.then, @n5.then], Condition=Get hasCvar() -> (hasCvar:bool))
   > setCvar(value=@n2.bIsChecked)
   n7 = Branch(
    execute=@n6.else
    Condition=IsNullItem!(
     ItemObject:Object Reference=Get overrideComboBox() -> (overrideComboBox:EUW Cvar Combo Box Object Reference)
     __WorldContext:Object Reference
    ) -> (Is Null Item:bool)
   )
   > n8 = Branch(Condition=IsNullItem(ItemObject=Get overrideSpinBox() -> (overrideSpinBox:EUW Cvar Spin Box Object Reference)))
   n9 = Branch(execute=@n8.else, Condition=@n2.bIsChecked)
   > updateCvar(self:EUW Cvar Spin Box Object Reference=Get overrideSpinBox())
   setCvar_param!(
    execute=@n9.else
    self
    value:bool=false
    cvar:String=Get cvar!(self:EUW Cvar Spin Box Object Reference=Get overrideSpinBox()) -> (cvar:String)
   ) -> (then)
   n10 = Branch(execute=@n7.else, Condition=@n2.bIsChecked)
   > updateCvar(self:EUW Cvar Combo Box Object Reference=Get overrideComboBox())
   setCvar_param(execute=@n10.else, value=false, cvar=Get cvar(self:EUW Cvar Combo Box Object Reference=Get overrideComboBox()))
   n11 = SetVisibility(execute=@n3.else, self=@n4, InVisibility=Visible)
   n12 = Event PreConstruct()
   n13 = Branch(execute=[@n12.then, @n22.then], Condition=Get hasCvar())
   > n14 = getCvar!(execute, self) -> (then, value:bool)
   > n15 = Set value(value:bool=@n14) -> (Output_Get:bool)
   n16 = Get label_1() -> (label_1:Text Object Reference)
   "SetText (Text)"(execute=[@n15.then, @n21.then], self=@n16, InText=Get labelText() -> (labelText:Text))
   n17 = Get labelToolTip() -> (labelToolTip:Text)
   n18 = "To Text (String)"(InString=Append!(
     A=Get cvar!() -> (cvar)
     B="\r\n"
     C:String="To String (Text)"(InText=@n17)
     D:String="\r\nDefault :"
     E:String="To String (Boolean)"(InBool=@n15)
    ) -> (ReturnValue))
   > SetToolTipText!(execute, self:Widget Object Reference=@n16, InToolTipText:"Text (by ref)"=@n18) -> (then)
   n19 = Get checkBox() -> (checkBox:Check Box Object Reference)
   > SetToolTipText(self=@n19, InToolTipText=@n18)
   n20 = Get value() -> (value:bool)
   > SetIsChecked!(execute, self:Check Box Object Reference=@n19, InIsChecked:bool=@n20) -> (then)
   > Set defaultValue(defaultValue:bool=@n20) -> (Output_Get:bool)
   > SetVisibility(self=Get resetButton(), InVisibility=Hidden)
   n21 = Set value(execute=@n13.else, value=true)
   n22 = Event Construct()
   Event Tick()
  ]}
 duplicate {nodes [
   n1 = duplicate!() -> (then, panel:Panel Widget Object Reference) {is=title K2Node_FunctionEntry}
   > n2 = Create Widget!(
    execute:Exec
    WorldContextObject:Object Reference
    Class:User Widget Class Reference=&/EUW_cvar_checkBox.EUW_cvar_checkBox_C
    OwningPlayer:Player Controller Object Reference
   ) -> (then:Exec, ReturnValue:EUW Cvar Check Box Object Reference)
   > AddChild(self=@n1, Content=@n2)
   n3 = Self-Reference()
   > n4 = getValue!(execute, self=@n3) -> (then, value:bool)
   > setValue!(execute, self=@n2, value:bool=@n4) -> (then)
   > Set defaultValue!(execute, defaultValue=Get defaultValue(None=@n3), self:EUW Cvar Check Box Object Reference=@n2) -> (then, Output_Get)
   n5 = Get labelText!(self:EUW Cvar Check Box Object Reference=@n3) -> (labelText)
   > "SetText (Text)"(self=Get label_1!(self:EUW Cvar Check Box Object Reference=@n2) -> (label_1), InText=@n5)
   > SetVisibility(self=Get resetButton!(self:EUW Cvar Check Box Object Reference=@n2) -> (resetButton), InVisibility=Visible)
   > Set cvar!(execute, cvar:String=Get cvar(None=@n3), self:EUW Cvar Check Box Object Reference=@n2) -> (then, Output_Get:String)
   > Set hasCvar!(execute, hasCvar:bool=Get hasCvar(None=@n3), self:EUW Cvar Check Box Object Reference=@n2) -> (then, Output_Get:bool)
   > PrintString(InString=Append!(A="check box ", B=GetDisplayName(Object=@n3), C=" ", D="To String (Text)"(InText=@n5), E=" parent:", F:String=GetDisplayName(Object=@n1)) -> (ReturnValue))
  ]}
 getCvar {nodes [
   getCvar() {is=title K2Node_FunctionEntry}
   n1 = Get cvar()
   > Return Node!(
    execute
    value:bool="float >= float"(A=GetNumericCVar!(CVar:String=@n1) -> (OutValue:float, ReturnValue:bool), B=1) {is=title K2Node_CallFunction}
   )
  ]}
 getCvarSetUI {nodes [
   n1 = getCvarSetUI!() -> (then, setDefault:bool=~) {is=title K2Node_FunctionEntry}
   > Branch(Condition=Get hasCvar())
   > n2 = getCvar() {is=title K2Node_CallFunction}
   > setUI(value=@n2)
   > n3 = Branch(Condition=@n1)
   > Set defaultValue!(execute, defaultValue=@n2) -> (then, Output_Get)
   > SetVisibility(self=Get resetButton!() -> (resetButton), InVisibility=Hidden)
   n4 = Branch(execute=@n3.else, Condition="Equal (Boolean)"(A=@n2, B=Get defaultValue()))
   n5 = Get resetButton()
   > SetVisibility(self=@n5, InVisibility=Hidden)
   SetVisibility(execute=@n4.else, self=@n5, InVisibility=Visible)
  ]}
 getValue {nodes [
   getValue() {is=title K2Node_FunctionEntry}
   > Return Node(value=IsChecked(self=Get checkBox()))
  ]}
 save {nodes [
   save() {is=title K2Node_FunctionEntry}
   n1 = "To Integer (Boolean)"!(InBool:bool=IsChecked(self=Get checkBox())) -> (ReturnValue:int=0)
   > Return Node(value:String=Append!(A=Get cvar(), B=",", C="To String (Integer)"(InInt=@n1)) -> (ReturnValue))
  ]}
 setCvar {nodes [
   n1 = setCvar!() -> (then, value:bool=~) {is=title K2Node_FunctionEntry}
   n2 = "To Integer (Boolean)"(InBool=@n1) -> (ReturnValue=0)
   n3 = Append(A=Get cvar(), B=" ", C="To String (Text)"(InText="To Text (Integer)"(Value=@n2)))
   > PrintString(InString=@n3)
   > ExecuteConsoleCommand(Command=@n3)
  ]}
 setCvar_param {nodes [
   n1 = setCvar_param!() -> (then, value:bool=~, cvar:String) {is=title K2Node_FunctionEntry}
   n2 = "To Integer (Boolean)"(InBool=@n1) -> (ReturnValue=0)
   n3 = Append(A=@n1.cvar, B=" ", C="To String (Text)"(InText="To Text (Integer)"(Value=@n2)))
   > PrintString(InString=@n3)
   > ExecuteConsoleCommand(Command=@n3)
  ]}
 setUI {nodes [
   n1 = setUI!() -> (then, value:bool=~) {is=title K2Node_FunctionEntry}
   > SetIsChecked(self=Get checkBox(), InIsChecked=@n1)
  ]}
 setValue {nodes [
   n1 = setValue!() -> (then, value:bool=~) {is=title K2Node_FunctionEntry}
   > SetIsChecked(self=Get checkBox(), InIsChecked=@n1)
  ]}
]
variables [
 cvar {type=string, default=""}
 defaultValue {type=boolean}
 hasCvar {type=boolean, default=true}
 labelText {type=string, default=Label}
 labelToolTip {type=string, default=""}
 overrideComboBox {type=EUW_cvar_comboBox_C, default=null}
 overrideSpinBox {type=EUW_cvar_spinBox_C, default=null}
 value {type=boolean}
]
widgets {
 info {rootWidgetClass=&CanvasPanel, widgetCount=4}
 widgets [
  CanvasPanel_0 {widgetClassPath=&CanvasPanel, widget=CanvasPanel_0}
  label_1 {
   widgetClassPath=&TextBlock
   bIsVariable=true
   parent=CanvasPanel_0
   widget=label_1
   properties {
    text=Name
    font {
     bForceMonospaced=false
     bMaterialIsStencil=false
     fontMaterial=null
     fontObject=&Roboto
     letterSpacing=0
     monospacedWidth=1
     size=9
     skewAmount=0
     typefaceFontName=Regular
     outlineSettings {
      bApplyOutlineToDropShadows=false
      bMiteredCorners=false
      bSeparateFillAlpha=false
      outlineMaterial=null
      outlineSize=0
      outlineColor=<rgba 0 0 0 1>
     }
    }
   }
   slot {class=CanvasPanelSlot, properties {layoutData {alignment=<0 0>, anchors {maximum=<0 0>, minimum=<0 0>}, offsets=<ltrb 2 2 180 20>}}}
  }
  checkBox {
   widgetClassPath=/Script/UMG.CheckBox
   bIsVariable=true
   parent=CanvasPanel_0
   widget=checkBox
   properties {widgetStyle {
     checkBoxType=CheckBox
     backgroundHoveredImage {
      bIsDynamicallyLoaded=false
      drawAs=RoundedBox
      imageType=NoImage
      mirroring=NoMirror
      resourceName=None
      resourceObject=null
      tiling=NoTile
      imageSize=<18 18>
      margin=<ltrb 0 0 0 0>
      outlineSettings {
       bUseBrushTransparency=false
       roundingType=FixedRadius
       width=1
       color {colorUseRule=UseColor_ColorTable, specifiedColor=<rgba 0 0 0 0>}
       cornerRadii=<3 3 3 3>
      }
      tintColor {colorUseRule=UseColor_Specified, specifiedColor=<rgba 0.0048 0.0048 0.0048 1>}
      uVRegion {bIsValid=false, max=<0 0>, min=<0 0>}
     }
     backgroundImage {
      bIsDynamicallyLoaded=false
      drawAs=RoundedBox
      imageType=NoImage
      mirroring=NoMirror
      resourceName=None
      resourceObject=null
      tiling=NoTile
      imageSize=<18 18>
      margin=<ltrb 0 0 0 0>
      outlineSettings {
       bUseBrushTransparency=false
       roundingType=FixedRadius
       width=1
       color {colorUseRule=UseColor_ColorTable, specifiedColor=<rgba 0 0 0 0>}
       cornerRadii=<3 3 3 3>
      }
      tintColor {colorUseRule=UseColor_Specified, specifiedColor=<rgba 0.0048 0.0048 0.0048 1>}
      uVRegion {bIsValid=false, max=<0 0>, min=<0 0>}
     }
     backgroundPressedImage {
      bIsDynamicallyLoaded=false
      drawAs=RoundedBox
      imageType=NoImage
      mirroring=NoMirror
      resourceName=None
      resourceObject=null
      tiling=NoTile
      imageSize=<18 18>
      margin=<ltrb 0 0 0 0>
      outlineSettings {
       bUseBrushTransparency=false
       roundingType=FixedRadius
       width=1
       color {colorUseRule=UseColor_ColorTable, specifiedColor=<rgba 0 0 0 0>}
       cornerRadii=<3 3 3 3>
      }
      tintColor {colorUseRule=UseColor_Specified, specifiedColor=<rgba 0.0048 0.0048 0.0048 1>}
      uVRegion {bIsValid=false, max=<0 0>, min=<0 0>}
     }
     borderBackgroundColor {colorUseRule=UseColor_Specified, specifiedColor=<rgba 1 1 1 1>}
     checkedForeground {colorUseRule=UseColor_Specified, specifiedColor=<rgba 1 1 1 1>}
     checkedHoveredForeground {colorUseRule=UseColor_Specified, specifiedColor=<rgba 1 1 1 1>}
     checkedHoveredImage {
      bIsDynamicallyLoaded=false
      drawAs=Image
      imageType=Vector
      mirroring=NoMirror
      resourceName=../../../Engine/Content/Slate/Starship/CoreWidgets/CheckBox/check.svg
      resourceObject=null
      tiling=NoTile
      imageSize=<16 16>
      margin=<ltrb 0 0 0 0>
      outlineSettings {
       bUseBrushTransparency=false
       roundingType=HalfHeightRadius
       width=0
       color {colorUseRule=UseColor_Specified, specifiedColor=<rgba 0 0 0 0>}
       cornerRadii=<0 0 0 0>
      }
      tintColor {colorUseRule=UseColor_Specified, specifiedColor=<rgba 0.0044 0.2384 1 1>}
      uVRegion {bIsValid=false, max=<0 0>, min=<0 0>}
     }
     checkedImage {
      bIsDynamicallyLoaded=false
      drawAs=Image
      imageType=Vector
      mirroring=NoMirror
      resourceName=../../../Engine/Content/Slate/Starship/CoreWidgets/CheckBox/check.svg
      resourceObject=null
      tiling=NoTile
      imageSize=<16 16>
      margin=<ltrb 0 0 0 0>
      outlineSettings {
       bUseBrushTransparency=false
       roundingType=HalfHeightRadius
       width=0
       color {colorUseRule=UseColor_Specified, specifiedColor=<rgba 0 0 0 0>}
       cornerRadii=<0 0 0 0>
      }
      tintColor {colorUseRule=UseColor_Specified, specifiedColor=<rgba 0 0.162 0.7454 1>}
      uVRegion {bIsValid=false, max=<0 0>, min=<0 0>}
     }
     checkedPressedForeground {colorUseRule=UseColor_Specified, specifiedColor=<rgba 1 1 1 1>}
     checkedPressedImage {
      bIsDynamicallyLoaded=false
      drawAs=Image
      imageType=Vector
      mirroring=NoMirror
      resourceName=../../../Engine/Content/Slate/Starship/CoreWidgets/CheckBox/check.svg
      resourceObject=null
      tiling=NoTile
      imageSize=<16 16>
      margin=<ltrb 0 0 0 0>
      outlineSettings {
       bUseBrushTransparency=false
       roundingType=HalfHeightRadius
       width=0
       color {colorUseRule=UseColor_Specified, specifiedColor=<rgba 0 0 0 0>}
       cornerRadii=<0 0 0 0>
      }
      tintColor {colorUseRule=UseColor_Specified, specifiedColor=<rgba 0 0.0802 0.3515 1>}
      uVRegion {bIsValid=false, max=<0 0>, min=<0 0>}
     }
     checkedSlateSound {resourceObject=null}
     foregroundColor {colorUseRule=UseColor_Specified, specifiedColor=<rgba 1 1 1 1>}
     hoveredForeground {colorUseRule=UseColor_Specified, specifiedColor=<rgba 1 1 1 1>}
     hoveredSlateSound {resourceObject=null}
     padding=<ltrb 2 0 0 0>
     pressedForeground {colorUseRule=UseColor_Specified, specifiedColor=<rgba 1 1 1 1>}
     uncheckedHoveredImage {
      bIsDynamicallyLoaded=false
      drawAs=NoDrawType
      imageType=NoImage
      mirroring=NoMirror
      resourceName=None
      resourceObject=null
      tiling=NoTile
      imageSize=<0 0>
      margin=<ltrb 0 0 0 0>
      outlineSettings {
       bUseBrushTransparency=false
       roundingType=HalfHeightRadius
       width=0
       color {colorUseRule=UseColor_Specified, specifiedColor=<rgba 0 0 0 0>}
       cornerRadii=<0 0 0 0>
      }
      tintColor {colorUseRule=UseColor_Specified, specifiedColor=<rgba 1 1 1 1>}
      uVRegion {bIsValid=false, max=<0 0>, min=<0 0>}
     }
     uncheckedImage {
      bIsDynamicallyLoaded=false
      drawAs=NoDrawType
      imageType=NoImage
      mirroring=NoMirror
      resourceName=None
      resourceObject=null
      tiling=NoTile
      imageSize=<0 0>
      margin=<ltrb 0 0 0 0>
      outlineSettings {
       bUseBrushTransparency=false
       roundingType=HalfHeightRadius
       width=0
       color {colorUseRule=UseColor_Specified, specifiedColor=<rgba 0 0 0 0>}
       cornerRadii=<0 0 0 0>
      }
      tintColor {colorUseRule=UseColor_Specified, specifiedColor=<rgba 1 1 1 1>}
      uVRegion {bIsValid=false, max=<0 0>, min=<0 0>}
     }
     uncheckedPressedImage {
      bIsDynamicallyLoaded=false
      drawAs=NoDrawType
      imageType=NoImage
      mirroring=NoMirror
      resourceName=None
      resourceObject=null
      tiling=NoTile
      imageSize=<0 0>
      margin=<ltrb 0 0 0 0>
      outlineSettings {
       bUseBrushTransparency=false
       roundingType=HalfHeightRadius
       width=0
       color {colorUseRule=UseColor_Specified, specifiedColor=<rgba 0 0 0 0>}
       cornerRadii=<0 0 0 0>
      }
      tintColor {colorUseRule=UseColor_Specified, specifiedColor=<rgba 1 1 1 1>}
      uVRegion {bIsValid=false, max=<0 0>, min=<0 0>}
     }
     uncheckedSlateSound {resourceObject=null}
     undeterminedForeground {colorUseRule=UseColor_Specified, specifiedColor=<rgba 1 1 1 1>}
     undeterminedHoveredImage {
      bIsDynamicallyLoaded=false
      drawAs=Image
      imageType=Vector
      mirroring=NoMirror
      resourceName=../../../Engine/Content/Slate/Starship/CoreWidgets/CheckBox/indeterminate.svg
      resourceObject=null
      tiling=NoTile
      imageSize=<16 16>
      margin=<ltrb 0 0 0 0>
      outlineSettings {
       bUseBrushTransparency=false
       roundingType=HalfHeightRadius
       width=0
       color {colorUseRule=UseColor_Specified, specifiedColor=<rgba 0 0 0 0>}
       cornerRadii=<0 0 0 0>
      }
      tintColor {colorUseRule=UseColor_Specified, specifiedColor=<rgba 0.0044 0.2384 1 1>}
      uVRegion {bIsValid=false, max=<0 0>, min=<0 0>}
     }
     undeterminedImage {
      bIsDynamicallyLoaded=false
      drawAs=Image
      imageType=Vector
      mirroring=NoMirror
      resourceName=../../../Engine/Content/Slate/Starship/CoreWidgets/CheckBox/indeterminate.svg
      resourceObject=null
      tiling=NoTile
      imageSize=<16 16>
      margin=<ltrb 0 0 0 0>
      outlineSettings {
       bUseBrushTransparency=false
       roundingType=HalfHeightRadius
       width=0
       color {colorUseRule=UseColor_Specified, specifiedColor=<rgba 0 0 0 0>}
       cornerRadii=<0 0 0 0>
      }
      tintColor {colorUseRule=UseColor_Specified, specifiedColor=<rgba 0 0.162 0.7454 1>}
      uVRegion {bIsValid=false, max=<0 0>, min=<0 0>}
     }
     undeterminedPressedImage {
      bIsDynamicallyLoaded=false
      drawAs=Image
      imageType=Vector
      mirroring=NoMirror
      resourceName=../../../Engine/Content/Slate/Starship/CoreWidgets/CheckBox/indeterminate.svg
      resourceObject=null
      tiling=NoTile
      imageSize=<16 16>
      margin=<ltrb 0 0 0 0>
      outlineSettings {
       bUseBrushTransparency=false
       roundingType=HalfHeightRadius
       width=0
       color {colorUseRule=UseColor_Specified, specifiedColor=<rgba 0 0 0 0>}
       cornerRadii=<0 0 0 0>
      }
      tintColor {colorUseRule=UseColor_Specified, specifiedColor=<rgba 0 0.0802 0.3515 1>}
      uVRegion {bIsValid=false, max=<0 0>, min=<0 0>}
     }
    }}
   slot {
    class=CanvasPanelSlot
    properties {bAutoSize=true, layoutData {alignment=<0 0>, anchors {maximum=<0 0>, minimum=<0 0>}, offsets=<ltrb 250 4 100 40>}}
   }
  }
  resetButton {
   widgetClassPath=&Button
   parent=CanvasPanel_0
   widget=resetButton
   properties {widgetStyle {
     clickedSlateSound {resourceObject=null}
     disabled {
      bIsDynamicallyLoaded=false
      drawAs=RoundedBox
      imageType=NoImage
      mirroring=NoMirror
      resourceName=None
      resourceObject=null
      tiling=NoTile
      imageSize=<0 0>
      margin=<ltrb 0 0 0 0>
      outlineSettings {
       bUseBrushTransparency=false
       roundingType=FixedRadius
       width=1
       color {colorUseRule=UseColor_ColorTable, specifiedColor=<rgba 0 0 0 0>}
       cornerRadii=<4 4 4 4>
      }
      tintColor {colorUseRule=UseColor_Specified, specifiedColor=<rgba 0.0395 0.0395 0.0395 1>}
      uVRegion {bIsValid=false, max=<0 0>, min=<0 0>}
     }
     disabledForeground {colorUseRule=UseColor_ColorTable, specifiedColor=<rgba 0 0 0 0>}
     hovered {
      bIsDynamicallyLoaded=false
      drawAs=RoundedBox
      imageType=NoImage
      mirroring=NoMirror
      resourceName=None
      resourceObject=null
      tiling=NoTile
      imageSize=<0 0>
      margin=<ltrb 0 0 0 0>
      outlineSettings {
       bUseBrushTransparency=false
       roundingType=FixedRadius
       width=1
       color {colorUseRule=UseColor_ColorTable, specifiedColor=<rgba 0 0 0 0>}
       cornerRadii=<4 4 4 4>
      }
      tintColor {colorUseRule=UseColor_Specified, specifiedColor=<rgba 0.0953 0.0953 0.0953 1>}
      uVRegion {bIsValid=false, max=<0 0>, min=<0 0>}
     }
     hoveredForeground {colorUseRule=UseColor_ColorTable, specifiedColor=<rgba 0 0 0 0>}
     hoveredSlateSound {resourceObject=null}
     normal {
      bIsDynamicallyLoaded=false
      drawAs=RoundedBox
      imageType=NoImage
      mirroring=NoMirror
      resourceName=None
      resourceObject=/Game/Athena/HUD/Art/BacchusHudIcons/Reset_Normal
      tiling=NoTile
      imageSize=<20 20>
      margin=<ltrb 0 0 0 0>
      outlineSettings {
       bUseBrushTransparency=false
       roundingType=FixedRadius
       width=1
       color {colorUseRule=UseColor_ColorTable, specifiedColor=<rgba 0 0 0 0>}
       cornerRadii=<4 4 4 4>
      }
      tintColor {colorUseRule=UseColor_Specified, specifiedColor=<rgba 1 1 1 0.5>}
      uVRegion {bIsValid=false, max=<0 0>, min=<0 0>}
     }
     normalForeground {colorUseRule=UseColor_ColorTable, specifiedColor=<rgba 0 0 0 0>}
     normalPadding=<ltrb 12 1 12 1>
     pressed {
      bIsDynamicallyLoaded=false
      drawAs=RoundedBox
      imageType=NoImage
      mirroring=NoMirror
      resourceName=None
      resourceObject=null
      tiling=NoTile
      imageSize=<0 0>
      margin=<ltrb 0 0 0 0>
      outlineSettings {
       bUseBrushTransparency=false
       roundingType=FixedRadius
       width=1
       color {colorUseRule=UseColor_ColorTable, specifiedColor=<rgba 0 0 0 0>}
       cornerRadii=<4 4 4 4>
      }
      tintColor {colorUseRule=UseColor_Specified, specifiedColor=<rgba 0.0284 0.0284 0.0284 1>}
      uVRegion {bIsValid=false, max=<0 0>, min=<0 0>}
     }
     pressedForeground {colorUseRule=UseColor_ColorTable, specifiedColor=<rgba 0 0 0 0>}
     pressedPadding=<ltrb 12 2 12 0>
     pressedSlateSound {resourceObject=null}
    }}
   slot {class=CanvasPanelSlot, properties {layoutData {alignment=<0 0>, anchors {maximum=<0 0>, minimum=<0 0>}, offsets=<ltrb 268 2 20 20>}}}
  }
 ]
}
