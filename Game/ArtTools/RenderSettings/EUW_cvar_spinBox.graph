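# Editor utility widget for render-settings tools: a labeled spin box that edits one numeric console variable, with a reset-to-default button.
# PreConstruct and Construct read the cvar and set label, tooltip and range, using a settingUp flag to suppress change handling.
# Value changes run a "cvar value" console command and show the reset button when off default.
# Helpers export "cvar,value" strings and duplicate the widget into another panel.
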EditorUtilityWidgetBlueprint /Game/ArtTools/RenderSettings/EUW_cvar_spinBox
graphs [
 EventGraph {nodes [
   n1 = Get v() -> (v:double)
   n2 = SetValue!(execute, self:Spin Box Object Reference=Get SpinBox_24() -> (SpinBox_24:Spin Box Object Reference), NewValue:float=@n1) -> (then)
   Set settingUp(execute=[@n2.then, @n16.then], settingUp:bool=false) -> (Output_Get:bool)
   "On Clicked (resetButton)"() {is=title K2Node_ComponentBoundEvent}
   > SetValue(self=Get SpinBox_24(), NewValue=Get defaultValue() -> (defaultValue:double))
   > SetVisibility(self=Get resetButton() -> (resetButton:Button Object Reference), InVisibility=Hidden)
   n3 = "On Value Changed (SpinBox_24)"!() -> (OutputDelegate, then, InValue:float) {is=title K2Node_ComponentBoundEvent}
   > n4 = Branch(Condition=Get settingUp() -> (settingUp:bool))
   n5 = Branch(
    execute=@n4.else
    Condition="Not Equal (Float)"!(A:double=@n3.InValue, B:double=Get defaultValue()) -> (ReturnValue:bool) {is=title K2Node_CallFunction}
   )
   > n6 = SetVisibility(self=Get resetButton(), InVisibility=Visible)
   updateCvar(execute=[@n6.then, @n5.else])
   n7 = Event PreConstruct()
   Set settingUp(execute=[@n7.then, @n17.then], settingUp=true)
   > n8 = getCvar!(execute, self) -> (then, value:double)
   > n9 = Set v(v:double=@n8) -> (Output_Get:double)
   > SetVisibility(self=Get resetButton(), InVisibility=Hidden)
   n10 = Get label() -> (label:Text Object Reference)
   > "SetText (Text)"(self=@n10, InText=Get labelText() -> (labelText:Text))
   n11 = Get labelToolTip() -> (labelToolTip:Text)
   n12 = "To Text (String)"(InString=Append!(
     A=Get cvar() -> (cvar:String)
     B="\r\n"
     C:String="To String (Text)"(InText=@n11)
     D:String="\r\nDefault : "
     E:String="To String (Float)"(InDouble=@n9)
    ) -> (ReturnValue))
   > SetToolTipText!(execute, self:Widget Object Reference=@n10, InToolTipText:"Text (by ref)"=@n12) -> (then)
   n13 = Get SpinBox_24()
   > SetToolTipText(self=@n13, InToolTipText=@n12)
   n14 = Get min() -> (min:double)
   > SetMinValue!(execute, self:Spin Box Object Reference=@n13, NewValue:float=@n14) -> (then)
   > SetMinSliderValue!(execute, self:Spin Box Object Reference=@n13, NewValue:float=@n14) -> (then)
   n15 = Get max() -> (max:double)
   > SetMaxValue!(execute, self:Spin Box Object Reference=@n13, NewValue:float=@n15) -> (then)
   > SetMaxSliderValue!(execute, self:Spin Box Object Reference=@n13, NewValue:float=@n15) -> (then)
   > n16 = Setup UI!(execute, self, value:double=@n1, resetDefault:bool=Get UseLoadingValueAsDefault() -> (UseLoadingValueAsDefault:bool)) -> (then)
   n17 = Event Construct()
   Event Tick()
  ]}
 Get Cvar Set UI {nodes [
   n1 = Get Cvar Set UI!() -> (then, resetDefault:bool=~) {is=title K2Node_FunctionEntry}
   > n2 = getCvar()
   > Setup UI(value=@n2, resetDefault=@n1)
  ]}
 Setup UI {nodes [
   n1 = Setup UI!() -> (then, value:double=~, resetDefault:bool=~) {is=title K2Node_FunctionEntry}
   > SetValue(self=Get SpinBox_24(), NewValue=@n1)
   > n2 = Branch(Condition=@n1.resetDefault)
   > n3 = Set defaultValue(defaultValue:double=@n1) -> (Output_Get:double)
   n4 = Get resetButton()
   > SetVisibility(self=@n4, InVisibility=Hidden)
   n5 = Branch(execute=@n2.else, Condition="Equal (Float)"!(A:double=@n1, B:double=@n3) -> (ReturnValue:bool) {is=title K2Node_CallFunction})
   > SetVisibility(self=@n4, InVisibility=Hidden)
   SetVisibility(execute=@n5.else, self=@n4, InVisibility=Visible)
  ]}
 duplicate {nodes [
   n1 = duplicate!() -> (then, panel:Panel Widget Object Reference) {is=title K2Node_FunctionEntry}
   > n2 = Create Widget!(
    execute:Exec
    WorldContextObject:Object Reference
    Class:User Widget Class Reference=&/EUW_cvar_spinBox.EUW_cvar_spinBox_C
    OwningPlayer:Player Controller Object Reference
   ) -> (then:Exec, ReturnValue:EUW Cvar Spin Box Object Reference)
   > AddChild(self=@n1, Content=@n2)
   n3 = Self-Reference()
   > n4 = getValue!(execute, self=@n3) -> (then, value:double)
   n5 = Get SpinBox_24(None=@n3)
   n6 = Get SpinBox_24!(self:EUW Cvar Spin Box Object Reference=@n2) -> (SpinBox_24)
   > SetMinValue(self=@n6, NewValue=GetMinValue!(self:Spin Box Object Reference=@n5) -> (ReturnValue:float))
   > SetMaxValue(self=@n6, NewValue=GetMaxValue!(self:Spin Box Object Reference=@n5) -> (ReturnValue:float))
   > SetMinSliderValue(self=@n6, NewValue=GetMinSliderValue!(self:Spin Box Object Reference=@n5) -> (ReturnValue:float))
   > SetMaxSliderValue(self=@n6, NewValue=GetMaxSliderValue!(self:Spin Box Object Reference=@n5) -> (ReturnValue:float))
   > setValue!(execute, self=@n2, value:double=@n4) -> (then)
   > Set cvar!(execute, cvar:String=Get cvar!(self:EUW Cvar Spin Box Object Reference=@n3) -> (cvar), self:EUW Cvar Spin Box Object Reference=@n2) -> (then, Output_Get:String)
   > Set defaultValue!(execute, defaultValue=Get defaultValue(None=@n3), self:EUW Cvar Spin Box Object Reference=@n2) -> (then, Output_Get)
   n7 = Get labelText(None=@n3)
   > "SetText (Text)"(self=Get label!(self:EUW Cvar Spin Box Object Reference=@n2) -> (label), InText=@n7)
   > SetVisibility(self=Get resetButton!(self:EUW Cvar Spin Box Object Reference=@n2) -> (resetButton), InVisibility=Visible)
   > PrintString(InString=Append!(A="spin box ", B=GetDisplayName(Object=@n2), C=" ", D="To String (Text)"(InText=@n7), E=" parent:", F:String=GetDisplayName(Object=@n1)) -> (ReturnValue))
  ]}
 getCvar {nodes [
   getCvar() {is=title K2Node_FunctionEntry}
   n1 = Get cvar!() -> (cvar)
   n2 = GetNumericCVar!(CVar:String=@n1) -> (OutValue:float, ReturnValue:bool)
   > n3 = Set v(v=@n2)
   > PrintString(InString=Append!(A=spinBox: getCvar, B=" ", C=@n1, D=" ", E="To String (Float)"(InDouble=@n2)) -> (ReturnValue))
   > Return Node!(execute, value:double=@n3)
  ]}
 getValue {nodes [
   getValue() {is=title K2Node_FunctionEntry}
   > Return Node(value=GetValue!(self:Spin Box Object Reference=Get SpinBox_24!() -> (SpinBox_24)) -> (ReturnValue:float))
  ]}
 roundToStep {nodes [
   n1 = roundToStep!() -> (then, v:double=~) {is=title K2Node_FunctionEntry}
   n2 = Get step() -> (step:double)
   n3 = Round(A=float / float(A=@n1, B=@n2) {is=title K2Node_CallFunction}) {is=title K2Node_CallFunction}
   > Return Node(value=float * float(A="To Float (Integer)"(InInt=@n3), B=@n2))
  ]}
 save {nodes [
   save() {is=title K2Node_FunctionEntry}
   n1 = GetValue(self=Get SpinBox_24())
   > Return Node(value:String=Append!(A=Get cvar(), B=",", C="To String (Float)"(InDouble=@n1)) -> (ReturnValue))
  ]}
 setCvar {nodes [
   n1 = setCvar!() -> (then, InFloat:double=~) {is=title K2Node_FunctionEntry}
   n2 = Append(A=Get cvar(), B=" ", C="To String (Float)"(InDouble=@n1))
   > PrintString(InString=@n2)
   > ExecuteConsoleCommand(Command=@n2)
  ]}
 setValue {nodes [
   n1 = setValue!() -> (then, value:double=~) {is=title K2Node_FunctionEntry}
   > SetValue(self=Get SpinBox_24(), NewValue=@n1)
  ]}
 updateCvar {nodes [
   updateCvar() {is=title K2Node_FunctionEntry}
   > setCvar!(execute, self, InFloat:double=GetValue(self=Get SpinBox_24())) -> (then) {is=title K2Node_CallFunction}
  ]}
]
variables [
 cvar {type=string, default=""}
 defaultValue {type=number}
 labelText {type=string, default=label}
 labelToolTip {type=string, default=Default ToolTip}
 max {type=number, default=10}
 min {type=number}
 settingUp {type=boolean, default=true}
 step {type=number, default=1}
 UseLoadingValueAsDefault {type=boolean}
 v {type=number}
]
widgets {
 info {rootWidgetClass=&CanvasPanel, widgetCount=4}
 widgets [
  CanvasPanel_0 {widgetClassPath=&CanvasPanel, widget=CanvasPanel_0}
  label {
   widgetClassPath=&TextBlock
   bIsVariable=true
   parent=CanvasPanel_0
   widget=label
   properties {
    text=Name
    font {
     bForceMonospaced=false
     bMaterialIsStencil=false
     fontMaterial=null
     fontObject=&Roboto
     letterSpacing=0
     monospacedWidth=1
     size=9
     skewAmount=0
     typefaceFontName=Regular
     outlineSettings {
      bApplyOutlineToDropShadows=false
      bMiteredCorners=false
      bSeparateFillAlpha=false
      outlineMaterial=null
      outlineSize=0
      outlineColor=<rgba 0 0 0 1>
     }
    }
   }
   slot {class=CanvasPanelSlot, properties {layoutData {alignment=<0 0>, anchors {maximum=<0 0>, minimum=<0 0>}, offsets=<ltrb 2 4 180 16>}}}
  }
  SpinBox_24 {
   widgetClassPath=/Script/UMG.SpinBox
   bIsVariable=true
   parent=CanvasPanel_0
   widget=SpinBox_24
   properties {
    bOverride_MaxSliderValue=true
    bOverride_MaxValue=true
    bOverride_MinSliderValue=true
    bOverride_MinValue=true
    delta=1
    justification=Center
    maxSliderValue=10
    value=1
    font {
     bForceMonospaced=false
     bMaterialIsStencil=false
     fontMaterial=null
     fontObject=&Roboto
     letterSpacing=0
     monospacedWidth=1
     size=9
     skewAmount=0
     typefaceFontName=Regular
     outlineSettings {
      bApplyOutlineToDropShadows=false
      bMiteredCorners=false
      bSeparateFillAlpha=false
      outlineMaterial=null
      outlineSize=0
      outlineColor=<rgba 0 0 0 1>
     }
    }
    foregroundColor {colorUseRule=UseColor_Specified, specifiedColor=<rgba 1 1 1 1>}
    widgetStyle {
     activeBackgroundBrush {
      bIsDynamicallyLoaded=false
      drawAs=RoundedBox
      imageType=NoImage
      mirroring=NoMirror
      resourceName=None
      resourceObject=null
      tiling=NoTile
      imageSize=<0 0>
      margin=<ltrb 0 0 0 0>
      outlineSettings {
       bUseBrushTransparency=false
       roundingType=FixedRadius
       width=1
       color {colorUseRule=UseColor_ColorTable, specifiedColor=<rgba 0 0 0 0>}
       cornerRadii=<4 4 4 4>
      }
      tintColor {colorUseRule=UseColor_Specified, specifiedColor=<rgba 0.0048 0.0048 0.0048 1>}
      uVRegion {bIsValid=false, max=<0 0>, min=<0 0>}
     }
     activeFillBrush {
      bIsDynamicallyLoaded=false
      drawAs=RoundedBox
      imageType=NoImage
      mirroring=NoMirror
      resourceName=None
      resourceObject=null
      tiling=NoTile
      imageSize=<0 0>
      margin=<ltrb 0 0 0 0>
      outlineSettings {
       bUseBrushTransparency=false
       roundingType=FixedRadius
       width=1
       color {colorUseRule=UseColor_Specified, specifiedColor=<rgba 0 0 0 0>}
       cornerRadii=<4 4 4 4>
      }
      tintColor {colorUseRule=UseColor_Specified, specifiedColor=<rgba 0 0.162 0.7454 1>}
      uVRegion {bIsValid=false, max=<0 0>, min=<0 0>}
     }
     arrowsImage {
      bIsDynamicallyLoaded=false
      drawAs=NoDrawType
      imageType=NoImage
      mirroring=NoMirror
      resourceName=None
      resourceObject=null
      tiling=NoTile
      imageSize=<0 0>
      margin=<ltrb 0 0 0 0>
      outlineSettings {
       bUseBrushTransparency=false
       roundingType=HalfHeightRadius
       width=0
       color {colorUseRule=UseColor_Specified, specifiedColor=<rgba 0 0 0 0>}
       cornerRadii=<0 0 0 0>
      }
      tintColor {colorUseRule=UseColor_Specified, specifiedColor=<rgba 1 1 1 1>}
      uVRegion {bIsValid=false, max=<0 0>, min=<0 0>}
     }
     backgroundBrush {
      bIsDynamicallyLoaded=false
      drawAs=RoundedBox
      imageType=NoImage
      mirroring=NoMirror
      resourceName=None
      resourceObject=null
      tiling=NoTile
      imageSize=<0 0>
      margin=<ltrb 0 0 0 0>
      outlineSettings {
       bUseBrushTransparency=false
       roundingType=FixedRadius
       width=1
       color {colorUseRule=UseColor_ColorTable, specifiedColor=<rgba 0 0 0 0>}
       cornerRadii=<4 4 4 4>
      }
      tintColor {colorUseRule=UseColor_Specified, specifiedColor=<rgba 0.0048 0.0048 0.0048 1>}
      uVRegion {bIsValid=false, max=<0 0>, min=<0 0>}
     }
     foregroundColor {colorUseRule=UseColor_Specified, specifiedColor=<rgba 1 1 1 1>}
     hoveredBackgroundBrush {
      bIsDynamicallyLoaded=false
      drawAs=RoundedBox
      imageType=NoImage
      mirroring=NoMirror
      resourceName=None
      resourceObject=null
      tiling=NoTile
      imageSize=<0 0>
      margin=<ltrb 0 0 0 0>
      outlineSettings {
       bUseBrushTransparency=false
       roundingType=FixedRadius
       width=1
       color {colorUseRule=UseColor_ColorTable, specifiedColor=<rgba 0 0 0 0>}
       cornerRadii=<4 4 4 4>
      }
      tintColor {colorUseRule=UseColor_Specified, specifiedColor=<rgba 0.0048 0.0048 0.0048 1>}
      uVRegion {bIsValid=false, max=<0 0>, min=<0 0>}
     }
     hoveredFillBrush {
      bIsDynamicallyLoaded=false
      drawAs=RoundedBox
      imageType=NoImage
      mirroring=NoMirror
      resourceName=None
      resourceObject=null
      tiling=NoTile
      imageSize=<0 0>
      margin=<ltrb 0 0 0 0>
      outlineSettings {
       bUseBrushTransparency=false
       roundingType=FixedRadius
       width=1
       color {colorUseRule=UseColor_Specified, specifiedColor=<rgba 0 0 0 0>}
       cornerRadii=<4 4 4 4>
      }
      tintColor {colorUseRule=UseColor_Specified, specifiedColor=<rgba 0.0953 0.0953 0.0953 1>}
      uVRegion {bIsValid=false, max=<0 0>, min=<0 0>}
     }
     inactiveFillBrush {
      bIsDynamicallyLoaded=false
      drawAs=RoundedBox
      imageType=NoImage
      mirroring=NoMirror
      resourceName=None
      resourceObject=null
      tiling=NoTile
      imageSize=<0 0>
      margin=<ltrb 0 0 0 0>
      outlineSettings {
       bUseBrushTransparency=false
       roundingType=FixedRadius
       width=1
       color {colorUseRule=UseColor_Specified, specifiedColor=<rgba 0 0 0 0>}
       cornerRadii=<4 4 4 4>
      }
      tintColor {colorUseRule=UseColor_Specified, specifiedColor=<rgba 0.0395 0.0395 0.0395 1>}
      uVRegion {bIsValid=false, max=<0 0>, min=<0 0>}
     }
     insetPadding=<ltrb 1 1 1 1>
     textPadding=<ltrb 7 2.5 7 2.5>
    }
   }
   slot {class=CanvasPanelSlot, properties {layoutData {alignment=<0 0>, anchors {maximum=<0 0>, minimum=<0 0>}, offsets=<ltrb 250 2 80 20>}}}
  }
  resetButton {
   widgetClassPath=&Button
   parent=CanvasPanel_0
   widget=resetButton
   properties {widgetStyle {
     clickedSlateSound {resourceObject=null}
     disabled {
      bIsDynamicallyLoaded=false
      drawAs=RoundedBox
      imageType=NoImage
      mirroring=NoMirror
      resourceName=None
      resourceObject=null
      tiling=NoTile
      imageSize=<0 0>
      margin=<ltrb 0 0 0 0>
      outlineSettings {
       bUseBrushTransparency=false
       roundingType=FixedRadius
       width=1
       color {colorUseRule=UseColor_ColorTable, specifiedColor=<rgba 0 0 0 0>}
       cornerRadii=<4 4 4 4>
      }
      tintColor {colorUseRule=UseColor_Specified, specifiedColor=<rgba 0.0395 0.0395 0.0395 1>}
      uVRegion {bIsValid=false, max=<0 0>, min=<0 0>}
     }
     disabledForeground {colorUseRule=UseColor_ColorTable, specifiedColor=<rgba 0 0 0 0>}
     hovered {
      bIsDynamicallyLoaded=false
      drawAs=RoundedBox
      imageType=NoImage
      mirroring=NoMirror
      resourceName=None
      resourceObject=null
      tiling=NoTile
      imageSize=<0 0>
      margin=<ltrb 0 0 0 0>
      outlineSettings {
       bUseBrushTransparency=false
       roundingType=FixedRadius
       width=1
       color {colorUseRule=UseColor_ColorTable, specifiedColor=<rgba 0 0 0 0>}
       cornerRadii=<4 4 4 4>
      }
      tintColor {colorUseRule=UseColor_Specified, specifiedColor=<rgba 0.0953 0.0953 0.0953 1>}
      uVRegion {bIsValid=false, max=<0 0>, min=<0 0>}
     }
     hoveredForeground {colorUseRule=UseColor_ColorTable, specifiedColor=<rgba 0 0 0 0>}
     hoveredSlateSound {resourceObject=null}
     normal {
      bIsDynamicallyLoaded=false
      drawAs=RoundedBox
      imageType=NoImage
      mirroring=NoMirror
      resourceName=None
      resourceObject=/Game/Athena/HUD/Art/BacchusHudIcons/Reset_Normal
      tiling=NoTile
      imageSize=<20 20>
      margin=<ltrb 0 0 0 0>
      outlineSettings {
       bUseBrushTransparency=false
       roundingType=FixedRadius
       width=1
       color {colorUseRule=UseColor_ColorTable, specifiedColor=<rgba 0 0 0 0>}
       cornerRadii=<4 4 4 4>
      }
      tintColor {colorUseRule=UseColor_Specified, specifiedColor=<rgba 1 1 1 0.5>}
      uVRegion {bIsValid=false, max=<0 0>, min=<0 0>}
     }
     normalForeground {colorUseRule=UseColor_ColorTable, specifiedColor=<rgba 0 0 0 0>}
     normalPadding=<ltrb 12 1 12 1>
     pressed {
      bIsDynamicallyLoaded=false
      drawAs=RoundedBox
      imageType=NoImage
      mirroring=NoMirror
      resourceName=None
      resourceObject=null
      tiling=NoTile
      imageSize=<0 0>
      margin=<ltrb 0 0 0 0>
      outlineSettings {
       bUseBrushTransparency=false
       roundingType=FixedRadius
       width=1
       color {colorUseRule=UseColor_ColorTable, specifiedColor=<rgba 0 0 0 0>}
       cornerRadii=<4 4 4 4>
      }
      tintColor {colorUseRule=UseColor_Specified, specifiedColor=<rgba 0.0284 0.0284 0.0284 1>}
      uVRegion {bIsValid=false, max=<0 0>, min=<0 0>}
     }
     pressedForeground {colorUseRule=UseColor_ColorTable, specifiedColor=<rgba 0 0 0 0>}
     pressedPadding=<ltrb 12 2 12 0>
     pressedSlateSound {resourceObject=null}
    }}
   slot {class=CanvasPanelSlot, properties {layoutData {alignment=<0 0>, anchors {maximum=<0 0>, minimum=<0 0>}, offsets=<ltrb 336 2 20 20>}}}
  }
 ]
}
