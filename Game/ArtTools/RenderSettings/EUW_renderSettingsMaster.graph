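# Master editor utility widget for render-settings tools, holding shared logic for custom cvar spin box, check box and combo box widgets, with no layout.
# Selecting its tab refreshes those controls from cvars; Deltas moves non-default ones into a scroll box.
# getSaveString calls Fill Widget Arrays, then concatenates each control's save line.
# Fill Widget Arrays clears the typed arrays, but its sorting casts are unwired, leaving only debug prints.
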EditorUtilityWidgetBlueprint /Game/ArtTools/RenderSettings/EUW_renderSettingsMaster
graphs [
 Deltas {nodes [
   n1 = Deltas!() -> (then, scrollBox:Scroll Box Object Reference) {is=title K2Node_FunctionEntry}
   n2 = Self-Reference()
   n3 = Length(TargetArray:Array of EUW Cvar Spin Box Object References=Get spinBoxes() -> (spinBoxes:Array of EUW Cvar Spin Box Object References))
   n4 = Length(TargetArray:Array of EUW Cvar Check Box Object References=Get checkBoxes() -> (checkBoxes:Array of EUW Cvar Check Box Object References))
   n5 = Length(TargetArray:Array of EUW Cvar Combo Box Object References=Get comboBoxes() -> (comboBoxes:Array of EUW Cvar Combo Box Object References))
   > PrintString(InString=Append!(
     A="Deltas "
     B=GetDisplayName(Object=@n2)
     C:String=" "
     D:String="To String (Integer)"(InInt=@n3)
     E:String=" "
     F:String="To String (Integer)"(InInt=@n4)
     G:String=" "
     H:String="To String (Integer)"(InInt=@n5)
    ) -> (ReturnValue))
   > n6 = For Each Loop(Array:Array of EUW Cvar Spin Box Object References=Get spinBoxes()) -> (Array Element:EUW Cvar Spin Box Object Reference)
   n7 = For Each Loop(Exec=@n6.Completed, Array:Array of EUW Cvar Check Box Object References=Get checkBoxes()) -> (Array Element:EUW Cvar Check Box Object Reference)
   n8 = For Each Loop(Exec=@n7.Completed, Array:Array of EUW Cvar Combo Box Object References=Get comboBoxes()) -> (Array Element:EUW Cvar Combo Box Object Reference)
   > n9 = getValue!(execute, self:EUW Cvar Combo Box Object Reference=@n8) -> (then, option:String)
   > Branch(Condition="Not Equal Exactly (String)"(A=@n9, B=Get defaultValue!(self:EUW Cvar Combo Box Object Reference=@n8) -> (defaultValue:String)))
   > PrintString(InString=GetDisplayName(Object=@n8))
   > AddChild(self=@n1, Content=@n8)
   n10 = getValue!(execute=@n7.LoopBody, self:EUW Cvar Check Box Object Reference=@n7) -> (then, value:bool)
   > Branch(Condition="Not Equal (Boolean)"!(A:bool=@n10, B:bool=Get defaultValue(self:EUW Cvar Check Box Object Reference=@n7) -> (defaultValue:bool)) -> (ReturnValue:bool) {is=title K2Node_CallFunction})
   > PrintString(InString=GetDisplayName(Object=@n7))
   > AddChild(self=@n1, Content=@n7)
   n11 = getValue(execute=@n6.LoopBody, self:EUW Cvar Spin Box Object Reference=@n6) -> (value:double)
   > Branch(Condition="Not Equal (Float)"!(A:double=@n11, B:double=Get defaultValue(self:EUW Cvar Spin Box Object Reference=@n6) -> (defaultValue:double)) -> (ReturnValue:bool) {is=title K2Node_CallFunction})
   > PrintString(InString=GetDisplayName(Object=@n6))
   > AddChild(self=@n1, Content=@n6)
  ]}
 EventGraph {nodes [
   Event onTabSelected() {is=title K2Node_Event}
   > refreshUI!(execute, self, resetDefault:bool=false) -> (then)
   Event PreConstruct()
   Event Construct()
   Event Tick()
  ]}
 Fill Widget Arrays {nodes [
   n1 = Fill Widget Arrays!() -> (then, panel:Scroll Box Object Reference) {is=title K2Node_FunctionEntry}
   n2 = GetAllChildren(self=@n1)
   n3 = For Each Loop(Exec=@n9.then, Array:Array of Widget Object References=@n2) -> (Array Element:Widget Object Reference)
   n4 = Get Class(Object=@n3) -> (ReturnValue:Widget Class Reference)
   n5 = Append!(
    A=GetDisplayName(Object=@n3)
    B=" "
    C=Get Class Display Name(Class=@n4)
    D=" "
    E="To String (Boolean)"(InBool=ClassIsChildOf(TestClass=@n4, ParentClass=/Script/UMG.PanelWidget))
   ) -> (ReturnValue)
   PrintString(InString=@n5)
   > n6 = Cast To EUW_cvar_spinBox!(execute, Object=@n3) -> (then, CastFailed, AsEUW Cvar Spin Box:EUW Cvar Spin Box Object Reference) {is=title K2Node_DynamicCast}
   > Add(TargetArray:Array of EUW Cvar Spin Box Object References=Get spinBoxes(), NewItem:"EUW Cvar Spin Box Object Reference (by ref)"=@n6) {is=title K2Node_CallArrayFunction}
   n7 = Cast To EUW_cvar_checkBox!(execute=@n6.CastFailed, Object=@n3) -> (then, CastFailed, AsEUW Cvar Check Box:EUW Cvar Check Box Object Reference) {is=title K2Node_DynamicCast}
   > Add(TargetArray:Array of EUW Cvar Check Box Object References=Get checkBoxes(), NewItem:"EUW Cvar Check Box Object Reference (by ref)"=@n7)
   n8 = Cast To EUW_cvar_comboBox!(execute=@n7.CastFailed, Object=@n3) -> (then, CastFailed, AsEUW Cvar Combo Box:EUW Cvar Combo Box Object Reference) {is=title K2Node_DynamicCast}
   > Add(TargetArray:Array of EUW Cvar Combo Box Object References=Get comboBoxes(), NewItem:"EUW Cvar Combo Box Object Reference (by ref)"=@n8)
   PrintString(execute=@n11.then, InString=Append!(A="panel widget ", B=@n5) -> (ReturnValue))
   PrintString(execute=@n11.CastFailed, InString=Append(A="not panel widget ", B=@n5))
   Clear(execute=@n10.then, TargetArray:Array of EUW Cvar Spin Box Object References=Get spinBoxes())
   > Clear(TargetArray:Array of EUW Cvar Check Box Object References=Get checkBoxes())
   > Clear(TargetArray:Array of EUW Cvar Combo Box Object References=Get comboBoxes())
   > n9 = PrintString(InString="To String (Integer)"(InInt=Length(TargetArray:Array of Widget Object References=@n2)))
   n10 = PrintString(execute=@n1.then, InString=Append(A="Fill Widget Arrays ", B=GetDisplayName(Object=@n1)))
   n11 = Cast To PanelWidget!(execute=@n3.LoopBody, Object=@n3) -> (then, CastFailed, AsPanel Widget:Panel Widget Object Reference) {is=title K2Node_DynamicCast}
  ]}
 getSaveString {nodes [
   n1 = getSaveString!() -> (then, scroll box:Scroll Box Object Reference) {is=title K2Node_FunctionEntry}
   > Fill Widget Arrays!(execute, self, panel:Scroll Box Object Reference=@n1) -> (then) {is=title K2Node_CallFunction}
   > n2 = For Each Loop(Array:Array of EUW Cvar Check Box Object References=Get checkBoxes()) -> (Array Element:EUW Cvar Check Box Object Reference)
   n3 = For Each Loop(Exec=@n2.Completed, Array:Array of EUW Cvar Combo Box Object References=Get comboBoxes()) -> (Array Element:EUW Cvar Combo Box Object Reference)
   n4 = For Each Loop(Exec=@n3.Completed, Array:Array of EUW Cvar Spin Box Object References=Get spinBoxes()) -> (Array Element:EUW Cvar Spin Box Object Reference)
   Return Node!(execute=@n4.Completed, save:String=Get saveString() -> (saveString:String))
   > n5 = save!(execute, self:EUW Cvar Spin Box Object Reference=@n4) -> (then, value:String)
   > Set saveString(saveString:String=Append!(A=Get saveString(), B=@n5, C:String="\r\n") -> (ReturnValue)) -> (Output_Get:String)
   n6 = save(execute=@n3.LoopBody, self:EUW Cvar Combo Box Object Reference=@n3)
   > Set saveString(saveString=Append(A=Get saveString(), B=@n6, C="\r\n"))
   Branch(execute=@n2.LoopBody, Condition="Not Equal Exactly (String)"(A=Get cvar!(self:EUW Cvar Check Box Object Reference=@n2) -> (cvar:String)))
   > n7 = save(self:EUW Cvar Check Box Object Reference=@n2)
   > Set saveString(saveString=Append(A=Get saveString(), B=@n7, C="\r\n"))
  ]}
 refreshUI {nodes [
   n1 = refreshUI!() -> (then, resetDefault:bool=true) {is=title K2Node_FunctionEntry}
   > n2 = Sequence!(execute) -> (then_0, then_1, then_2:Exec)
   > n3 = For Each Loop(Array=Get spinBoxes())
   > Get Cvar Set UI!(execute, self:EUW Cvar Spin Box Object Reference=@n3, resetDefault:bool=@n1) -> (then) {is=title K2Node_CallFunction}
   n4 = For Each Loop(Exec=@n2.then_1, Array:Array of EUW Cvar Check Box Object References=Get checkBoxes()) -> (Array Element:EUW Cvar Check Box Object Reference)
   > getCvarSetUI!(execute, self:EUW Cvar Check Box Object Reference=@n4, setDefault:bool=@n1) -> (then)
   n5 = For Each Loop(Exec=@n2.then_2, Array:Array of EUW Cvar Combo Box Object References=Get comboBoxes()) -> (Array Element:EUW Cvar Combo Box Object Reference)
   > getCvarSetUI(self:EUW Cvar Combo Box Object Reference=@n5, setDefault=@n1)
  ]}
]
variables [
 checkBoxes {type="array<EUW_cvar_checkBox_C>", default []}
 comboBoxes {type="array<EUW_cvar_comboBox_C>", default []}
 spinBoxes {type="array<EUW_cvar_spinBox_C>", default []}
]
widgets {info {rootWidgetClass=null, widgetCount=0}, widgets []}
