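# Editor utility widget for the render settings tool that exposes raytracing and related console variables, including reflections, shadows, skylight, AO, GI, translucency and DLSS.
# Collapsible sections in a scroll box hold reusable checkbox, combo box and spin box widgets, each configured with a cvar.
# Fill Widget Arrays registers these controls by type, then passes the scroll box to the EUW_renderSettingsMaster parent.
# Construct calls refreshUI with resetDefault true; tab selection calls it with false.
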EditorUtilityWidgetBlueprint /Game/ArtTools/RenderSettings/EUW_raytraceSettings
graphs [
 EventGraph {nodes [
   Event Construct()
   > Fill Widget Arrays!(execute, self:EUW Render Settings Master Object Reference, panel:Scroll Box Object Reference) -> (then)
   > refreshUI!(execute, self:EUW Render Settings Master Object Reference, resetDefault:bool=true) -> (then)
   Event onTabSelected() {is=title K2Node_Event}
   > PrintString(InString=Raytracng)
   > refreshUI(resetDefault=false)
   Event PreConstruct()
   Event Tick()
  ]}
 Fill Widget Arrays {nodes [
   Fill Widget Arrays!() -> (then, panel:Scroll Box Object Reference) {is=title K2Node_FunctionEntry}
   n1 = Self-Reference()
   > Clear(TargetArray:Array of EUW Cvar Check Box Object References=Get checkBoxes(None=@n1) -> (checkBoxes:Array of EUW Cvar Check Box Object References))
   > Clear(TargetArray:Array of EUW Cvar Spin Box Object References=Get spinBoxes(None=@n1) -> (spinBoxes:Array of EUW Cvar Spin Box Object References))
   > Clear(TargetArray:Array of EUW Cvar Combo Box Object References=Get comboBoxes(None=@n1) -> (comboBoxes:Array of EUW Cvar Combo Box Object References))
   > Set checkBoxes(checkBoxes:Array of EUW Cvar Check Box Object References=Make Array!(
     "[0]":EUW Cvar Check Box Object Reference=Get EUW_checkBox_DiffuseIndirectDenoiser() -> (EUW_checkBox_DiffuseIndirectDenoiser:EUW Cvar Check Box Object Reference)
     "[1]":EUW Cvar Check Box Object Reference=Get EUW_checkBox_GI_Denoiser() -> (EUW_checkBox_GI_Denoiser:EUW Cvar Check Box Object Reference)
     "[2]":EUW Cvar Check Box Object Reference=Get EUW_checkBox_GI_EvalSkylight() -> (EUW_checkBox_GI_EvalSkylight:EUW Cvar Check Box Object Reference)
     "[3]":EUW Cvar Check Box Object Reference=Get EUW_checkBox_rayTracingShadows() -> (EUW_checkBox_rayTracingShadows:EUW Cvar Check Box Object Reference)
     "[4]":EUW Cvar Check Box Object Reference=Get EUW_checkBox_shadowDenoiser() -> (EUW_checkBox_shadowDenoiser:EUW Cvar Check Box Object Reference)
     "[5]":EUW Cvar Check Box Object Reference=Get EUW_checkBox_shadowDenoiser_1() -> (EUW_checkBox_shadowDenoiser_1:EUW Cvar Check Box Object Reference)
     "[6]":EUW Cvar Check Box Object Reference=Get EUW_checkBox_skylightDenoiser() -> (EUW_checkBox_skylightDenoiser:EUW Cvar Check Box Object Reference)
     "[7]":EUW Cvar Check Box Object Reference=Get EUW_checkBox_SSGI() -> (EUW_checkBox_SSGI:EUW Cvar Check Box Object Reference)
     "[8]":EUW Cvar Check Box Object Reference=Get EUW_checkBox_SSGI_halfRes() -> (EUW_checkBox_SSGI_halfRes:EUW Cvar Check Box Object Reference)
     "[9]":EUW Cvar Check Box Object Reference=Get EUW_checkBox_Translucency_Refraction() -> (EUW_checkBox_Translucency_Refraction:EUW Cvar Check Box Object Reference)
     "[10]":EUW Cvar Check Box Object Reference=Get EUW_checkBox_DLSS_Enable() -> (EUW_checkBox_DLSS_Enable:EUW Cvar Check Box Object Reference)
    ) -> (Array:Array of EUW Cvar Check Box Object References)) -> (Output_Get:Array of EUW Cvar Check Box Object References)
   > Set comboBoxes(comboBoxes:Array of EUW Cvar Combo Box Object References=Make Array!(
     "[0]":EUW Cvar Combo Box Object Reference=Get EUW_comboBox_AO() -> (EUW_comboBox_AO:EUW Cvar Combo Box Object Reference)
     "[1]":EUW Cvar Combo Box Object Reference=Get EUW_comboBox_AO_Denoiser() -> (EUW_comboBox_AO_Denoiser:EUW Cvar Combo Box Object Reference)
     "[2]":EUW Cvar Combo Box Object Reference=Get EUW_comboBox_AO_method() -> (EUW_comboBox_AO_method:EUW Cvar Combo Box Object Reference)
     "[3]":EUW Cvar Combo Box Object Reference=Get EUW_comboBox_forceAll() -> (EUW_comboBox_forceAll:EUW Cvar Combo Box Object Reference)
     "[4]":EUW Cvar Combo Box Object Reference=Get EUW_comboBox_GI() -> (EUW_comboBox_GI:EUW Cvar Combo Box Object Reference)
     "[5]":EUW Cvar Combo Box Object Reference=Get EUW_comboBox_reflections() -> (EUW_comboBox_reflections:EUW Cvar Combo Box Object Reference)
     "[6]":EUW Cvar Combo Box Object Reference=Get EUW_comboBox_Reflections_Denoiser_1() -> (EUW_comboBox_Reflections_Denoiser_1:EUW Cvar Combo Box Object Reference)
     "[7]":EUW Cvar Combo Box Object Reference=Get EUW_comboBox_Reflections_Shadows_1() -> (EUW_comboBox_Reflections_Shadows_1:EUW Cvar Combo Box Object Reference)
     "[8]":EUW Cvar Combo Box Object Reference=Get EUW_comboBox_Reflections_Translucency() -> (EUW_comboBox_Reflections_Translucency:EUW Cvar Combo Box Object Reference)
     "[9]":EUW Cvar Combo Box Object Reference=Get EUW_comboBox_Translucency() -> (EUW_comboBox_Translucency:EUW Cvar Combo Box Object Reference)
     "[10]":EUW Cvar Combo Box Object Reference=Get EUW_comboBox_Translucency_Shadows() -> (EUW_comboBox_Translucency_Shadows:EUW Cvar Combo Box Object Reference)
     "[11]":EUW Cvar Combo Box Object Reference=Get EUW_comboBox_DLSS_Quality() -> (EUW_comboBox_DLSS_Quality:EUW Cvar Combo Box Object Reference)
    ) -> (Array:Array of EUW Cvar Combo Box Object References)) -> (Output_Get:Array of EUW Cvar Combo Box Object References)
   > Set spinBoxes(spinBoxes:Array of EUW Cvar Spin Box Object References=Make Array!(
     "[0]":EUW Cvar Spin Box Object Reference=Get EUW_slider_GI_MaxBounces() -> (EUW_slider_GI_MaxBounces:EUW Cvar Spin Box Object Reference)
     "[1]":EUW Cvar Spin Box Object Reference=Get EUW_slider_GI_SamplesPerPixel_1() -> (EUW_slider_GI_SamplesPerPixel_1:EUW Cvar Spin Box Object Reference)
     "[2]":EUW Cvar Spin Box Object Reference=Get EUW_slider_GI_ScreenPercentage() -> (EUW_slider_GI_ScreenPercentage:EUW Cvar Spin Box Object Reference)
     "[3]":EUW Cvar Spin Box Object Reference=Get EUW_slider_Reflections_MaxBounces() -> (EUW_slider_Reflections_MaxBounces:EUW Cvar Spin Box Object Reference)
     "[4]":EUW Cvar Spin Box Object Reference=Get EUW_slider_Reflections_MaxRoughness() -> (EUW_slider_Reflections_MaxRoughness:EUW Cvar Spin Box Object Reference)
     "[5]":EUW Cvar Spin Box Object Reference=Get EUW_slider_Reflections_SamplesPerPixel() -> (EUW_slider_Reflections_SamplesPerPixel:EUW Cvar Spin Box Object Reference)
     "[6]":EUW Cvar Spin Box Object Reference=Get EUW_slider_Reflections_ScreenPercentage() -> (EUW_slider_Reflections_ScreenPercentage:EUW Cvar Spin Box Object Reference)
     "[7]":EUW Cvar Spin Box Object Reference=Get EUW_slider_SkylightSamplesPerPixel() -> (EUW_slider_SkylightSamplesPerPixel:EUW Cvar Spin Box Object Reference)
     "[8]":EUW Cvar Spin Box Object Reference=Get EUW_slider_Translucency_MaxRefractionRays() -> (EUW_slider_Translucency_MaxRefractionRays:EUW Cvar Spin Box Object Reference)
     "[9]":EUW Cvar Spin Box Object Reference=Get EUW_slider_Translucency_SamplesPerPixel() -> (EUW_slider_Translucency_SamplesPerPixel:EUW Cvar Spin Box Object Reference)
     "[10]":EUW Cvar Spin Box Object Reference=Get EUW_slider_Translucency_SamplesPerPixel_1() -> (EUW_slider_Translucency_SamplesPerPixel_1:EUW Cvar Spin Box Object Reference)
    ) -> (Array:Array of EUW Cvar Spin Box Object References)) -> (Output_Get:Array of EUW Cvar Spin Box Object References)
   Parent: Fill Widget Arrays!(execute, panel:Scroll Box Object Reference=Get ScrollBox() -> (ScrollBox:Scroll Box Object Reference)) -> (then) {is=title K2Node_CallParentFunction}
  ]}
]
variables []
widgets {
 info {
  namedSlotCount=18
  parentClass=&/EUW_renderSettingsMaster.EUW_renderSettingsMaster_C
  rootWidgetClass=&CanvasPanel
  widgetCount=71
 }
 widgets [
  CanvasPanel_0 {widgetClassPath=&CanvasPanel, widget=CanvasPanel_0}
  ScrollBox {
   widgetClassPath=&ScrollBox
   bIsVariable=true
   parent=CanvasPanel_0
   widget=ScrollBox
   properties {
    alwaysShowScrollbar=true
    consumeMouseWheel=Always
    widgetBarStyle {
     thickness=8
     draggedThumbImage {
      bIsDynamicallyLoaded=false
      drawAs=RoundedBox
      imageType=NoImage
      mirroring=NoMirror
      resourceName=None
      resourceObject=null
      tiling=NoTile
      imageSize=<0 0>
      margin=<ltrb 0 0 0 0>
      outlineSettings {
       bUseBrushTransparency=false
       roundingType=FixedRadius
       width=0
       color {colorUseRule=UseColor_Specified, specifiedColor=<rgba 0 0 0 0>}
       cornerRadii=<4 4 4 4>
      }
      tintColor {colorUseRule=UseColor_Specified, specifiedColor=<rgba 0.2159 0.2159 0.2159 1>}
      uVRegion {bIsValid=false, max=<0 0>, min=<0 0>}
     }
     horizontalBackgroundImage {
      bIsDynamicallyLoaded=false
      drawAs=NoDrawType
      imageType=NoImage
      mirroring=NoMirror
      resourceName=None
      resourceObject=null
      tiling=NoTile
      imageSize=<0 0>
      margin=<ltrb 0 0 0 0>
      outlineSettings {
       bUseBrushTransparency=false
       roundingType=HalfHeightRadius
       width=0
       color {colorUseRule=UseColor_Specified, specifiedColor=<rgba 0 0 0 0>}
       cornerRadii=<0 0 0 0>
      }
      tintColor {colorUseRule=UseColor_Specified, specifiedColor=<rgba 1 1 1 1>}
      uVRegion {bIsValid=false, max=<0 0>, min=<0 0>}
     }
     horizontalBottomSlotImage {
      bIsDynamicallyLoaded=false
      drawAs=NoDrawType
      imageType=NoImage
      mirroring=NoMirror
      resourceName=None
      resourceObject=null
      tiling=NoTile
      imageSize=<0 0>
      margin=<ltrb 0 0 0 0>
      outlineSettings {
       bUseBrushTransparency=false
       roundingType=HalfHeightRadius
       width=0
       color {colorUseRule=UseColor_Specified, specifiedColor=<rgba 0 0 0 0>}
       cornerRadii=<0 0 0 0>
      }
      tintColor {colorUseRule=UseColor_Specified, specifiedColor=<rgba 1 1 1 1>}
      uVRegion {bIsValid=false, max=<0 0>, min=<0 0>}
     }
     horizontalTopSlotImage {
      bIsDynamicallyLoaded=false
      drawAs=NoDrawType
      imageType=NoImage
      mirroring=NoMirror
      resourceName=None
      resourceObject=null
      tiling=NoTile
      imageSize=<0 0>
      margin=<ltrb 0 0 0 0>
      outlineSettings {
       bUseBrushTransparency=false
       roundingType=HalfHeightRadius
       width=0
       color {colorUseRule=UseColor_Specified, specifiedColor=<rgba 0 0 0 0>}
       cornerRadii=<0 0 0 0>
      }
      tintColor {colorUseRule=UseColor_Specified, specifiedColor=<rgba 1 1 1 1>}
      uVRegion {bIsValid=false, max=<0 0>, min=<0 0>}
     }
     hoveredThumbImage {
      bIsDynamicallyLoaded=false
      drawAs=RoundedBox
      imageType=NoImage
      mirroring=NoMirror
      resourceName=None
      resourceObject=null
      tiling=NoTile
      imageSize=<0 0>
      margin=<ltrb 0 0 0 0>
      outlineSettings {
       bUseBrushTransparency=false
       roundingType=FixedRadius
       width=0
       color {colorUseRule=UseColor_Specified, specifiedColor=<rgba 0 0 0 0>}
       cornerRadii=<4 4 4 4>
      }
      tintColor {colorUseRule=UseColor_Specified, specifiedColor=<rgba 0.2159 0.2159 0.2159 1>}
      uVRegion {bIsValid=false, max=<0 0>, min=<0 0>}
     }
     normalThumbImage {
      bIsDynamicallyLoaded=false
      drawAs=RoundedBox
      imageType=NoImage
      mirroring=NoMirror
      resourceName=None
      resourceObject=null
      tiling=NoTile
      imageSize=<0 0>
      margin=<ltrb 0 0 0 0>
      outlineSettings {
       bUseBrushTransparency=false
       roundingType=FixedRadius
       width=0
       color {colorUseRule=UseColor_Specified, specifiedColor=<rgba 0 0 0 0>}
       cornerRadii=<4 4 4 4>
      }
      tintColor {colorUseRule=UseColor_Specified, specifiedColor=<rgba 0.0953 0.0953 0.0953 1>}
      uVRegion {bIsValid=false, max=<0 0>, min=<0 0>}
     }
     verticalBackgroundImage {
      bIsDynamicallyLoaded=false
      drawAs=NoDrawType
      imageType=NoImage
      mirroring=NoMirror
      resourceName=None
      resourceObject=null
      tiling=NoTile
      imageSize=<0 0>
      margin=<ltrb 0 0 0 0>
      outlineSettings {
       bUseBrushTransparency=false
       roundingType=HalfHeightRadius
       width=0
       color {colorUseRule=UseColor_Specified, specifiedColor=<rgba 0 0 0 0>}
       cornerRadii=<0 0 0 0>
      }
      tintColor {colorUseRule=UseColor_Specified, specifiedColor=<rgba 1 1 1 1>}
      uVRegion {bIsValid=false, max=<0 0>, min=<0 0>}
     }
     verticalBottomSlotImage {
      bIsDynamicallyLoaded=false
      drawAs=NoDrawType
      imageType=NoImage
      mirroring=NoMirror
      resourceName=None
      resourceObject=null
      tiling=NoTile
      imageSize=<0 0>
      margin=<ltrb 0 0 0 0>
      outlineSettings {
       bUseBrushTransparency=false
       roundingType=HalfHeightRadius
       width=0
       color {colorUseRule=UseColor_Specified, specifiedColor=<rgba 0 0 0 0>}
       cornerRadii=<0 0 0 0>
      }
      tintColor {colorUseRule=UseColor_Specified, specifiedColor=<rgba 1 1 1 1>}
      uVRegion {bIsValid=false, max=<0 0>, min=<0 0>}
     }
     verticalTopSlotImage {
      bIsDynamicallyLoaded=false
      drawAs=NoDrawType
      imageType=NoImage
      mirroring=NoMirror
      resourceName=None
      resourceObject=null
      tiling=NoTile
      imageSize=<0 0>
      margin=<ltrb 0 0 0 0>
      outlineSettings {
       bUseBrushTransparency=false
       roundingType=HalfHeightRadius
       width=0
       color {colorUseRule=UseColor_Specified, specifiedColor=<rgba 0 0 0 0>}
       cornerRadii=<0 0 0 0>
      }
      tintColor {colorUseRule=UseColor_Specified, specifiedColor=<rgba 1 1 1 1>}
      uVRegion {bIsValid=false, max=<0 0>, min=<0 0>}
     }
    }
    widgetStyle {
     barThickness=8
     bottomShadowBrush {
      bIsDynamicallyLoaded=false
      drawAs=Box
      imageType=FullColor
      mirroring=NoMirror
      resourceName=../../../Engine/Content/Slate/Common/ScrollBoxShadowBottom.png
      resourceObject=null
      tiling=NoTile
      imageSize=<16 8>
      margin=<ltrb 0.5 0 0.5 1>
      outlineSettings {
       bUseBrushTransparency=false
       roundingType=HalfHeightRadius
       width=0
       color {colorUseRule=UseColor_Specified, specifiedColor=<rgba 0 0 0 0>}
       cornerRadii=<0 0 0 0>
      }
      tintColor {colorUseRule=UseColor_Specified, specifiedColor=<rgba 1 1 1 1>}
      uVRegion {bIsValid=false, max=<0 0>, min=<0 0>}
     }
     horizontalScrolledContentPadding=<ltrb 0 0 1 0>
     leftShadowBrush {
      bIsDynamicallyLoaded=false
      drawAs=Box
      imageType=FullColor
      mirroring=NoMirror
      resourceName=../../../Engine/Content/Slate/Common/ScrollBoxShadowLeft.png
      resourceObject=null
      tiling=NoTile
      imageSize=<8 16>
      margin=<ltrb 1 0.5 0 0.5>
      outlineSettings {
       bUseBrushTransparency=false
       roundingType=HalfHeightRadius
       width=0
       color {colorUseRule=UseColor_Specified, specifiedColor=<rgba 0 0 0 0>}
       cornerRadii=<0 0 0 0>
      }
      tintColor {colorUseRule=UseColor_Specified, specifiedColor=<rgba 1 1 1 1>}
      uVRegion {bIsValid=false, max=<0 0>, min=<0 0>}
     }
     rightShadowBrush {
      bIsDynamicallyLoaded=false
      drawAs=Box
      imageType=FullColor
      mirroring=NoMirror
      resourceName=../../../Engine/Content/Slate/Common/ScrollBoxShadowRight.png
      resourceObject=null
      tiling=NoTile
      imageSize=<8 16>
      margin=<ltrb 0 0.5 1 0.5>
      outlineSettings {
       bUseBrushTransparency=false
       roundingType=HalfHeightRadius
       width=0
       color {colorUseRule=UseColor_Specified, specifiedColor=<rgba 0 0 0 0>}
       cornerRadii=<0 0 0 0>
      }
      tintColor {colorUseRule=UseColor_Specified, specifiedColor=<rgba 1 1 1 1>}
      uVRegion {bIsValid=false, max=<0 0>, min=<0 0>}
     }
     topShadowBrush {
      bIsDynamicallyLoaded=false
      drawAs=Box
      imageType=FullColor
      mirroring=NoMirror
      resourceName=../../../Engine/Content/Slate/Common/ScrollBoxShadowTop.png
      resourceObject=null
      tiling=NoTile
      imageSize=<16 8>
      margin=<ltrb 0.5 1 0.5 0>
      outlineSettings {
       bUseBrushTransparency=false
       roundingType=HalfHeightRadius
       width=0
       color {colorUseRule=UseColor_Specified, specifiedColor=<rgba 0 0 0 0>}
       cornerRadii=<0 0 0 0>
      }
      tintColor {colorUseRule=UseColor_Specified, specifiedColor=<rgba 1 1 1 1>}
      uVRegion {bIsValid=false, max=<0 0>, min=<0 0>}
     }
     verticalScrolledContentPadding=<ltrb 0 0 0 1>
    }
   }
   slot {
    class=CanvasPanelSlot
    properties {bAutoSize=true, layoutData {alignment=<0 0>, anchors {maximum=<1 1>, minimum=<0 0>}, offsets=<ltrb 0 0 0 0>}}
   }
  }
  ExpandableArea_93 {
   widgetClassPath=/Script/UMG.ExpandableArea
   bIsVariable=true
   parent=ScrollBox
   widget=ExpandableArea_93
   properties {
    bIsExpanded=true
    areaPadding=<ltrb 10 10 10 10>
    borderBrush {
     bIsDynamicallyLoaded=false
     drawAs=RoundedBox
     imageType=NoImage
     mirroring=NoMirror
     resourceName=None
     resourceObject=null
     tiling=NoTile
     imageSize=<0 0>
     margin=<ltrb 0 0 0 0>
     outlineSettings {
      bUseBrushTransparency=false
      roundingType=FixedRadius
      width=0
      color {colorUseRule=UseColor_Specified, specifiedColor=<rgba 0 0 0 0>}
      cornerRadii=<4 4 4 4>
     }
     tintColor {colorUseRule=UseColor_Specified, specifiedColor=<rgba 0.0176 0.0176 0.0176 1>}
     uVRegion {bIsValid=false, max=<0 0>, min=<0 0>}
    }
    style {
     rolloutAnimationSeconds=0.1
     collapsedImage {
      bIsDynamicallyLoaded=false
      drawAs=Image
      imageType=Vector
      mirroring=NoMirror
      resourceName=../../../Engine/Content/Slate/Starship/Common/chevron-right.svg
      resourceObject=null
      tiling=NoTile
      imageSize=<16 16>
      margin=<ltrb 0 0 0 0>
      outlineSettings {
       bUseBrushTransparency=false
       roundingType=HalfHeightRadius
       width=0
       color {colorUseRule=UseColor_Specified, specifiedColor=<rgba 0 0 0 0>}
       cornerRadii=<0 0 0 0>
      }
      tintColor {colorUseRule=UseColor_Specified, specifiedColor=<rgba 0.5271 0.5271 0.5271 1>}
      uVRegion {bIsValid=false, max=<0 0>, min=<0 0>}
     }
     expandedImage {
      bIsDynamicallyLoaded=false
      drawAs=Image
      imageType=Vector
      mirroring=NoMirror
      resourceName=../../../Engine/Content/Slate/Starship/Common/chevron-down.svg
      resourceObject=null
      tiling=NoTile
      imageSize=<16 16>
      margin=<ltrb 0 0 0 0>
      outlineSettings {
       bUseBrushTransparency=false
       roundingType=HalfHeightRadius
       width=0
       color {colorUseRule=UseColor_Specified, specifiedColor=<rgba 0 0 0 0>}
       cornerRadii=<0 0 0 0>
      }
      tintColor {colorUseRule=UseColor_Specified, specifiedColor=<rgba 0.5271 0.5271 0.5271 1>}
      uVRegion {bIsValid=false, max=<0 0>, min=<0 0>}
     }
    }
   }
   slot {class=ScrollBoxSlot}
  }
  label_7 {
   widgetClassPath=&TextBlock
   bIsVariable=true
   namedSlotHost=ExpandableArea_93
   widget=label_7
   properties {
    text=Raytracing Global Controls
    font {
     bForceMonospaced=false
     bMaterialIsStencil=false
     fontMaterial=null
     fontObject=&Roboto
     letterSpacing=0
     monospacedWidth=1
     size=10
     skewAmount=0
     typefaceFontName=Bold
     outlineSettings {
      bApplyOutlineToDropShadows=false
      bMiteredCorners=false
      bSeparateFillAlpha=false
      outlineMaterial=null
      outlineSize=0
      outlineColor=<rgba 0 0 0 1>
     }
    }
   }
  }
  VerticalBox_261 {widgetClassPath=&VerticalBox, namedSlotHost=ExpandableArea_93, widget=VerticalBox_261}
  EUW_comboBox_forceAll {
   widgetClassPath=&/EUW_cvar_comboBox
   bIsVariable=true
   parent=VerticalBox_261
   widget=EUW_comboBox_forceAll
   properties {
    cvar=r.RayTracing.ForceAllRayTracingEffects
    labelText=RayTracing Force All RayTracing Effects
    labelToolTip="Force all ray tracing effects ON/OFF.\r\n -1: Do not force (default) \r\n 0: All ray tracing effects disabled\r\n 1: All ray tracing effects enabled"
    options [-1,0,1]
   }
   slot {class=VerticalBoxSlot}
  }
  Spacer_347 {widgetClassPath=&Spacer, parent=ScrollBox, widget=Spacer_347, properties {size=<1 20>}, slot {class=ScrollBoxSlot}}
  ExpandableArea_RaytracingReflections {
   widgetClassPath=&ExpandableArea
   bIsVariable=true
   parent=ScrollBox
   widget=ExpandableArea_RaytracingReflections
   properties {
    bIsExpanded=true
    areaPadding=<ltrb 10 10 10 10>
    borderBrush {
     bIsDynamicallyLoaded=false
     drawAs=RoundedBox
     imageType=NoImage
     mirroring=NoMirror
     resourceName=None
     resourceObject=null
     tiling=NoTile
     imageSize=<0 0>
     margin=<ltrb 0 0 0 0>
     outlineSettings {
      bUseBrushTransparency=false
      roundingType=FixedRadius
      width=0
      color {colorUseRule=UseColor_Specified, specifiedColor=<rgba 0 0 0 0>}
      cornerRadii=<4 4 4 4>
     }
     tintColor {colorUseRule=UseColor_Specified, specifiedColor=<rgba 0.0176 0.0176 0.0176 1>}
     uVRegion {bIsValid=false, max=<0 0>, min=<0 0>}
    }
    style {
     rolloutAnimationSeconds=0.1
     collapsedImage {
      bIsDynamicallyLoaded=false
      drawAs=Image
      imageType=Vector
      mirroring=NoMirror
      resourceName=../../../Engine/Content/Slate/Starship/Common/chevron-right.svg
      resourceObject=null
      tiling=NoTile
      imageSize=<16 16>
      margin=<ltrb 0 0 0 0>
      outlineSettings {
       bUseBrushTransparency=false
       roundingType=HalfHeightRadius
       width=0
       color {colorUseRule=UseColor_Specified, specifiedColor=<rgba 0 0 0 0>}
       cornerRadii=<0 0 0 0>
      }
      tintColor {colorUseRule=UseColor_Specified, specifiedColor=<rgba 0.5271 0.5271 0.5271 1>}
      uVRegion {bIsValid=false, max=<0 0>, min=<0 0>}
     }
     expandedImage {
      bIsDynamicallyLoaded=false
      drawAs=Image
      imageType=Vector
      mirroring=NoMirror
      resourceName=../../../Engine/Content/Slate/Starship/Common/chevron-down.svg
      resourceObject=null
      tiling=NoTile
      imageSize=<16 16>
      margin=<ltrb 0 0 0 0>
      outlineSettings {
       bUseBrushTransparency=false
       roundingType=HalfHeightRadius
       width=0
       color {colorUseRule=UseColor_Specified, specifiedColor=<rgba 0 0 0 0>}
       cornerRadii=<0 0 0 0>
      }
      tintColor {colorUseRule=UseColor_Specified, specifiedColor=<rgba 0.5271 0.5271 0.5271 1>}
      uVRegion {bIsValid=false, max=<0 0>, min=<0 0>}
     }
    }
   }
   slot {class=ScrollBoxSlot}
  }
  label_4 {
   widgetClassPath=&TextBlock
   bIsVariable=true
   namedSlotHost=ExpandableArea_RaytracingReflections
   widget=label_4
   properties {
    text=Raytracing Reflections
    font {
     bForceMonospaced=false
     bMaterialIsStencil=false
     fontMaterial=null
     fontObject=&Roboto
     letterSpacing=0
     monospacedWidth=1
     size=10
     skewAmount=0
     typefaceFontName=Bold
     outlineSettings {
      bApplyOutlineToDropShadows=false
      bMiteredCorners=false
      bSeparateFillAlpha=false
      outlineMaterial=null
      outlineSize=0
      outlineColor=<rgba 0 0 0 1>
     }
    }
   }
  }
  VerticalBox_299 {widgetClassPath=&VerticalBox, namedSlotHost=ExpandableArea_RaytracingReflections, widget=VerticalBox_299}
  EUW_comboBox_reflections {
   widgetClassPath=&EUW_cvar_comboBox
   bIsVariable=true
   parent=VerticalBox_299
   widget=EUW_comboBox_reflections
   properties {
    cvar=r.RayTracing.Reflections
    labelText=Raytracing Reflections
    labelToolTip="-1: Value driven by postprocess volume (default) \r\n0: use traditional rasterized SSR\r\n1: use ray traced reflections"
    option=-1
    options [-1,0,1]
   }
   slot {class=VerticalBoxSlot}
  }
  EUW_slider_Reflections_MaxRoughness {
   widgetClassPath=&/EUW_cvar_spinBox
   bIsVariable=true
   parent=VerticalBox_299
   widget=EUW_slider_Reflections_MaxRoughness
   properties {
    cvar=r.RayTracing.Reflections.MaxRoughness
    labelText=Raytracing Reflections MaxRoughness
    labelToolTip="Sets the maximum roughness until which ray tracing reflections will be visible (default = -1 (max roughness driven by postprocessing volume))"
    max=1
    min=-1
    step=0.01
   }
   slot {class=VerticalBoxSlot}
  }
  EUW_slider_Reflections_MaxBounces {
   widgetClassPath=&EUW_cvar_spinBox
   bIsVariable=true
   parent=VerticalBox_299
   widget=EUW_slider_Reflections_MaxBounces
   properties {
    cvar=r.RayTracing.Reflections.MaxBounces
    labelText=Raytracing Reflections MaxBounces
    labelToolTip="Sets the maximum number of ray tracing reflection bounces (default = -1 (max bounces driven by postprocessing volume))"
    max=32
    min=-1
   }
   slot {class=VerticalBoxSlot}
  }
  EUW_slider_Reflections_SamplesPerPixel {
   widgetClassPath=&EUW_cvar_spinBox
   bIsVariable=true
   parent=VerticalBox_299
   widget=EUW_slider_Reflections_SamplesPerPixel
   properties {
    cvar=r.RayTracing.Reflections.SamplesPerPixel
    labelText=Raytracing Reflections SamplesPerPixel
    labelToolTip="Sets the samples-per-pixel for reflections (default = -1 (driven by postprocesing volume))"
    max=32
    min=-1
   }
   slot {class=VerticalBoxSlot}
  }
  EUW_slider_Reflections_ScreenPercentage {
   widgetClassPath=&EUW_cvar_spinBox
   bIsVariable=true
   parent=VerticalBox_299
   widget=EUW_slider_Reflections_ScreenPercentage
   properties {
    cvar=r.RayTracing.Reflections.ScreenPercentage
    labelText=Raytracing Reflections ScreenPercentage
    labelToolTip="Screen percentage the reflections should be ray traced at (default = 100)."
    max=100
   }
   slot {class=VerticalBoxSlot}
  }
  EUW_comboBox_Reflections_Shadows_1 {
   widgetClassPath=&EUW_cvar_comboBox
   bIsVariable=true
   parent=VerticalBox_299
   widget=EUW_comboBox_Reflections_Shadows_1
   properties {
    cvar=r.rayTracing.reflections.shadows
    labelText=Raytracing Reflections Shadows
    labelToolTip="Enables shadows in ray tracing reflections) -1: Shadows driven by postprocessing volume (default) 0: Shadows disabled  1: Hard shadows 2: Soft area shadows"
    option=-1
    options [-1,0,1,2]
   }
   slot {class=VerticalBoxSlot}
  }
  EUW_comboBox_Reflections_Denoiser_1 {
   widgetClassPath=&EUW_cvar_comboBox
   bIsVariable=true
   parent=VerticalBox_299
   widget=EUW_comboBox_Reflections_Denoiser_1
   properties {
    cvar=r.reflections.denoiser
    labelText=Reflections Denoiser
    labelToolTip=" 0: Disabled;\r\n 1: Forces the default denoiser of the renderer;\r\n 2: GScreenSpaceDenoiser which may be overriden by a third party plugin (default)."
    option=-1
    options [0,1,2]
   }
   slot {class=VerticalBoxSlot}
  }
  EUW_comboBox_Reflections_Translucency {
   widgetClassPath=&EUW_cvar_comboBox
   bIsVariable=true
   parent=VerticalBox_299
   widget=EUW_comboBox_Reflections_Translucency
   properties {
    cvar=r.RayTracing.Reflections.Translucency
    labelText=Raytracing Reflections Translucency
    labelToolTip="Translucent objects visible in ray tracing reflections) \r\n-1: Driven by postprocessing volume (default)\r\n0: Translucent objects not visible\r\n1: Translucent objects visible"
    option=-1
    options [-1,0,1]
   }
   slot {class=VerticalBoxSlot}
  }
  Spacer_2 {widgetClassPath=&Spacer, parent=ScrollBox, widget=Spacer_2, properties {size=<1 20>}, slot {class=ScrollBoxSlot}}
  ExpandableArea_Raytracing_DirectionalShadows {
   widgetClassPath=&ExpandableArea
   bIsVariable=true
   parent=ScrollBox
   widget=ExpandableArea_Raytracing_DirectionalShadows
   properties {
    bIsExpanded=true
    areaPadding=<ltrb 10 10 10 10>
    borderBrush {
     bIsDynamicallyLoaded=false
     drawAs=RoundedBox
     imageType=NoImage
     mirroring=NoMirror
     resourceName=None
     resourceObject=null
     tiling=NoTile
     imageSize=<0 0>
     margin=<ltrb 0 0 0 0>
     outlineSettings {
      bUseBrushTransparency=false
      roundingType=FixedRadius
      width=0
      color {colorUseRule=UseColor_Specified, specifiedColor=<rgba 0 0 0 0>}
      cornerRadii=<4 4 4 4>
     }
     tintColor {colorUseRule=UseColor_Specified, specifiedColor=<rgba 0.0176 0.0176 0.0176 1>}
     uVRegion {bIsValid=false, max=<0 0>, min=<0 0>}
    }
    style {
     rolloutAnimationSeconds=0.1
     collapsedImage {
      bIsDynamicallyLoaded=false
      drawAs=Image
      imageType=Vector
      mirroring=NoMirror
      resourceName=../../../Engine/Content/Slate/Starship/Common/chevron-right.svg
      resourceObject=null
      tiling=NoTile
      imageSize=<16 16>
      margin=<ltrb 0 0 0 0>
      outlineSettings {
       bUseBrushTransparency=false
       roundingType=HalfHeightRadius
       width=0
       color {colorUseRule=UseColor_Specified, specifiedColor=<rgba 0 0 0 0>}
       cornerRadii=<0 0 0 0>
      }
      tintColor {colorUseRule=UseColor_Specified, specifiedColor=<rgba 0.5271 0.5271 0.5271 1>}
      uVRegion {bIsValid=false, max=<0 0>, min=<0 0>}
     }
     expandedImage {
      bIsDynamicallyLoaded=false
      drawAs=Image
      imageType=Vector
      mirroring=NoMirror
      resourceName=../../../Engine/Content/Slate/Starship/Common/chevron-down.svg
      resourceObject=null
      tiling=NoTile
      imageSize=<16 16>
      margin=<ltrb 0 0 0 0>
      outlineSettings {
       bUseBrushTransparency=false
       roundingType=HalfHeightRadius
       width=0
       color {colorUseRule=UseColor_Specified, specifiedColor=<rgba 0 0 0 0>}
       cornerRadii=<0 0 0 0>
      }
      tintColor {colorUseRule=UseColor_Specified, specifiedColor=<rgba 0.5271 0.5271 0.5271 1>}
      uVRegion {bIsValid=false, max=<0 0>, min=<0 0>}
     }
    }
   }
   slot {class=ScrollBoxSlot}
  }
  label {
   widgetClassPath=&TextBlock
   bIsVariable=true
   namedSlotHost=ExpandableArea_Raytracing_DirectionalShadows
   widget=label
   properties {
    text=Raytracing Directional Shadows
    font {
     bForceMonospaced=false
     bMaterialIsStencil=false
     fontMaterial=null
     fontObject=&Roboto
     letterSpacing=0
     monospacedWidth=1
     size=10
     skewAmount=0
     typefaceFontName=Bold
     outlineSettings {
      bApplyOutlineToDropShadows=false
      bMiteredCorners=false
      bSeparateFillAlpha=false
      outlineMaterial=null
      outlineSize=0
      outlineColor=<rgba 0 0 0 1>
     }
    }
   }
  }
  VerticalBox_508 {widgetClassPath=&VerticalBox, namedSlotHost=ExpandableArea_Raytracing_DirectionalShadows, widget=VerticalBox_508}
  EUW_checkBox_rayTracingShadows {
   widgetClassPath=&/EUW_cvar_checkBox
   bIsVariable=true
   parent=VerticalBox_508
   widget=EUW_checkBox_rayTracingShadows
   properties {cvar=r.Raytracing.Shadows, labelText=Raytracing Shadows, labelToolTip=Raytracing Shadow toggle}
   slot {class=VerticalBoxSlot}
  }
  EUW_checkBox_shadowDenoiser {
   widgetClassPath=&EUW_cvar_checkBox
   bIsVariable=true
   parent=VerticalBox_508
   widget=EUW_checkBox_shadowDenoiser
   properties {cvar=r.Shadow.Denoiser, labelText=Shadow Denoiser, labelToolTip=Shadow Denoiser}
   slot {class=VerticalBoxSlot}
  }
  Spacer_5 {widgetClassPath=&Spacer, parent=ScrollBox, widget=Spacer_5, properties {size=<1 20>}, slot {class=ScrollBoxSlot}}
  ExpandableArea_RaytracingSkylight {
   widgetClassPath=&ExpandableArea
   bIsVariable=true
   parent=ScrollBox
   widget=ExpandableArea_RaytracingSkylight
   properties {
    bIsExpanded=true
    areaPadding=<ltrb 10 10 10 10>
    borderBrush {
     bIsDynamicallyLoaded=false
     drawAs=RoundedBox
     imageType=NoImage
     mirroring=NoMirror
     resourceName=None
     resourceObject=null
     tiling=NoTile
     imageSize=<0 0>
     margin=<ltrb 0 0 0 0>
     outlineSettings {
      bUseBrushTransparency=false
      roundingType=FixedRadius
      width=0
      color {colorUseRule=UseColor_Specified, specifiedColor=<rgba 0 0 0 0>}
      cornerRadii=<4 4 4 4>
     }
     tintColor {colorUseRule=UseColor_Specified, specifiedColor=<rgba 0.0176 0.0176 0.0176 1>}
     uVRegion {bIsValid=false, max=<0 0>, min=<0 0>}
    }
    style {
     rolloutAnimationSeconds=0.1
     collapsedImage {
      bIsDynamicallyLoaded=false
      drawAs=Image
      imageType=Vector
      mirroring=NoMirror
      resourceName=../../../Engine/Content/Slate/Starship/Common/chevron-right.svg
      resourceObject=null
      tiling=NoTile
      imageSize=<16 16>
      margin=<ltrb 0 0 0 0>
      outlineSettings {
       bUseBrushTransparency=false
       roundingType=HalfHeightRadius
       width=0
       color {colorUseRule=UseColor_Specified, specifiedColor=<rgba 0 0 0 0>}
       cornerRadii=<0 0 0 0>
      }
      tintColor {colorUseRule=UseColor_Specified, specifiedColor=<rgba 0.5271 0.5271 0.5271 1>}
      uVRegion {bIsValid=false, max=<0 0>, min=<0 0>}
     }
     expandedImage {
      bIsDynamicallyLoaded=false
      drawAs=Image
      imageType=Vector
      mirroring=NoMirror
      resourceName=../../../Engine/Content/Slate/Starship/Common/chevron-down.svg
      resourceObject=null
      tiling=NoTile
      imageSize=<16 16>
      margin=<ltrb 0 0 0 0>
      outlineSettings {
       bUseBrushTransparency=false
       roundingType=HalfHeightRadius
       width=0
       color {colorUseRule=UseColor_Specified, specifiedColor=<rgba 0 0 0 0>}
       cornerRadii=<0 0 0 0>
      }
      tintColor {colorUseRule=UseColor_Specified, specifiedColor=<rgba 0.5271 0.5271 0.5271 1>}
      uVRegion {bIsValid=false, max=<0 0>, min=<0 0>}
     }
    }
   }
   slot {class=ScrollBoxSlot}
  }
  label_1 {
   widgetClassPath=&TextBlock
   bIsVariable=true
   namedSlotHost=ExpandableArea_RaytracingSkylight
   widget=label_1
   properties {
    text=Raytracing Skylight
    font {
     bForceMonospaced=false
     bMaterialIsStencil=false
     fontMaterial=null
     fontObject=&Roboto
     letterSpacing=0
     monospacedWidth=1
     size=10
     skewAmount=0
     typefaceFontName=Bold
     outlineSettings {
      bApplyOutlineToDropShadows=false
      bMiteredCorners=false
      bSeparateFillAlpha=false
      outlineMaterial=null
      outlineSize=0
      outlineColor=<rgba 0 0 0 1>
     }
    }
   }
  }
  VerticalBox_557 {widgetClassPath=&VerticalBox, namedSlotHost=ExpandableArea_RaytracingSkylight, widget=VerticalBox_557}
  EUW_checkBox_shadowDenoiser_1 {
   widgetClassPath=&EUW_cvar_checkBox
   bIsVariable=true
   parent=VerticalBox_557
   widget=EUW_checkBox_shadowDenoiser_1
   properties {cvar=r.Raytracing.Skylight, labelText=Raytracing Skylight, labelToolTip=Enables ray tracing SkyLight}
   slot {class=VerticalBoxSlot}
  }
  EUW_slider_SkylightSamplesPerPixel {
   widgetClassPath=&EUW_cvar_spinBox
   bIsVariable=true
   parent=VerticalBox_557
   widget=EUW_slider_SkylightSamplesPerPixel
   properties {
    cvar=r.RayTracing.SkyLight.SamplesPerPixel
    labelText=Raytracing Skylight SamplesPerPixel
    labelToolTip="Sets the samples-per-pixel for ray tracing SkyLight "
    max=32
    min=-1
   }
   slot {class=VerticalBoxSlot}
  }
  EUW_checkBox_skylightDenoiser {
   widgetClassPath=&EUW_cvar_checkBox
   bIsVariable=true
   parent=VerticalBox_557
   widget=EUW_checkBox_skylightDenoiser
   properties {cvar=r.Raytracing.Skylight.Denoiser, labelText=Raytracing Skylight Denoiser, labelToolTip="Denoising options "}
   slot {class=VerticalBoxSlot}
  }
  EUW_checkBox_DiffuseIndirectDenoiser {
   widgetClassPath=&EUW_cvar_checkBox
   bIsVariable=true
   parent=VerticalBox_557
   widget=EUW_checkBox_DiffuseIndirectDenoiser
   properties {cvar=r.DiffuseIndirect.Denoiser, labelText=DiffuseIndirect Denoiser, labelToolTip="Denoising options "}
   slot {class=VerticalBoxSlot}
  }
  Spacer_1 {widgetClassPath=&Spacer, parent=ScrollBox, widget=Spacer_1, properties {size=<1 20>}, slot {class=ScrollBoxSlot}}
  ExpandableArea_Raytracing_AO {
   widgetClassPath=&ExpandableArea
   bIsVariable=true
   parent=ScrollBox
   widget=ExpandableArea_Raytracing_AO
   properties {
    bIsExpanded=true
    areaPadding=<ltrb 10 10 10 10>
    borderBrush {
     bIsDynamicallyLoaded=false
     drawAs=RoundedBox
     imageType=NoImage
     mirroring=NoMirror
     resourceName=None
     resourceObject=null
     tiling=NoTile
     imageSize=<0 0>
     margin=<ltrb 0 0 0 0>
     outlineSettings {
      bUseBrushTransparency=false
      roundingType=FixedRadius
      width=0
      color {colorUseRule=UseColor_Specified, specifiedColor=<rgba 0 0 0 0>}
      cornerRadii=<4 4 4 4>
     }
     tintColor {colorUseRule=UseColor_Specified, specifiedColor=<rgba 0.0176 0.0176 0.0176 1>}
     uVRegion {bIsValid=false, max=<0 0>, min=<0 0>}
    }
    style {
     rolloutAnimationSeconds=0.1
     collapsedImage {
      bIsDynamicallyLoaded=false
      drawAs=Image
      imageType=Vector
      mirroring=NoMirror
      resourceName=../../../Engine/Content/Slate/Starship/Common/chevron-right.svg
      resourceObject=null
      tiling=NoTile
      imageSize=<16 16>
      margin=<ltrb 0 0 0 0>
      outlineSettings {
       bUseBrushTransparency=false
       roundingType=HalfHeightRadius
       width=0
       color {colorUseRule=UseColor_Specified, specifiedColor=<rgba 0 0 0 0>}
       cornerRadii=<0 0 0 0>
      }
      tintColor {colorUseRule=UseColor_Specified, specifiedColor=<rgba 0.5271 0.5271 0.5271 1>}
      uVRegion {bIsValid=false, max=<0 0>, min=<0 0>}
     }
     expandedImage {
      bIsDynamicallyLoaded=false
      drawAs=Image
      imageType=Vector
      mirroring=NoMirror
      resourceName=../../../Engine/Content/Slate/Starship/Common/chevron-down.svg
      resourceObject=null
      tiling=NoTile
      imageSize=<16 16>
      margin=<ltrb 0 0 0 0>
      outlineSettings {
       bUseBrushTransparency=false
       roundingType=HalfHeightRadius
       width=0
       color {colorUseRule=UseColor_Specified, specifiedColor=<rgba 0 0 0 0>}
       cornerRadii=<0 0 0 0>
      }
      tintColor {colorUseRule=UseColor_Specified, specifiedColor=<rgba 0.5271 0.5271 0.5271 1>}
      uVRegion {bIsValid=false, max=<0 0>, min=<0 0>}
     }
    }
   }
   slot {class=ScrollBoxSlot}
  }
  label_3 {
   widgetClassPath=&TextBlock
   bIsVariable=true
   namedSlotHost=ExpandableArea_Raytracing_AO
   widget=label_3
   properties {
    text=Raytracing Ambient Occlusion
    font {
     bForceMonospaced=false
     bMaterialIsStencil=false
     fontMaterial=null
     fontObject=&Roboto
     letterSpacing=0
     monospacedWidth=1
     size=10
     skewAmount=0
     typefaceFontName=Bold
     outlineSettings {
      bApplyOutlineToDropShadows=false
      bMiteredCorners=false
      bSeparateFillAlpha=false
      outlineMaterial=null
      outlineSize=0
      outlineColor=<rgba 0 0 0 1>
     }
    }
   }
  }
  VerticalBox_384 {widgetClassPath=&VerticalBox, namedSlotHost=ExpandableArea_Raytracing_AO, widget=VerticalBox_384}
  EUW_comboBox_AO_method {
   widgetClassPath=&EUW_cvar_comboBox
   bIsVariable=true
   parent=VerticalBox_384
   widget=EUW_comboBox_AO_method
   properties {cvar=r.AmbientOcclusion.Method, labelText=Ambient Occlusion Method, labelToolTip="0: SSAO (default)\r\n1: GTAO", option=-1, options [0,1]}
   slot {class=VerticalBoxSlot}
  }
  EUW_comboBox_AO {
   widgetClassPath=&EUW_cvar_comboBox
   bIsVariable=true
   parent=VerticalBox_384
   widget=EUW_comboBox_AO
   properties {
    cvar=r.rayTracing.AmbientOcclusion
    labelText=Raytracing Ambient Occlusion
    labelToolTip="-1: Value driven by postprocess volume (default) \r\n 0: ray tracing ambient occlusion off \r\n 1: ray tracing ambient occlusion enabled"
    option=-1
    options [-1,0,1]
   }
   slot {class=VerticalBoxSlot}
  }
  EUW_comboBox_AO_Denoiser {
   widgetClassPath=&EUW_cvar_comboBox
   bIsVariable=true
   parent=VerticalBox_384
   widget=EUW_comboBox_AO_Denoiser
   properties {
    cvar=r.AmbientOcclusion.Denoiser
    labelText=AmibentOcclusion Denoiser
    labelToolTip=" 0: Disabled;\r\n 1: Forces the default denoiser of the renderer;\r\n 2: GScreenSpaceDenoiser which may be overriden by a third party plugin (default)."
    option=-1
    options [0,1,2]
   }
   slot {class=VerticalBoxSlot}
  }
  Spacer_3 {widgetClassPath=&Spacer, parent=ScrollBox, widget=Spacer_3, properties {size=<1 20>}, slot {class=ScrollBoxSlot}}
  ExpandableArea_Raytracing_Translucency {
   widgetClassPath=&ExpandableArea
   bIsVariable=true
   parent=ScrollBox
   widget=ExpandableArea_Raytracing_Translucency
   properties {
    bIsExpanded=true
    areaPadding=<ltrb 10 10 10 10>
    borderBrush {
     bIsDynamicallyLoaded=false
     drawAs=RoundedBox
     imageType=NoImage
     mirroring=NoMirror
     resourceName=None
     resourceObject=null
     tiling=NoTile
     imageSize=<0 0>
     margin=<ltrb 0 0 0 0>
     outlineSettings {
      bUseBrushTransparency=false
      roundingType=FixedRadius
      width=0
      color {colorUseRule=UseColor_Specified, specifiedColor=<rgba 0 0 0 0>}
      cornerRadii=<4 4 4 4>
     }
     tintColor {colorUseRule=UseColor_Specified, specifiedColor=<rgba 0.0176 0.0176 0.0176 1>}
     uVRegion {bIsValid=false, max=<0 0>, min=<0 0>}
    }
    style {
     rolloutAnimationSeconds=0.1
     collapsedImage {
      bIsDynamicallyLoaded=false
      drawAs=Image
      imageType=Vector
      mirroring=NoMirror
      resourceName=../../../Engine/Content/Slate/Starship/Common/chevron-right.svg
      resourceObject=null
      tiling=NoTile
      imageSize=<16 16>
      margin=<ltrb 0 0 0 0>
      outlineSettings {
       bUseBrushTransparency=false
       roundingType=HalfHeightRadius
       width=0
       color {colorUseRule=UseColor_Specified, specifiedColor=<rgba 0 0 0 0>}
       cornerRadii=<0 0 0 0>
      }
      tintColor {colorUseRule=UseColor_Specified, specifiedColor=<rgba 0.5271 0.5271 0.5271 1>}
      uVRegion {bIsValid=false, max=<0 0>, min=<0 0>}
     }
     expandedImage {
      bIsDynamicallyLoaded=false
      drawAs=Image
      imageType=Vector
      mirroring=NoMirror
      resourceName=../../../Engine/Content/Slate/Starship/Common/chevron-down.svg
      resourceObject=null
      tiling=NoTile
      imageSize=<16 16>
      margin=<ltrb 0 0 0 0>
      outlineSettings {
       bUseBrushTransparency=false
       roundingType=HalfHeightRadius
       width=0
       color {colorUseRule=UseColor_Specified, specifiedColor=<rgba 0 0 0 0>}
       cornerRadii=<0 0 0 0>
      }
      tintColor {colorUseRule=UseColor_Specified, specifiedColor=<rgba 0.5271 0.5271 0.5271 1>}
      uVRegion {bIsValid=false, max=<0 0>, min=<0 0>}
     }
    }
   }
   slot {class=ScrollBoxSlot}
  }
  label_5 {
   widgetClassPath=&TextBlock
   bIsVariable=true
   namedSlotHost=ExpandableArea_Raytracing_Translucency
   widget=label_5
   properties {
    text=Raytracing Translucency
    font {
     bForceMonospaced=false
     bMaterialIsStencil=false
     fontMaterial=null
     fontObject=&Roboto
     letterSpacing=0
     monospacedWidth=1
     size=10
     skewAmount=0
     typefaceFontName=Bold
     outlineSettings {
      bApplyOutlineToDropShadows=false
      bMiteredCorners=false
      bSeparateFillAlpha=false
      outlineMaterial=null
      outlineSize=0
      outlineColor=<rgba 0 0 0 1>
     }
    }
   }
  }
  VerticalBox_219 {widgetClassPath=&VerticalBox, namedSlotHost=ExpandableArea_Raytracing_Translucency, widget=VerticalBox_219}
  EUW_comboBox_Translucency {
   widgetClassPath=&EUW_cvar_comboBox
   bIsVariable=true
   parent=VerticalBox_219
   widget=EUW_comboBox_Translucency
   properties {
    cvar=r.RayTracing.Translucency
    labelText=Raytracing Translucency
    labelToolTip="-1: Value driven by postprocess volume (default)\r\n0: ray tracing translucency off (use raster)\r\n1: ray tracing translucency enabled\r\n"
    option=-1
    options [-1,0,1]
   }
   slot {class=VerticalBoxSlot}
  }
  EUW_checkBox_Translucency_Refraction {
   widgetClassPath=&EUW_cvar_checkBox
   bIsVariable=true
   parent=VerticalBox_219
   widget=EUW_checkBox_Translucency_Refraction
   properties {
    cvar=r.RayTracing.Translucency.Refraction
    labelText=Raytracing Translucency Refraction
    labelToolTip="Enables refraction in ray traced Translucency (default = 1)"
   }
   slot {class=VerticalBoxSlot}
  }
  EUW_slider_Translucency_MaxRefractionRays {
   widgetClassPath=&EUW_cvar_spinBox
   bIsVariable=true
   parent=VerticalBox_219
   widget=EUW_slider_Translucency_MaxRefractionRays
   properties {
    cvar=r.RayTracing.Translucency.MaxRefractionRays
    labelText=Raytracing Translucency MaxRefractionRays
    labelToolTip="Sets the maximum number of refraction rays for ray traced translucency (default = -1 (max bounces driven by postprocessing volume)"
    max=32
    min=-1
   }
   slot {class=VerticalBoxSlot}
  }
  EUW_slider_Translucency_SamplesPerPixel {
   widgetClassPath=&EUW_cvar_spinBox
   bIsVariable=true
   parent=VerticalBox_219
   widget=EUW_slider_Translucency_SamplesPerPixel
   properties {
    cvar=r.RayTracing.Translucency.SamplesPerPixel
    labelText=Raytracing Translucency SamplesPerPixel
    labelToolTip="Sets the samples-per-pixel for Translucency (default = 1)"
    max=32
    min=1
   }
   slot {class=VerticalBoxSlot}
  }
  EUW_comboBox_Translucency_Shadows {
   widgetClassPath=&EUW_cvar_comboBox
   bIsVariable=true
   parent=VerticalBox_219
   widget=EUW_comboBox_Translucency_Shadows
   properties {
    cvar=r.RayTracing.Translucency.Shadows
    labelText=Raytracing Translucency Shadows
    labelToolTip="Enables shadows in ray tracing translucency) -1: Shadows driven by postprocessing volume (default) 0: Shadows disabled  1: Hard shadows 2: Soft area shadows"
    option=-1
    options [-1,0,1,2]
   }
   slot {class=VerticalBoxSlot}
  }
  Spacer {widgetClassPath=&Spacer, parent=ScrollBox, widget=Spacer, properties {size=<1 20>}, slot {class=ScrollBoxSlot}}
  ExpandableArea_RaytracingGI {
   widgetClassPath=&ExpandableArea
   bIsVariable=true
   parent=ScrollBox
   widget=ExpandableArea_RaytracingGI
   properties {
    bIsExpanded=true
    areaPadding=<ltrb 10 10 10 10>
    borderBrush {
     bIsDynamicallyLoaded=false
     drawAs=RoundedBox
     imageType=NoImage
     mirroring=NoMirror
     resourceName=None
     resourceObject=null
     tiling=NoTile
     imageSize=<0 0>
     margin=<ltrb 0 0 0 0>
     outlineSettings {
      bUseBrushTransparency=false
      roundingType=FixedRadius
      width=0
      color {colorUseRule=UseColor_Specified, specifiedColor=<rgba 0 0 0 0>}
      cornerRadii=<4 4 4 4>
     }
     tintColor {colorUseRule=UseColor_Specified, specifiedColor=<rgba 0.0176 0.0176 0.0176 1>}
     uVRegion {bIsValid=false, max=<0 0>, min=<0 0>}
    }
    style {
     rolloutAnimationSeconds=0.1
     collapsedImage {
      bIsDynamicallyLoaded=false
      drawAs=Image
      imageType=Vector
      mirroring=NoMirror
      resourceName=../../../Engine/Content/Slate/Starship/Common/chevron-right.svg
      resourceObject=null
      tiling=NoTile
      imageSize=<16 16>
      margin=<ltrb 0 0 0 0>
      outlineSettings {
       bUseBrushTransparency=false
       roundingType=HalfHeightRadius
       width=0
       color {colorUseRule=UseColor_Specified, specifiedColor=<rgba 0 0 0 0>}
       cornerRadii=<0 0 0 0>
      }
      tintColor {colorUseRule=UseColor_Specified, specifiedColor=<rgba 0.5271 0.5271 0.5271 1>}
      uVRegion {bIsValid=false, max=<0 0>, min=<0 0>}
     }
     expandedImage {
      bIsDynamicallyLoaded=false
      drawAs=Image
      imageType=Vector
      mirroring=NoMirror
      resourceName=../../../Engine/Content/Slate/Starship/Common/chevron-down.svg
      resourceObject=null
      tiling=NoTile
      imageSize=<16 16>
      margin=<ltrb 0 0 0 0>
      outlineSettings {
       bUseBrushTransparency=false
       roundingType=HalfHeightRadius
       width=0
       color {colorUseRule=UseColor_Specified, specifiedColor=<rgba 0 0 0 0>}
       cornerRadii=<0 0 0 0>
      }
      tintColor {colorUseRule=UseColor_Specified, specifiedColor=<rgba 0.5271 0.5271 0.5271 1>}
      uVRegion {bIsValid=false, max=<0 0>, min=<0 0>}
     }
    }
   }
   slot {class=ScrollBoxSlot}
  }
  label_2 {
   widgetClassPath=&TextBlock
   bIsVariable=true
   namedSlotHost=ExpandableArea_RaytracingGI
   widget=label_2
   properties {
    text=Raytracing Global illumination
    font {
     bForceMonospaced=false
     bMaterialIsStencil=false
     fontMaterial=null
     fontObject=&Roboto
     letterSpacing=0
     monospacedWidth=1
     size=10
     skewAmount=0
     typefaceFontName=Bold
     outlineSettings {
      bApplyOutlineToDropShadows=false
      bMiteredCorners=false
      bSeparateFillAlpha=false
      outlineMaterial=null
      outlineSize=0
      outlineColor=<rgba 0 0 0 1>
     }
    }
   }
  }
  VerticalBox_137 {widgetClassPath=&VerticalBox, namedSlotHost=ExpandableArea_RaytracingGI, widget=VerticalBox_137}
  EUW_comboBox_GI {
   widgetClassPath=&EUW_cvar_comboBox
   bIsVariable=true
   parent=VerticalBox_137
   widget=EUW_comboBox_GI
   properties {
    cvar=r.rayTracing.GlobalIllumination
    labelText=Raytracing Global illumination
    labelToolTip="-1: Value driven by postprocess volume (default) \r\n 0: ray tracing global illumination off \r\n 1: ray tracing global illumination enabled"
    option=-1
    options [-1,0,1,2]
   }
   slot {class=VerticalBoxSlot}
  }
  EUW_checkBox_GI_Denoiser {
   widgetClassPath=&EUW_cvar_checkBox
   bIsVariable=true
   parent=VerticalBox_137
   widget=EUW_checkBox_GI_Denoiser
   properties {cvar=r.RayTracing.GlobalIllumination.Denoiser, labelText=Raytracing GI Denoiser, labelToolTip="Denoising options "}
   slot {class=VerticalBoxSlot}
  }
  EUW_slider_GI_SamplesPerPixel_1 {
   widgetClassPath=&EUW_cvar_spinBox
   bIsVariable=true
   parent=VerticalBox_137
   widget=EUW_slider_GI_SamplesPerPixel_1
   properties {
    cvar=r.RayTracing.GlobalIllumination.SamplesPerPixel
    labelText=Raytracing GI SamplesPerPixel
    labelToolTip="Samples per pixel (default = -1 (driven by postprocesing volume))"
    max=32
    min=-1
   }
   slot {class=VerticalBoxSlot}
  }
  EUW_slider_GI_MaxBounces {
   widgetClassPath=&EUW_cvar_spinBox
   bIsVariable=true
   parent=VerticalBox_137
   widget=EUW_slider_GI_MaxBounces
   properties {
    cvar=r.RayTracing.GlobalIllumination.MaxBounces
    labelText=Raytracing GI MaxBounces
    labelToolTip="Max bouncesl (default = -1 (driven by postprocesing volume))"
    max=32
    min=-1
   }
   slot {class=VerticalBoxSlot}
  }
  EUW_slider_GI_ScreenPercentage {
   widgetClassPath=&EUW_cvar_spinBox
   bIsVariable=true
   parent=VerticalBox_137
   widget=EUW_slider_GI_ScreenPercentage
   properties {
    cvar=r.RayTracing.GlobalIllumination.ScreenPercentage
    labelText=Raytracing GI ScreenPercentage
    labelToolTip="Screen percentage for ray tracing global illumination (default = 50)"
    max=100
   }
   slot {class=VerticalBoxSlot}
  }
  EUW_checkBox_GI_EvalSkylight {
   widgetClassPath=&EUW_cvar_checkBox
   bIsVariable=true
   parent=VerticalBox_137
   widget=EUW_checkBox_GI_EvalSkylight
   properties {
    cvar=r.RayTracing.GlobalIllumination.EvalSkyLight
    labelText=Raytracing GI EvalSkylight
    labelToolTip=Evaluate SkyLight multi-bounce contribution NOTE: This parameter is experimental
   }
   slot {class=VerticalBoxSlot}
  }
  Spacer_4 {widgetClassPath=&Spacer, parent=ScrollBox, widget=Spacer_4, properties {size=<1 20>}, slot {class=ScrollBoxSlot}}
  ExpandableArea_SSGI {
   widgetClassPath=&ExpandableArea
   bIsVariable=true
   parent=ScrollBox
   widget=ExpandableArea_SSGI
   properties {
    bIsExpanded=true
    areaPadding=<ltrb 10 10 10 10>
    borderBrush {
     bIsDynamicallyLoaded=false
     drawAs=RoundedBox
     imageType=NoImage
     mirroring=NoMirror
     resourceName=None
     resourceObject=null
     tiling=NoTile
     imageSize=<0 0>
     margin=<ltrb 0 0 0 0>
     outlineSettings {
      bUseBrushTransparency=false
      roundingType=FixedRadius
      width=0
      color {colorUseRule=UseColor_Specified, specifiedColor=<rgba 0 0 0 0>}
      cornerRadii=<4 4 4 4>
     }
     tintColor {colorUseRule=UseColor_Specified, specifiedColor=<rgba 0.0176 0.0176 0.0176 1>}
     uVRegion {bIsValid=false, max=<0 0>, min=<0 0>}
    }
    headerPadding=<ltrb 4 4 4 4>
    style {
     rolloutAnimationSeconds=0.1
     collapsedImage {
      bIsDynamicallyLoaded=false
      drawAs=Image
      imageType=Vector
      mirroring=NoMirror
      resourceName=../../../Engine/Content/Slate/Starship/Common/chevron-right.svg
      resourceObject=null
      tiling=NoTile
      imageSize=<16 16>
      margin=<ltrb 0 0 0 0>
      outlineSettings {
       bUseBrushTransparency=false
       roundingType=HalfHeightRadius
       width=0
       color {colorUseRule=UseColor_Specified, specifiedColor=<rgba 0 0 0 0>}
       cornerRadii=<0 0 0 0>
      }
      tintColor {colorUseRule=UseColor_Specified, specifiedColor=<rgba 0.5271 0.5271 0.5271 1>}
      uVRegion {bIsValid=false, max=<0 0>, min=<0 0>}
     }
     expandedImage {
      bIsDynamicallyLoaded=false
      drawAs=Image
      imageType=Vector
      mirroring=NoMirror
      resourceName=../../../Engine/Content/Slate/Starship/Common/chevron-down.svg
      resourceObject=null
      tiling=NoTile
      imageSize=<16 16>
      margin=<ltrb 0 0 0 0>
      outlineSettings {
       bUseBrushTransparency=false
       roundingType=HalfHeightRadius
       width=0
       color {colorUseRule=UseColor_Specified, specifiedColor=<rgba 0 0 0 0>}
       cornerRadii=<0 0 0 0>
      }
      tintColor {colorUseRule=UseColor_Specified, specifiedColor=<rgba 0.5271 0.5271 0.5271 1>}
      uVRegion {bIsValid=false, max=<0 0>, min=<0 0>}
     }
    }
   }
   slot {class=ScrollBoxSlot}
  }
  label_6 {
   widgetClassPath=&TextBlock
   bIsVariable=true
   namedSlotHost=ExpandableArea_SSGI
   widget=label_6
   properties {
    text=ScreenSpace Global Illumination
    font {
     bForceMonospaced=false
     bMaterialIsStencil=false
     fontMaterial=null
     fontObject=&Roboto
     letterSpacing=0
     monospacedWidth=1
     size=10
     skewAmount=0
     typefaceFontName=Bold
     outlineSettings {
      bApplyOutlineToDropShadows=false
      bMiteredCorners=false
      bSeparateFillAlpha=false
      outlineMaterial=null
      outlineSize=0
      outlineColor=<rgba 0 0 0 1>
     }
    }
   }
  }
  VerticalBox_82 {widgetClassPath=&VerticalBox, namedSlotHost=ExpandableArea_SSGI, widget=VerticalBox_82}
  EUW_checkBox_SSGI {
   widgetClassPath=&EUW_cvar_checkBox
   bIsVariable=true
   parent=VerticalBox_82
   widget=EUW_checkBox_SSGI
   properties {cvar=r.SSGI.Enable, labelText=ScreenSpace Global Illumination, labelToolTip="Whether to enable SSGI (defaults to 0)."}
   slot {class=VerticalBoxSlot}
  }
  EUW_checkBox_SSGI_halfRes {
   widgetClassPath=&EUW_cvar_checkBox
   bIsVariable=true
   parent=VerticalBox_82
   widget=EUW_checkBox_SSGI_halfRes
   properties {cvar=r.SSGI.HalfRes, labelText=ScreenSpace GI HalfRes, labelToolTip="Whether to do SSGI at half resolution (defaults to 0)."}
   slot {class=VerticalBoxSlot}
  }
  EUW_slider_Translucency_SamplesPerPixel_1 {
   widgetClassPath=&EUW_cvar_spinBox
   bIsVariable=true
   parent=VerticalBox_82
   widget=EUW_slider_Translucency_SamplesPerPixel_1
   properties {
    cvar=r.SSGI.Quality
    labelText=ScreenSpace GI Quality
    labelToolTip="Quality setting to control number of ray shot with SSGI, between 1 and 4 (defaults to 4)."
    max=4
    min=1
   }
   slot {class=VerticalBoxSlot}
  }
  Spacer_140 {widgetClassPath=&Spacer, parent=ScrollBox, widget=Spacer_140, properties {size=<1 20>}, slot {class=ScrollBoxSlot}}
  ExpandableArea_DLSS {
   widgetClassPath=&ExpandableArea
   bIsVariable=true
   parent=ScrollBox
   widget=ExpandableArea_DLSS
   properties {
    bIsExpanded=true
    areaPadding=<ltrb 10 10 10 10>
    borderBrush {
     bIsDynamicallyLoaded=false
     drawAs=RoundedBox
     imageType=NoImage
     mirroring=NoMirror
     resourceName=None
     resourceObject=null
     tiling=NoTile
     imageSize=<0 0>
     margin=<ltrb 0 0 0 0>
     outlineSettings {
      bUseBrushTransparency=false
      roundingType=FixedRadius
      width=0
      color {colorUseRule=UseColor_Specified, specifiedColor=<rgba 0 0 0 0>}
      cornerRadii=<4 4 4 4>
     }
     tintColor {colorUseRule=UseColor_Specified, specifiedColor=<rgba 0.0176 0.0176 0.0176 1>}
     uVRegion {bIsValid=false, max=<0 0>, min=<0 0>}
    }
    style {
     rolloutAnimationSeconds=0.1
     collapsedImage {
      bIsDynamicallyLoaded=false
      drawAs=Image
      imageType=Vector
      mirroring=NoMirror
      resourceName=../../../Engine/Content/Slate/Starship/Common/chevron-right.svg
      resourceObject=null
      tiling=NoTile
      imageSize=<16 16>
      margin=<ltrb 0 0 0 0>
      outlineSettings {
       bUseBrushTransparency=false
       roundingType=HalfHeightRadius
       width=0
       color {colorUseRule=UseColor_Specified, specifiedColor=<rgba 0 0 0 0>}
       cornerRadii=<0 0 0 0>
      }
      tintColor {colorUseRule=UseColor_Specified, specifiedColor=<rgba 0.5271 0.5271 0.5271 1>}
      uVRegion {bIsValid=false, max=<0 0>, min=<0 0>}
     }
     expandedImage {
      bIsDynamicallyLoaded=false
      drawAs=Image
      imageType=Vector
      mirroring=NoMirror
      resourceName=../../../Engine/Content/Slate/Starship/Common/chevron-down.svg
      resourceObject=null
      tiling=NoTile
      imageSize=<16 16>
      margin=<ltrb 0 0 0 0>
      outlineSettings {
       bUseBrushTransparency=false
       roundingType=HalfHeightRadius
       width=0
       color {colorUseRule=UseColor_Specified, specifiedColor=<rgba 0 0 0 0>}
       cornerRadii=<0 0 0 0>
      }
      tintColor {colorUseRule=UseColor_Specified, specifiedColor=<rgba 0.5271 0.5271 0.5271 1>}
      uVRegion {bIsValid=false, max=<0 0>, min=<0 0>}
     }
    }
   }
   slot {class=ScrollBoxSlot}
  }
  DLSS {
   widgetClassPath=&TextBlock
   bIsVariable=true
   namedSlotHost=ExpandableArea_DLSS
   widget=DLSS
   properties {
    text=DLSS
    font {
     bForceMonospaced=false
     bMaterialIsStencil=false
     fontMaterial=null
     fontObject=&Roboto
     letterSpacing=0
     monospacedWidth=1
     size=10
     skewAmount=0
     typefaceFontName=Bold
     outlineSettings {
      bApplyOutlineToDropShadows=false
      bMiteredCorners=false
      bSeparateFillAlpha=false
      outlineMaterial=null
      outlineSize=0
      outlineColor=<rgba 0 0 0 1>
     }
    }
   }
  }
  VerticalBox_116 {widgetClassPath=&VerticalBox, namedSlotHost=ExpandableArea_DLSS, widget=VerticalBox_116}
  EUW_checkBox_DLSS_Enable {
   widgetClassPath=&EUW_cvar_checkBox
   bIsVariable=true
   parent=VerticalBox_116
   widget=EUW_checkBox_DLSS_Enable
   properties {
    cvar=r.NGX.DLSS.Enable
    labelText=DLSS Enable
    labelToolTip="Enable/Disable DLSS entirely. Recommended to set it to 0 after enabling the DLSS plugin, until you get the production DLLs from NVIDIA.\r\n\r\nPlugin is enabled by adding -ngxenable to command line"
   }
   slot {class=VerticalBoxSlot}
  }
  EUW_comboBox_DLSS_Quality {
   widgetClassPath=&EUW_cvar_comboBox
   bIsVariable=true
   parent=VerticalBox_116
   widget=EUW_comboBox_DLSS_Quality
   properties {
    cvar=r.NGX.DLSS.Quality
    labelText=DLSS Quality
    labelToolTip="DLSS Performance/Quality setting.\r\n -1: Performance (default)\r\n  0: Balanced\r\n  1: Quality"
    option=-1
    options [-1,0,1]
   }
   slot {class=VerticalBoxSlot}
  }
 ]
}
