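# Editor utility widget for the ShowFlags page of the render settings tool, derived from EUW_renderSettingsMaster, with combo boxes that override ShowFlag cvars.
# On Construct it runs Fill Widget Arrays, which rebuilds the inherited comboBoxes array and passes the scroll box to the parent, then calls refreshUI with resetDefault.
# Selecting the tab prints "flags" and refreshes again.
# The MeshEdges combo box is in the layout but absent from that explicitly built array.
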EditorUtilityWidgetBlueprint /Game/ArtTools/RenderSettings/EUW_renderSettings_flags
graphs [
 EventGraph {nodes [
   Event Construct()
   > Fill Widget Arrays!(execute, self:EUW Render Settings Master Object Reference, panel:Scroll Box Object Reference) -> (then)
   > refreshUI!(execute, self:EUW Render Settings Master Object Reference, resetDefault:bool=true) -> (then)
   Event onTabSelected() {is=title K2Node_Event}
   > PrintString(InString=flags)
   > refreshUI(resetDefault=true)
   Event PreConstruct()
   Event Tick()
  ]}
 Fill Widget Arrays {nodes [
   Fill Widget Arrays!() -> (then, panel:Scroll Box Object Reference) {is=title K2Node_FunctionEntry}
   n1 = Self-Reference()
   > Clear(TargetArray:Array of EUW Cvar Check Box Object References=Get checkBoxes(None=@n1) -> (checkBoxes:Array of EUW Cvar Check Box Object References))
   > Clear(TargetArray:Array of EUW Cvar Combo Box Object References=Get comboBoxes(None=@n1) -> (comboBoxes:Array of EUW Cvar Combo Box Object References))
   > Clear(TargetArray:Array of EUW Cvar Spin Box Object References=Get spinBoxes(None=@n1) -> (spinBoxes:Array of EUW Cvar Spin Box Object References))
   > Set comboBoxes(comboBoxes:Array of EUW Cvar Combo Box Object References=Make Array!(
     "[0]":EUW Cvar Combo Box Object Reference=Get EUW_comboBox_Diffuse() -> (EUW_comboBox_Diffuse:EUW Cvar Combo Box Object Reference)
     "[1]":EUW Cvar Combo Box Object Reference=Get EUW_comboBox_Particles() -> (EUW_comboBox_Particles:EUW Cvar Combo Box Object Reference)
     "[2]":EUW Cvar Combo Box Object Reference=Get EUW_comboBox_Specular() -> (EUW_comboBox_Specular:EUW Cvar Combo Box Object Reference)
    ) -> (Array:Array of EUW Cvar Combo Box Object References)) -> (Output_Get:Array of EUW Cvar Combo Box Object References)
   Parent: Fill Widget Arrays!(execute, panel:Scroll Box Object Reference=Get ScrollBox_296() -> (ScrollBox_296:Scroll Box Object Reference)) -> (then) {is=title K2Node_CallParentFunction}
  ]}
]
variables []
widgets {
 info {
  namedSlotCount=2
  parentClass=&/EUW_renderSettingsMaster.EUW_renderSettingsMaster_C
  rootWidgetClass=&CanvasPanel
  widgetCount=9
 }
 widgets [
  CanvasPanel_0 {widgetClassPath=&CanvasPanel, widget=CanvasPanel_0}
  ScrollBox_296 {
   widgetClassPath=&ScrollBox
   bIsVariable=true
   parent=CanvasPanel_0
   widget=ScrollBox_296
   properties {
    alwaysShowScrollbar=true
    consumeMouseWheel=Always
    widgetBarStyle {
     thickness=8
     draggedThumbImage {
      bIsDynamicallyLoaded=false
      drawAs=RoundedBox
      imageType=NoImage
      mirroring=NoMirror
      resourceName=None
      resourceObject=null
      tiling=NoTile
      imageSize=<0 0>
      margin=<ltrb 0 0 0 0>
      outlineSettings {
       bUseBrushTransparency=false
       roundingType=FixedRadius
       width=0
       color {colorUseRule=UseColor_Specified, specifiedColor=<rgba 0 0 0 0>}
       cornerRadii=<4 4 4 4>
      }
      tintColor {colorUseRule=UseColor_Specified, specifiedColor=<rgba 0.2159 0.2159 0.2159 1>}
      uVRegion {bIsValid=false, max=<0 0>, min=<0 0>}
     }
     horizontalBackgroundImage {
      bIsDynamicallyLoaded=false
      drawAs=NoDrawType
      imageType=NoImage
      mirroring=NoMirror
      resourceName=None
      resourceObject=null
      tiling=NoTile
      imageSize=<0 0>
      margin=<ltrb 0 0 0 0>
      outlineSettings {
       bUseBrushTransparency=false
       roundingType=HalfHeightRadius
       width=0
       color {colorUseRule=UseColor_Specified, specifiedColor=<rgba 0 0 0 0>}
       cornerRadii=<0 0 0 0>
      }
      tintColor {colorUseRule=UseColor_Specified, specifiedColor=<rgba 1 1 1 1>}
      uVRegion {bIsValid=false, max=<0 0>, min=<0 0>}
     }
     horizontalBottomSlotImage {
      bIsDynamicallyLoaded=false
      drawAs=NoDrawType
      imageType=NoImage
      mirroring=NoMirror
      resourceName=None
      resourceObject=null
      tiling=NoTile
      imageSize=<0 0>
      margin=<ltrb 0 0 0 0>
      outlineSettings {
       bUseBrushTransparency=false
       roundingType=HalfHeightRadius
       width=0
       color {colorUseRule=UseColor_Specified, specifiedColor=<rgba 0 0 0 0>}
       cornerRadii=<0 0 0 0>
      }
      tintColor {colorUseRule=UseColor_Specified, specifiedColor=<rgba 1 1 1 1>}
      uVRegion {bIsValid=false, max=<0 0>, min=<0 0>}
     }
     horizontalTopSlotImage {
      bIsDynamicallyLoaded=false
      drawAs=NoDrawType
      imageType=NoImage
      mirroring=NoMirror
      resourceName=None
      resourceObject=null
      tiling=NoTile
      imageSize=<0 0>
      margin=<ltrb 0 0 0 0>
      outlineSettings {
       bUseBrushTransparency=false
       roundingType=HalfHeightRadius
       width=0
       color {colorUseRule=UseColor_Specified, specifiedColor=<rgba 0 0 0 0>}
       cornerRadii=<0 0 0 0>
      }
      tintColor {colorUseRule=UseColor_Specified, specifiedColor=<rgba 1 1 1 1>}
      uVRegion {bIsValid=false, max=<0 0>, min=<0 0>}
     }
     hoveredThumbImage {
      bIsDynamicallyLoaded=false
      drawAs=RoundedBox
      imageType=NoImage
      mirroring=NoMirror
      resourceName=None
      resourceObject=null
      tiling=NoTile
      imageSize=<0 0>
      margin=<ltrb 0 0 0 0>
      outlineSettings {
       bUseBrushTransparency=false
       roundingType=FixedRadius
       width=0
       color {colorUseRule=UseColor_Specified, specifiedColor=<rgba 0 0 0 0>}
       cornerRadii=<4 4 4 4>
      }
      tintColor {colorUseRule=UseColor_Specified, specifiedColor=<rgba 0.2159 0.2159 0.2159 1>}
      uVRegion {bIsValid=false, max=<0 0>, min=<0 0>}
     }
     normalThumbImage {
      bIsDynamicallyLoaded=false
      drawAs=RoundedBox
      imageType=NoImage
      mirroring=NoMirror
      resourceName=None
      resourceObject=null
      tiling=NoTile
      imageSize=<0 0>
      margin=<ltrb 0 0 0 0>
      outlineSettings {
       bUseBrushTransparency=false
       roundingType=FixedRadius
       width=0
       color {colorUseRule=UseColor_Specified, specifiedColor=<rgba 0 0 0 0>}
       cornerRadii=<4 4 4 4>
      }
      tintColor {colorUseRule=UseColor_Specified, specifiedColor=<rgba 0.0953 0.0953 0.0953 1>}
      uVRegion {bIsValid=false, max=<0 0>, min=<0 0>}
     }
     verticalBackgroundImage {
      bIsDynamicallyLoaded=false
      drawAs=NoDrawType
      imageType=NoImage
      mirroring=NoMirror
      resourceName=None
      resourceObject=null
      tiling=NoTile
      imageSize=<0 0>
      margin=<ltrb 0 0 0 0>
      outlineSettings {
       bUseBrushTransparency=false
       roundingType=HalfHeightRadius
       width=0
       color {colorUseRule=UseColor_Specified, specifiedColor=<rgba 0 0 0 0>}
       cornerRadii=<0 0 0 0>
      }
      tintColor {colorUseRule=UseColor_Specified, specifiedColor=<rgba 1 1 1 1>}
      uVRegion {bIsValid=false, max=<0 0>, min=<0 0>}
     }
     verticalBottomSlotImage {
      bIsDynamicallyLoaded=false
      drawAs=NoDrawType
      imageType=NoImage
      mirroring=NoMirror
      resourceName=None
      resourceObject=null
      tiling=NoTile
      imageSize=<0 0>
      margin=<ltrb 0 0 0 0>
      outlineSettings {
       bUseBrushTransparency=false
       roundingType=HalfHeightRadius
       width=0
       color {colorUseRule=UseColor_Specified, specifiedColor=<rgba 0 0 0 0>}
       cornerRadii=<0 0 0 0>
      }
      tintColor {colorUseRule=UseColor_Specified, specifiedColor=<rgba 1 1 1 1>}
      uVRegion {bIsValid=false, max=<0 0>, min=<0 0>}
     }
     verticalTopSlotImage {
      bIsDynamicallyLoaded=false
      drawAs=NoDrawType
      imageType=NoImage
      mirroring=NoMirror
      resourceName=None
      resourceObject=null
      tiling=NoTile
      imageSize=<0 0>
      margin=<ltrb 0 0 0 0>
      outlineSettings {
       bUseBrushTransparency=false
       roundingType=HalfHeightRadius
       width=0
       color {colorUseRule=UseColor_Specified, specifiedColor=<rgba 0 0 0 0>}
       cornerRadii=<0 0 0 0>
      }
      tintColor {colorUseRule=UseColor_Specified, specifiedColor=<rgba 1 1 1 1>}
      uVRegion {bIsValid=false, max=<0 0>, min=<0 0>}
     }
    }
    widgetStyle {
     barThickness=8
     bottomShadowBrush {
      bIsDynamicallyLoaded=false
      drawAs=Box
      imageType=FullColor
      mirroring=NoMirror
      resourceName=../../../Engine/Content/Slate/Common/ScrollBoxShadowBottom.png
      resourceObject=null
      tiling=NoTile
      imageSize=<16 8>
      margin=<ltrb 0.5 0 0.5 1>
      outlineSettings {
       bUseBrushTransparency=false
       roundingType=HalfHeightRadius
       width=0
       color {colorUseRule=UseColor_Specified, specifiedColor=<rgba 0 0 0 0>}
       cornerRadii=<0 0 0 0>
      }
      tintColor {colorUseRule=UseColor_Specified, specifiedColor=<rgba 1 1 1 1>}
      uVRegion {bIsValid=false, max=<0 0>, min=<0 0>}
     }
     horizontalScrolledContentPadding=<ltrb 0 0 1 0>
     leftShadowBrush {
      bIsDynamicallyLoaded=false
      drawAs=Box
      imageType=FullColor
      mirroring=NoMirror
      resourceName=../../../Engine/Content/Slate/Common/ScrollBoxShadowLeft.png
      resourceObject=null
      tiling=NoTile
      imageSize=<8 16>
      margin=<ltrb 1 0.5 0 0.5>
      outlineSettings {
       bUseBrushTransparency=false
       roundingType=HalfHeightRadius
       width=0
       color {colorUseRule=UseColor_Specified, specifiedColor=<rgba 0 0 0 0>}
       cornerRadii=<0 0 0 0>
      }
      tintColor {colorUseRule=UseColor_Specified, specifiedColor=<rgba 1 1 1 1>}
      uVRegion {bIsValid=false, max=<0 0>, min=<0 0>}
     }
     rightShadowBrush {
      bIsDynamicallyLoaded=false
      drawAs=Box
      imageType=FullColor
      mirroring=NoMirror
      resourceName=../../../Engine/Content/Slate/Common/ScrollBoxShadowRight.png
      resourceObject=null
      tiling=NoTile
      imageSize=<8 16>
      margin=<ltrb 0 0.5 1 0.5>
      outlineSettings {
       bUseBrushTransparency=false
       roundingType=HalfHeightRadius
       width=0
       color {colorUseRule=UseColor_Specified, specifiedColor=<rgba 0 0 0 0>}
       cornerRadii=<0 0 0 0>
      }
      tintColor {colorUseRule=UseColor_Specified, specifiedColor=<rgba 1 1 1 1>}
      uVRegion {bIsValid=false, max=<0 0>, min=<0 0>}
     }
     topShadowBrush {
      bIsDynamicallyLoaded=false
      drawAs=Box
      imageType=FullColor
      mirroring=NoMirror
      resourceName=../../../Engine/Content/Slate/Common/ScrollBoxShadowTop.png
      resourceObject=null
      tiling=NoTile
      imageSize=<16 8>
      margin=<ltrb 0.5 1 0.5 0>
      outlineSettings {
       bUseBrushTransparency=false
       roundingType=HalfHeightRadius
       width=0
       color {colorUseRule=UseColor_Specified, specifiedColor=<rgba 0 0 0 0>}
       cornerRadii=<0 0 0 0>
      }
      tintColor {colorUseRule=UseColor_Specified, specifiedColor=<rgba 1 1 1 1>}
      uVRegion {bIsValid=false, max=<0 0>, min=<0 0>}
     }
     verticalScrolledContentPadding=<ltrb 0 0 0 1>
    }
   }
   slot {class=CanvasPanelSlot, properties {layoutData {alignment=<0 0>, anchors {maximum=<1 1>, minimum=<0 0>}, offsets=<ltrb 0 0 0 0>}}}
  }
  ExpandableArea_76 {
   widgetClassPath=/Script/UMG.ExpandableArea
   bIsVariable=true
   parent=ScrollBox_296
   widget=ExpandableArea_76
   properties {
    bIsExpanded=true
    areaPadding=<ltrb 10 10 10 10>
    borderBrush {
     bIsDynamicallyLoaded=false
     drawAs=RoundedBox
     imageType=NoImage
     mirroring=NoMirror
     resourceName=None
     resourceObject=null
     tiling=NoTile
     imageSize=<0 0>
     margin=<ltrb 0 0 0 0>
     outlineSettings {
      bUseBrushTransparency=false
      roundingType=FixedRadius
      width=0
      color {colorUseRule=UseColor_Specified, specifiedColor=<rgba 0 0 0 0>}
      cornerRadii=<4 4 4 4>
     }
     tintColor {colorUseRule=UseColor_Specified, specifiedColor=<rgba 0.0176 0.0176 0.0176 1>}
     uVRegion {bIsValid=false, max=<0 0>, min=<0 0>}
    }
    style {
     rolloutAnimationSeconds=0.1
     collapsedImage {
      bIsDynamicallyLoaded=false
      drawAs=Image
      imageType=Vector
      mirroring=NoMirror
      resourceName=../../../Engine/Content/Slate/Starship/Common/chevron-right.svg
      resourceObject=null
      tiling=NoTile
      imageSize=<16 16>
      margin=<ltrb 0 0 0 0>
      outlineSettings {
       bUseBrushTransparency=false
       roundingType=HalfHeightRadius
       width=0
       color {colorUseRule=UseColor_Specified, specifiedColor=<rgba 0 0 0 0>}
       cornerRadii=<0 0 0 0>
      }
      tintColor {colorUseRule=UseColor_Specified, specifiedColor=<rgba 0.5271 0.5271 0.5271 1>}
      uVRegion {bIsValid=false, max=<0 0>, min=<0 0>}
     }
     expandedImage {
      bIsDynamicallyLoaded=false
      drawAs=Image
      imageType=Vector
      mirroring=NoMirror
      resourceName=../../../Engine/Content/Slate/Starship/Common/chevron-down.svg
      resourceObject=null
      tiling=NoTile
      imageSize=<16 16>
      margin=<ltrb 0 0 0 0>
      outlineSettings {
       bUseBrushTransparency=false
       roundingType=HalfHeightRadius
       width=0
       color {colorUseRule=UseColor_Specified, specifiedColor=<rgba 0 0 0 0>}
       cornerRadii=<0 0 0 0>
      }
      tintColor {colorUseRule=UseColor_Specified, specifiedColor=<rgba 0.5271 0.5271 0.5271 1>}
      uVRegion {bIsValid=false, max=<0 0>, min=<0 0>}
     }
    }
   }
   slot {class=ScrollBoxSlot}
  }
  label_6 {
   widgetClassPath=&TextBlock
   bIsVariable=true
   namedSlotHost=ExpandableArea_76
   widget=label_6
   properties {
    text=ShowFlags
    font {
     bForceMonospaced=false
     bMaterialIsStencil=false
     fontMaterial=null
     fontObject=&Roboto
     letterSpacing=0
     monospacedWidth=1
     size=11
     skewAmount=0
     typefaceFontName=Bold
     outlineSettings {
      bApplyOutlineToDropShadows=false
      bMiteredCorners=false
      bSeparateFillAlpha=false
      outlineMaterial=null
      outlineSize=0
      outlineColor=<rgba 0 0 0 1>
     }
    }
   }
  }
  VerticalBox_124 {widgetClassPath=&VerticalBox, namedSlotHost=ExpandableArea_76, widget=VerticalBox_124}
  EUW_comboBox_Diffuse {
   widgetClassPath=&/EUW_cvar_comboBox
   bIsVariable=true
   parent=VerticalBox_124
   widget=EUW_comboBox_Diffuse
   properties {
    cvar=ShowFlag.Diffuse
    labelText=ShowFlag Diffuse
    labelToolTip="Allows to override a specific showflag (works in editor and game, \"show\" only works in game and UI only in editor)\r\nUseful to run a build many time with the same showflags (when put in consolevariables.ini like \"showflag.abc=0\")\r\n 0: force the showflag to be OFF\r\n 1: force the showflag to be ON\r\n 2: do not override this showflag (default)"
    options [0,1,2]
   }
   slot {class=VerticalBoxSlot}
  }
  EUW_comboBox_Specular {
   widgetClassPath=&EUW_cvar_comboBox
   bIsVariable=true
   parent=VerticalBox_124
   widget=EUW_comboBox_Specular
   properties {
    cvar=ShowFlag.Specular
    labelText=ShowFlag Specular
    labelToolTip="Allows to override a specific showflag (works in editor and game, \"show\" only works in game and UI only in editor)\r\nUseful to run a build many time with the same showflags (when put in consolevariables.ini like \"showflag.abc=0\")\r\n 0: force the showflag to be OFF\r\n 1: force the showflag to be ON\r\n 2: do not override this showflag (default)"
    options [0,1,2]
   }
   slot {class=VerticalBoxSlot}
  }
  EUW_comboBox_Particles {
   widgetClassPath=&EUW_cvar_comboBox
   bIsVariable=true
   parent=VerticalBox_124
   widget=EUW_comboBox_Particles
   properties {
    cvar=ShowFlag.Particles
    labelText=ShowFlag Particles
    labelToolTip="Allows to override a specific showflag (works in editor and game, \"show\" only works in game and UI only in editor)\r\nUseful to run a build many time with the same showflags (when put in consolevariables.ini like \"showflag.abc=0\")\r\n 0: force the showflag to be OFF\r\n 1: force the showflag to be ON\r\n 2: do not override this showflag (default)"
    options [0,1,2]
   }
   slot {class=VerticalBoxSlot}
  }
  EUW_comboBox_MeshEdges {
   widgetClassPath=&EUW_cvar_comboBox
   bIsVariable=true
   parent=VerticalBox_124
   widget=EUW_comboBox_MeshEdges
   properties {
    cvar=ShowFlag.MeshEdges
    labelText=ShowFlag MeshEdges
    labelToolTip="Allows to override a specific showflag (works in editor and game, \"show\" only works in game and UI only in editor)\r\nUseful to run a build many time with the same showflags (when put in consolevariables.ini like \"showflag.abc=0\")\r\n 0: force the showflag to be OFF\r\n 1: force the showflag to be ON\r\n 2: do not override this showflag (default)"
    options [0,1,2]
   }
   slot {class=VerticalBoxSlot}
  }
 ]
}
