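# Editor utility widget, the File tab of the render settings tool, that saves and loads cvar settings through a user-picked DataTable.
# Save fills the table from EUW_renderSettings' CSV and runs a Python save script; Load executes rows as console commands, then refreshes widgets.
# A DetailsView bound to the widget itself picks the table, which a SaveGame slot remembers.
# Selecting the tab rebuilds a Deltas list by duplicating cvar widgets with non-default values.
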EditorUtilityWidgetBlueprint /Game/ArtTools/RenderSettings/EUW_rendersettings_file
graphs [
 EventGraph {nodes [
   "On Clicked (SaveButton)"() {is=title K2Node_ComponentBoundEvent}
   > Save()
   "On Clicked (LoadButton)"() {is=title K2Node_ComponentBoundEvent}
   > Load()
   n1 = SaveGameEvent!() -> (OutputDelegate, then, PropertyName:Name) {is=title K2Node_CustomEvent}
   > Save Game()
   Event Construct()
   n2 = Get DetailsView_DataTable() -> (DetailsView_DataTable:Details View Object Reference)
   > SetObject(self=@n2, NewObject=Self-Reference())
   > n3 = Load Game!(execute, self) -> (then, DataTable:Data Table Object Reference)
   > Set dataTable(dataTable:Data Table Object Reference=@n3) -> (Output_Get:Data Table Object Reference)
   > Bind Event to On Property Changed(self:Property View Base Object Reference=@n2, Delegate=@n1) {is=title K2Node_AddDelegate}
   Event onTabSelected() {is=title K2Node_Event}
   > PrintString(InString=File Tab Selected)
   > ClearChildren(self=Get ScrollBox_Deltas() -> (ScrollBox_Deltas:Scroll Box Object Reference))
   > PrintString(InString=------------------------------------------)
   > n4 = GetAllWidgetsOfClass!(
    execute
    WorldContextObject:Object Reference
    WidgetClass:User Widget Class Reference=&/EUW_cvar_checkBox.EUW_cvar_checkBox_C
    TopLevelOnly:bool=false
   ) -> (then, FoundWidgets:Array of EUW Cvar Check Box Object References)
   > n5 = For Each Loop(Array:Array of EUW Cvar Check Box Object References=@n4) -> (Array Element:EUW Cvar Check Box Object Reference)
   n6 = GetAllWidgetsOfClass(execute=@n5.Completed, WidgetClass=&/EUW_cvar_comboBox.EUW_cvar_comboBox_C, TopLevelOnly=false) -> (FoundWidgets:Array of EUW Cvar Combo Box Object References)
   > n7 = For Each Loop(Array:Array of EUW Cvar Combo Box Object References=@n6) -> (Array Element:EUW Cvar Combo Box Object Reference)
   n8 = GetAllWidgetsOfClass(execute=@n7.Completed, WidgetClass=&/EUW_cvar_spinBox.EUW_cvar_spinBox_C, TopLevelOnly=false) -> (FoundWidgets:Array of EUW Cvar Spin Box Object References)
   > n9 = For Each Loop(Array:Array of EUW Cvar Spin Box Object References=@n8) -> (Array Element:EUW Cvar Spin Box Object Reference)
   > Branch(Condition=IsValid(Object=GetParent!(self:Widget Object Reference=@n9) -> (ReturnValue:Panel Widget Object Reference)))
   > n10 = getValue!(execute, self:EUW Cvar Spin Box Object Reference=@n9) -> (then, value:double)
   > Branch(Condition="Not Equal (Float)"!(A:double=@n10, B:double=Get defaultValue!(self:EUW Cvar Spin Box Object Reference=@n9) -> (defaultValue:double)) -> (ReturnValue:bool) {is=title K2Node_CallFunction})
   > duplicate!(execute, self:EUW Cvar Spin Box Object Reference=@n9, panel:Panel Widget Object Reference=Get ScrollBox_Deltas()) -> (then)
   Branch(execute=@n7.LoopBody, Condition=IsValid(Object=GetParent(self=@n7)))
   > n11 = getValue!(execute, self:EUW Cvar Combo Box Object Reference=@n7) -> (then, option:String)
   > Branch(Condition="Not Equal Exactly (String)"(A=@n11, B=Get defaultValue(self:EUW Cvar Combo Box Object Reference=@n7) -> (defaultValue:String)))
   > duplicate(self:EUW Cvar Combo Box Object Reference=@n7, panel=Get ScrollBox_Deltas())
   n12 = GetParent(self=@n5)
   n13 = Get ScrollBox_Deltas()
   Branch(
    execute=@n5.LoopBody
    Condition=AND Boolean(A=IsValid(Object=@n12), B="Not Equal (Object)"!(A:Object Reference=@n12, B:Object Reference=@n13) -> (ReturnValue:bool))
   )
   > n14 = getValue!(execute, self:EUW Cvar Check Box Object Reference=@n5) -> (then, value:bool)
   > Branch(Condition="Not Equal (Boolean)"!(A:bool=@n14, B:bool=Get defaultValue(self:EUW Cvar Check Box Object Reference=@n5) -> (defaultValue:bool)) -> (ReturnValue:bool) {is=title K2Node_CallFunction})
   > duplicate(self:EUW Cvar Check Box Object Reference=@n5, panel=Get ScrollBox_Deltas())
   deltas!(
    execute
    self:EUW Render Settings Object Reference=Get EUW_renderSettings() -> (EUW_renderSettings:EUW Render Settings Object Reference)
    scrollBox:Scroll Box Object Reference=Get ScrollBox_Deltas()
   ) -> (then)
   Event PreConstruct()
   Event Tick()
  ]}
 Load {nodes [
   Load() {is=title K2Node_FunctionEntry}
   > Branch(Condition=IsValid(Object=Get dataTable() -> (dataTable:Data Table Object Reference)))
   n1 = Get dataTable()
   > n2 = GetDataTableRowNames(Table=@n1)
   > n3 = For Each Loop(Array:Array of Names=@n2) -> (Array Element:Name)
   Branch(execute=@n3.Completed, Condition=IsValid(Object=Get EUW_renderSettings()))
   > refreshWidgets(self:EUW Render Settings Object Reference=Get EUW_renderSettings())
   n4 = Get Data Table Row!(execute:Exec=@n3.LoopBody, DataTable:Data Table Object Reference=@n1, RowName:Name=@n3) -> (then:Exec, RowNotFound:Exec, ReturnValue:S Render Settings Structure)
   n5 = Append!(
    A="To String (Name)"(InName=@n3)
    B=" "
    C:String=Break S Render Settings!(S_renderSettings:"S Render Settings Structure (by ref)"=@n4) -> (value_4_#1:String) {is=title K2Node_BreakStruct}
   ) -> (ReturnValue)
   > PrintString(InString=@n5)
   > ExecuteConsoleCommand(Command=@n5)
  ]}
 Load Game {nodes [
   Load Game() {is=title K2Node_FunctionEntry}
   > n1 = DoesSaveGameExist!(execute, SlotName:String=RenderSettings, UserIndex:int=0) -> (then, ReturnValue:bool)
   > n2 = Branch(Condition=@n1)
   > n3 = LoadGameFromSlot!(execute, SlotName:String=RenderSettings, UserIndex:int=0) -> (then, ReturnValue:Save Game Object Reference)
   > n4 = Cast To BP_renderSettings_saveGame!(execute, Object=@n3) -> (then, CastFailed, AsBP Render Settings Save Game:BP Render Settings Save Game Object Reference) {is=title K2Node_DynamicCast}
   > Return Node!(execute, DataTable:Data Table Object Reference=Get dataTable!(self:BP Render Settings Save Game Object Reference=@n4) -> (dataTable))
   Return Node(execute=@n4.CastFailed)
   Return Node(execute=@n2.else)
  ]}
 Save {nodes [
   Save() {is=title K2Node_FunctionEntry}
   n1 = Get EUW_renderSettings()
   n2 = Get dataTable!() -> (dataTable)
   > n3 = Branch(Condition=AND Boolean(A=IsValid(Object=@n2), B=IsValid(Object=@n1)))
   > n4 = getCSV!(execute, self:EUW Render Settings Object Reference=Get EUW_renderSettings()) -> (then, save:String)
   > PrintString(InString=@n4)
   > Fill Data Table from CSV String!(execute, DataTable:Data Table Object Reference=Get dataTable(), CSVString:String=@n4, ImportRowStruct:Script Struct Object Reference) -> (then, ReturnValue:bool)
   n5 = GameUserDeveloperDir!() -> (ReturnValue:String)
   > ExecuteConsoleCommand(Command=Append!(
     A="py TechArt/RenderSettings/save.py "
     B=ConvertRelativePathToFull!(InPath:String=@n5, InBasePath:String) -> (ReturnValue:String)
     C=" \""
     D:String=@n4
     E:String="\""
    ) -> (ReturnValue))
   PrintString(execute=@n3.else, InString=EUW_renderSettings variable or DataTable invalid)
  ]}
 Save Game {nodes [
   Save Game() {is=title K2Node_FunctionEntry}
   > n1 = CreateSaveGameObject!(execute, SaveGameClass:Save Game Class Reference=&/BP_renderSettings_saveGame.BP_renderSettings_saveGame_C) -> (then, ReturnValue:BP Render Settings Save Game Object Reference)
   > Set dataTable!(execute, dataTable=Get dataTable(), self:BP Render Settings Save Game Object Reference=@n1) -> (then, Output_Get)
   > n2 = SaveGameToSlot!(execute, SaveGameObject:Save Game Object Reference=@n1, SlotName:String=RenderSettings, UserIndex:int=0) -> (then, ReturnValue:bool)
   > n3 = Branch(Condition=@n2)
   > PrintString(InString=RenderSettings Options saved)
   PrintString(execute=@n3.else, InString=RenderSettings Options NOT saved)
  ]}
 getDataTables {nodes [
   getDataTables() {is=title K2Node_FunctionEntry}
   > n1 = GetAssetsByPath!(
    execute:Exec
    self:Object Reference="To Object (Interface)"!(Interface:"Interface Interface (by ref)"=GetAssetRegistry()) -> (ReturnValue:Object Reference)
    PackagePath:Name=/Game/ArtTools/RenderSettings
    bRecursive:bool=false
    bIncludeOnlyOnDiskAssets:bool=false
   ) -> (then:Exec, OutAssetData:Array of Asset Data Structures, ReturnValue:bool) {is=title K2Node_Message}
   > n2 = For Each Loop(Array:Array of Asset Data Structures=@n1) -> (Array Element:Asset Data Structure)
   Return Node!(execute=@n2.Completed, tables:Array of Names=Get tables() -> (tables:Array of Names))
   n3 = Break Asset Data!(AssetData=@n2) -> (PackageName, PackagePath, AssetName, AssetClassPath, AssetClass:Name)
   > Branch(Condition="Equal (Name)"!(A:Name=@n3.AssetClass, B:Name=DataTable) -> (ReturnValue:bool) {is=title K2Node_CallFunction})
   n4 = ToSoftObjectPath!(InAssetData:"Asset Data Structure (by ref)"=@n2) -> (ReturnValue_PathString:String)
   > PrintString(InString=Append(A="To String (Name)"(InName=@n3.AssetClass), B=" ", C="To String (Name)"(InName=@n3.AssetName), E=@n4))
   > Add(TargetArray:Array of Names=Get tables(), NewItem:"Name (by ref)"=String To Name(InString=@n4)) {is=title K2Node_CallArrayFunction}
  ]}
]
variables [
 dataTable {type=DataTable, default=null}
 EUW_renderSettings {type=EUW_renderSettings_C, default=null}
 tables {type="array<string>", default []}
]
widgets {
 info {namedSlotCount=2, rootWidgetClass=&CanvasPanel, widgetCount=16}
 widgets [
  CanvasPanel_0 {widgetClassPath=&CanvasPanel, widget=CanvasPanel_0}
  VerticalBox_311 {
   widgetClassPath=&VerticalBox
   parent=CanvasPanel_0
   widget=VerticalBox_311
   slot {class=CanvasPanelSlot, properties {layoutData {alignment=<0 0>, anchors {maximum=<1 1>, minimum=<0 0>}, offsets=<ltrb 0 0 0 0>}}}
  }
  DetailsView_DataTable {
   widgetClassPath=&DetailsView
   parent=VerticalBox_311
   widget=DetailsView_DataTable
   properties {propertiesToShow [datatable]}
   slot {class=VerticalBoxSlot}
  }
  label_1 {
   widgetClassPath=&TextBlock
   bIsVariable=true
   parent=VerticalBox_311
   widget=label_1
   properties {
    text=Choose a DataTable asset to save or load setting from or to.
    font {
     bForceMonospaced=false
     bMaterialIsStencil=false
     fontMaterial=null
     fontObject=&Roboto
     letterSpacing=0
     monospacedWidth=1
     size=12
     skewAmount=0
     typefaceFontName=Regular
     outlineSettings {
      bApplyOutlineToDropShadows=false
      bMiteredCorners=false
      bSeparateFillAlpha=false
      outlineMaterial=null
      outlineSize=0
      outlineColor=<rgba 0 0 0 1>
     }
    }
   }
   slot {class=VerticalBoxSlot, properties {horizontalAlignment=HAlign_Center, padding=<ltrb 10 10 10 10>}}
  }
  HorizontalBox_100 {widgetClassPath=&HorizontalBox, parent=VerticalBox_311, widget=HorizontalBox_100, slot {class=VerticalBoxSlot}}
  VerticalBox_203 {
   widgetClassPath=&VerticalBox
   parent=HorizontalBox_100
   widget=VerticalBox_203
   slot {class=HorizontalBoxSlot, properties {size {sizeRule=Fill, value=1}}}
  }
  SaveButton {
   widgetClassPath=&Button
   parent=VerticalBox_203
   widget=SaveButton
   properties {widgetStyle {
     clickedSlateSound {resourceObject=null}
     disabled {
      bIsDynamicallyLoaded=false
      drawAs=RoundedBox
      imageType=NoImage
      mirroring=NoMirror
      resourceName=None
      resourceObject=null
      tiling=NoTile
      imageSize=<0 0>
      margin=<ltrb 0 0 0 0>
      outlineSettings {
       bUseBrushTransparency=false
       roundingType=FixedRadius
       width=1
       color {colorUseRule=UseColor_ColorTable, specifiedColor=<rgba 0 0 0 0>}
       cornerRadii=<4 4 4 4>
      }
      tintColor {colorUseRule=UseColor_Specified, specifiedColor=<rgba 0.0395 0.0395 0.0395 1>}
      uVRegion {bIsValid=false, max=<0 0>, min=<0 0>}
     }
     disabledForeground {colorUseRule=UseColor_ColorTable, specifiedColor=<rgba 0 0 0 0>}
     hovered {
      bIsDynamicallyLoaded=false
      drawAs=RoundedBox
      imageType=NoImage
      mirroring=NoMirror
      resourceName=None
      resourceObject=null
      tiling=NoTile
      imageSize=<0 0>
      margin=<ltrb 0 0 0 0>
      outlineSettings {
       bUseBrushTransparency=false
       roundingType=FixedRadius
       width=1
       color {colorUseRule=UseColor_ColorTable, specifiedColor=<rgba 0 0 0 0>}
       cornerRadii=<4 4 4 4>
      }
      tintColor {colorUseRule=UseColor_Specified, specifiedColor=<rgba 0.0953 0.0953 0.0953 1>}
      uVRegion {bIsValid=false, max=<0 0>, min=<0 0>}
     }
     hoveredForeground {colorUseRule=UseColor_ColorTable, specifiedColor=<rgba 0 0 0 0>}
     hoveredSlateSound {resourceObject=null}
     normal {
      bIsDynamicallyLoaded=false
      drawAs=RoundedBox
      imageType=NoImage
      mirroring=NoMirror
      resourceName=None
      resourceObject=null
      tiling=NoTile
      imageSize=<0 0>
      margin=<ltrb 0 0 0 0>
      outlineSettings {
       bUseBrushTransparency=false
       roundingType=FixedRadius
       width=1
       color {colorUseRule=UseColor_ColorTable, specifiedColor=<rgba 0 0 0 0>}
       cornerRadii=<4 4 4 4>
      }
      tintColor {colorUseRule=UseColor_Specified, specifiedColor=<rgba 0.0395 0.0395 0.0395 1>}
      uVRegion {bIsValid=false, max=<0 0>, min=<0 0>}
     }
     normalForeground {colorUseRule=UseColor_ColorTable, specifiedColor=<rgba 0 0 0 0>}
     normalPadding=<ltrb 12 1 12 1>
     pressed {
      bIsDynamicallyLoaded=false
      drawAs=RoundedBox
      imageType=NoImage
      mirroring=NoMirror
      resourceName=None
      resourceObject=null
      tiling=NoTile
      imageSize=<0 0>
      margin=<ltrb 0 0 0 0>
      outlineSettings {
       bUseBrushTransparency=false
       roundingType=FixedRadius
       width=1
       color {colorUseRule=UseColor_ColorTable, specifiedColor=<rgba 0 0 0 0>}
       cornerRadii=<4 4 4 4>
      }
      tintColor {colorUseRule=UseColor_Specified, specifiedColor=<rgba 0.0284 0.0284 0.0284 1>}
      uVRegion {bIsValid=false, max=<0 0>, min=<0 0>}
     }
     pressedForeground {colorUseRule=UseColor_ColorTable, specifiedColor=<rgba 0 0 0 0>}
     pressedPadding=<ltrb 12 2 12 0>
     pressedSlateSound {resourceObject=null}
    }}
   slot {class=VerticalBoxSlot, properties {size {sizeRule=Fill, value=1}}}
  }
  SaveLabel {
   widgetClassPath=&TextBlock
   bIsVariable=true
   parent=SaveButton
   widget=SaveLabel
   properties {
    text=Save
    colorAndOpacity {colorUseRule=UseColor_Specified, specifiedColor=<rgba 0 0 0 1>}
    font {
     bForceMonospaced=false
     bMaterialIsStencil=false
     fontMaterial=null
     fontObject=&Roboto
     letterSpacing=0
     monospacedWidth=1
     size=16
     skewAmount=0
     typefaceFontName=Regular
     outlineSettings {
      bApplyOutlineToDropShadows=false
      bMiteredCorners=false
      bSeparateFillAlpha=false
      outlineMaterial=null
      outlineSize=0
      outlineColor=<rgba 0 0 0 1>
     }
    }
   }
   slot {class=ButtonSlot}
  }
  Spacer_221 {widgetClassPath=&Spacer, parent=HorizontalBox_100, widget=Spacer_221, properties {size=<10 1>}, slot {class=HorizontalBoxSlot}}
  VerticalBox_299 {
   widgetClassPath=&VerticalBox
   parent=HorizontalBox_100
   widget=VerticalBox_299
   slot {class=HorizontalBoxSlot, properties {size {sizeRule=Fill, value=1}}}
  }
  LoadButton {
   widgetClassPath=&Button
   parent=VerticalBox_299
   widget=LoadButton
   properties {widgetStyle {
     clickedSlateSound {resourceObject=null}
     disabled {
      bIsDynamicallyLoaded=false
      drawAs=RoundedBox
      imageType=NoImage
      mirroring=NoMirror
      resourceName=None
      resourceObject=null
      tiling=NoTile
      imageSize=<0 0>
      margin=<ltrb 0 0 0 0>
      outlineSettings {
       bUseBrushTransparency=false
       roundingType=FixedRadius
       width=1
       color {colorUseRule=UseColor_ColorTable, specifiedColor=<rgba 0 0 0 0>}
       cornerRadii=<4 4 4 4>
      }
      tintColor {colorUseRule=UseColor_Specified, specifiedColor=<rgba 0.0395 0.0395 0.0395 1>}
      uVRegion {bIsValid=false, max=<0 0>, min=<0 0>}
     }
     disabledForeground {colorUseRule=UseColor_ColorTable, specifiedColor=<rgba 0 0 0 0>}
     hovered {
      bIsDynamicallyLoaded=false
      drawAs=RoundedBox
      imageType=NoImage
      mirroring=NoMirror
      resourceName=None
      resourceObject=null
      tiling=NoTile
      imageSize=<0 0>
      margin=<ltrb 0 0 0 0>
      outlineSettings {
       bUseBrushTransparency=false
       roundingType=FixedRadius
       width=1
       color {colorUseRule=UseColor_ColorTable, specifiedColor=<rgba 0 0 0 0>}
       cornerRadii=<4 4 4 4>
      }
      tintColor {colorUseRule=UseColor_Specified, specifiedColor=<rgba 0.0953 0.0953 0.0953 1>}
      uVRegion {bIsValid=false, max=<0 0>, min=<0 0>}
     }
     hoveredForeground {colorUseRule=UseColor_ColorTable, specifiedColor=<rgba 0 0 0 0>}
     hoveredSlateSound {resourceObject=null}
     normal {
      bIsDynamicallyLoaded=false
      drawAs=RoundedBox
      imageType=NoImage
      mirroring=NoMirror
      resourceName=None
      resourceObject=null
      tiling=NoTile
      imageSize=<0 0>
      margin=<ltrb 0 0 0 0>
      outlineSettings {
       bUseBrushTransparency=false
       roundingType=FixedRadius
       width=1
       color {colorUseRule=UseColor_ColorTable, specifiedColor=<rgba 0 0 0 0>}
       cornerRadii=<4 4 4 4>
      }
      tintColor {colorUseRule=UseColor_Specified, specifiedColor=<rgba 0.0395 0.0395 0.0395 1>}
      uVRegion {bIsValid=false, max=<0 0>, min=<0 0>}
     }
     normalForeground {colorUseRule=UseColor_ColorTable, specifiedColor=<rgba 0 0 0 0>}
     normalPadding=<ltrb 12 1 12 1>
     pressed {
      bIsDynamicallyLoaded=false
      drawAs=RoundedBox
      imageType=NoImage
      mirroring=NoMirror
      resourceName=None
      resourceObject=null
      tiling=NoTile
      imageSize=<0 0>
      margin=<ltrb 0 0 0 0>
      outlineSettings {
       bUseBrushTransparency=false
       roundingType=FixedRadius
       width=1
       color {colorUseRule=UseColor_ColorTable, specifiedColor=<rgba 0 0 0 0>}
       cornerRadii=<4 4 4 4>
      }
      tintColor {colorUseRule=UseColor_Specified, specifiedColor=<rgba 0.0284 0.0284 0.0284 1>}
      uVRegion {bIsValid=false, max=<0 0>, min=<0 0>}
     }
     pressedForeground {colorUseRule=UseColor_ColorTable, specifiedColor=<rgba 0 0 0 0>}
     pressedPadding=<ltrb 12 2 12 0>
     pressedSlateSound {resourceObject=null}
    }}
   slot {class=VerticalBoxSlot, properties {size {sizeRule=Fill, value=1}}}
  }
  LoadLabel {
   widgetClassPath=&TextBlock
   bIsVariable=true
   parent=LoadButton
   widget=LoadLabel
   properties {
    text=Load
    colorAndOpacity {colorUseRule=UseColor_Specified, specifiedColor=<rgba 0 0 0 1>}
    font {
     bForceMonospaced=false
     bMaterialIsStencil=false
     fontMaterial=null
     fontObject=&Roboto
     letterSpacing=0
     monospacedWidth=1
     size=16
     skewAmount=0
     typefaceFontName=Regular
     outlineSettings {
      bApplyOutlineToDropShadows=false
      bMiteredCorners=false
      bSeparateFillAlpha=false
      outlineMaterial=null
      outlineSize=0
      outlineColor=<rgba 0 0 0 1>
     }
    }
   }
   slot {class=ButtonSlot}
  }
  Spacer_101 {widgetClassPath=&Spacer, parent=VerticalBox_311, widget=Spacer_101, properties {size=<10 10>}, slot {class=VerticalBoxSlot}}
  ExpandableArea_8 {
   widgetClassPath=/Script/UMG.ExpandableArea
   bIsVariable=true
   parent=VerticalBox_311
   widget=ExpandableArea_8
   properties {
    bIsExpanded=true
    borderBrush {
     bIsDynamicallyLoaded=false
     drawAs=RoundedBox
     imageType=NoImage
     mirroring=NoMirror
     resourceName=None
     resourceObject=null
     tiling=NoTile
     imageSize=<0 0>
     margin=<ltrb 0 0 0 0>
     outlineSettings {
      bUseBrushTransparency=false
      roundingType=FixedRadius
      width=0
      color {colorUseRule=UseColor_Specified, specifiedColor=<rgba 0 0 0 0>}
      cornerRadii=<4 4 4 4>
     }
     tintColor {colorUseRule=UseColor_Specified, specifiedColor=<rgba 0.0176 0.0176 0.0176 1>}
     uVRegion {bIsValid=false, max=<0 0>, min=<0 0>}
    }
    style {
     rolloutAnimationSeconds=0.1
     collapsedImage {
      bIsDynamicallyLoaded=false
      drawAs=Image
      imageType=Vector
      mirroring=NoMirror
      resourceName=../../../Engine/Content/Slate/Starship/Common/chevron-right.svg
      resourceObject=null
      tiling=NoTile
      imageSize=<16 16>
      margin=<ltrb 0 0 0 0>
      outlineSettings {
       bUseBrushTransparency=false
       roundingType=HalfHeightRadius
       width=0
       color {colorUseRule=UseColor_Specified, specifiedColor=<rgba 0 0 0 0>}
       cornerRadii=<0 0 0 0>
      }
      tintColor {colorUseRule=UseColor_Specified, specifiedColor=<rgba 0.5271 0.5271 0.5271 1>}
      uVRegion {bIsValid=false, max=<0 0>, min=<0 0>}
     }
     expandedImage {
      bIsDynamicallyLoaded=false
      drawAs=Image
      imageType=Vector
      mirroring=NoMirror
      resourceName=../../../Engine/Content/Slate/Starship/Common/chevron-down.svg
      resourceObject=null
      tiling=NoTile
      imageSize=<16 16>
      margin=<ltrb 0 0 0 0>
      outlineSettings {
       bUseBrushTransparency=false
       roundingType=HalfHeightRadius
       width=0
       color {colorUseRule=UseColor_Specified, specifiedColor=<rgba 0 0 0 0>}
       cornerRadii=<0 0 0 0>
      }
      tintColor {colorUseRule=UseColor_Specified, specifiedColor=<rgba 0.5271 0.5271 0.5271 1>}
      uVRegion {bIsValid=false, max=<0 0>, min=<0 0>}
     }
    }
   }
   slot {class=VerticalBoxSlot}
  }
  label {
   widgetClassPath=&TextBlock
   bIsVariable=true
   namedSlotHost=ExpandableArea_8
   widget=label
   properties {
    text=Deltas
    font {
     bForceMonospaced=false
     bMaterialIsStencil=false
     fontMaterial=null
     fontObject=&Roboto
     letterSpacing=0
     monospacedWidth=1
     size=12
     skewAmount=0
     typefaceFontName=Regular
     outlineSettings {
      bApplyOutlineToDropShadows=false
      bMiteredCorners=false
      bSeparateFillAlpha=false
      outlineMaterial=null
      outlineSize=0
      outlineColor=<rgba 0 0 0 1>
     }
    }
   }
  }
  ScrollBox_Deltas {
   widgetClassPath=&ScrollBox
   bIsVariable=true
   namedSlotHost=ExpandableArea_8
   widget=ScrollBox_Deltas
   properties {
    widgetBarStyle {
     thickness=8
     draggedThumbImage {
      bIsDynamicallyLoaded=false
      drawAs=RoundedBox
      imageType=NoImage
      mirroring=NoMirror
      resourceName=None
      resourceObject=null
      tiling=NoTile
      imageSize=<0 0>
      margin=<ltrb 0 0 0 0>
      outlineSettings {
       bUseBrushTransparency=false
       roundingType=FixedRadius
       width=0
       color {colorUseRule=UseColor_Specified, specifiedColor=<rgba 0 0 0 0>}
       cornerRadii=<4 4 4 4>
      }
      tintColor {colorUseRule=UseColor_Specified, specifiedColor=<rgba 0.2159 0.2159 0.2159 1>}
      uVRegion {bIsValid=false, max=<0 0>, min=<0 0>}
     }
     horizontalBackgroundImage {
      bIsDynamicallyLoaded=false
      drawAs=NoDrawType
      imageType=NoImage
      mirroring=NoMirror
      resourceName=None
      resourceObject=null
      tiling=NoTile
      imageSize=<0 0>
      margin=<ltrb 0 0 0 0>
      outlineSettings {
       bUseBrushTransparency=false
       roundingType=HalfHeightRadius
       width=0
       color {colorUseRule=UseColor_Specified, specifiedColor=<rgba 0 0 0 0>}
       cornerRadii=<0 0 0 0>
      }
      tintColor {colorUseRule=UseColor_Specified, specifiedColor=<rgba 1 1 1 1>}
      uVRegion {bIsValid=false, max=<0 0>, min=<0 0>}
     }
     horizontalBottomSlotImage {
      bIsDynamicallyLoaded=false
      drawAs=NoDrawType
      imageType=NoImage
      mirroring=NoMirror
      resourceName=None
      resourceObject=null
      tiling=NoTile
      imageSize=<0 0>
      margin=<ltrb 0 0 0 0>
      outlineSettings {
       bUseBrushTransparency=false
       roundingType=HalfHeightRadius
       width=0
       color {colorUseRule=UseColor_Specified, specifiedColor=<rgba 0 0 0 0>}
       cornerRadii=<0 0 0 0>
      }
      tintColor {colorUseRule=UseColor_Specified, specifiedColor=<rgba 1 1 1 1>}
      uVRegion {bIsValid=false, max=<0 0>, min=<0 0>}
     }
     horizontalTopSlotImage {
      bIsDynamicallyLoaded=false
      drawAs=NoDrawType
      imageType=NoImage
      mirroring=NoMirror
      resourceName=None
      resourceObject=null
      tiling=NoTile
      imageSize=<0 0>
      margin=<ltrb 0 0 0 0>
      outlineSettings {
       bUseBrushTransparency=false
       roundingType=HalfHeightRadius
       width=0
       color {colorUseRule=UseColor_Specified, specifiedColor=<rgba 0 0 0 0>}
       cornerRadii=<0 0 0 0>
      }
      tintColor {colorUseRule=UseColor_Specified, specifiedColor=<rgba 1 1 1 1>}
      uVRegion {bIsValid=false, max=<0 0>, min=<0 0>}
     }
     hoveredThumbImage {
      bIsDynamicallyLoaded=false
      drawAs=RoundedBox
      imageType=NoImage
      mirroring=NoMirror
      resourceName=None
      resourceObject=null
      tiling=NoTile
      imageSize=<0 0>
      margin=<ltrb 0 0 0 0>
      outlineSettings {
       bUseBrushTransparency=false
       roundingType=FixedRadius
       width=0
       color {colorUseRule=UseColor_Specified, specifiedColor=<rgba 0 0 0 0>}
       cornerRadii=<4 4 4 4>
      }
      tintColor {colorUseRule=UseColor_Specified, specifiedColor=<rgba 0.2159 0.2159 0.2159 1>}
      uVRegion {bIsValid=false, max=<0 0>, min=<0 0>}
     }
     normalThumbImage {
      bIsDynamicallyLoaded=false
      drawAs=RoundedBox
      imageType=NoImage
      mirroring=NoMirror
      resourceName=None
      resourceObject=null
      tiling=NoTile
      imageSize=<0 0>
      margin=<ltrb 0 0 0 0>
      outlineSettings {
       bUseBrushTransparency=false
       roundingType=FixedRadius
       width=0
       color {colorUseRule=UseColor_Specified, specifiedColor=<rgba 0 0 0 0>}
       cornerRadii=<4 4 4 4>
      }
      tintColor {colorUseRule=UseColor_Specified, specifiedColor=<rgba 0.0953 0.0953 0.0953 1>}
      uVRegion {bIsValid=false, max=<0 0>, min=<0 0>}
     }
     verticalBackgroundImage {
      bIsDynamicallyLoaded=false
      drawAs=NoDrawType
      imageType=NoImage
      mirroring=NoMirror
      resourceName=None
      resourceObject=null
      tiling=NoTile
      imageSize=<0 0>
      margin=<ltrb 0 0 0 0>
      outlineSettings {
       bUseBrushTransparency=false
       roundingType=HalfHeightRadius
       width=0
       color {colorUseRule=UseColor_Specified, specifiedColor=<rgba 0 0 0 0>}
       cornerRadii=<0 0 0 0>
      }
      tintColor {colorUseRule=UseColor_Specified, specifiedColor=<rgba 1 1 1 1>}
      uVRegion {bIsValid=false, max=<0 0>, min=<0 0>}
     }
     verticalBottomSlotImage {
      bIsDynamicallyLoaded=false
      drawAs=NoDrawType
      imageType=NoImage
      mirroring=NoMirror
      resourceName=None
      resourceObject=null
      tiling=NoTile
      imageSize=<0 0>
      margin=<ltrb 0 0 0 0>
      outlineSettings {
       bUseBrushTransparency=false
       roundingType=HalfHeightRadius
       width=0
       color {colorUseRule=UseColor_Specified, specifiedColor=<rgba 0 0 0 0>}
       cornerRadii=<0 0 0 0>
      }
      tintColor {colorUseRule=UseColor_Specified, specifiedColor=<rgba 1 1 1 1>}
      uVRegion {bIsValid=false, max=<0 0>, min=<0 0>}
     }
     verticalTopSlotImage {
      bIsDynamicallyLoaded=false
      drawAs=NoDrawType
      imageType=NoImage
      mirroring=NoMirror
      resourceName=None
      resourceObject=null
      tiling=NoTile
      imageSize=<0 0>
      margin=<ltrb 0 0 0 0>
      outlineSettings {
       bUseBrushTransparency=false
       roundingType=HalfHeightRadius
       width=0
       color {colorUseRule=UseColor_Specified, specifiedColor=<rgba 0 0 0 0>}
       cornerRadii=<0 0 0 0>
      }
      tintColor {colorUseRule=UseColor_Specified, specifiedColor=<rgba 1 1 1 1>}
      uVRegion {bIsValid=false, max=<0 0>, min=<0 0>}
     }
    }
    widgetStyle {
     barThickness=8
     bottomShadowBrush {
      bIsDynamicallyLoaded=false
      drawAs=Box
      imageType=FullColor
      mirroring=NoMirror
      resourceName=../../../Engine/Content/Slate/Common/ScrollBoxShadowBottom.png
      resourceObject=null
      tiling=NoTile
      imageSize=<16 8>
      margin=<ltrb 0.5 0 0.5 1>
      outlineSettings {
       bUseBrushTransparency=false
       roundingType=HalfHeightRadius
       width=0
       color {colorUseRule=UseColor_Specified, specifiedColor=<rgba 0 0 0 0>}
       cornerRadii=<0 0 0 0>
      }
      tintColor {colorUseRule=UseColor_Specified, specifiedColor=<rgba 1 1 1 1>}
      uVRegion {bIsValid=false, max=<0 0>, min=<0 0>}
     }
     horizontalScrolledContentPadding=<ltrb 0 0 1 0>
     leftShadowBrush {
      bIsDynamicallyLoaded=false
      drawAs=Box
      imageType=FullColor
      mirroring=NoMirror
      resourceName=../../../Engine/Content/Slate/Common/ScrollBoxShadowLeft.png
      resourceObject=null
      tiling=NoTile
      imageSize=<8 16>
      margin=<ltrb 1 0.5 0 0.5>
      outlineSettings {
       bUseBrushTransparency=false
       roundingType=HalfHeightRadius
       width=0
       color {colorUseRule=UseColor_Specified, specifiedColor=<rgba 0 0 0 0>}
       cornerRadii=<0 0 0 0>
      }
      tintColor {colorUseRule=UseColor_Specified, specifiedColor=<rgba 1 1 1 1>}
      uVRegion {bIsValid=false, max=<0 0>, min=<0 0>}
     }
     rightShadowBrush {
      bIsDynamicallyLoaded=false
      drawAs=Box
      imageType=FullColor
      mirroring=NoMirror
      resourceName=../../../Engine/Content/Slate/Common/ScrollBoxShadowRight.png
      resourceObject=null
      tiling=NoTile
      imageSize=<8 16>
      margin=<ltrb 0 0.5 1 0.5>
      outlineSettings {
       bUseBrushTransparency=false
       roundingType=HalfHeightRadius
       width=0
       color {colorUseRule=UseColor_Specified, specifiedColor=<rgba 0 0 0 0>}
       cornerRadii=<0 0 0 0>
      }
      tintColor {colorUseRule=UseColor_Specified, specifiedColor=<rgba 1 1 1 1>}
      uVRegion {bIsValid=false, max=<0 0>, min=<0 0>}
     }
     topShadowBrush {
      bIsDynamicallyLoaded=false
      drawAs=Box
      imageType=FullColor
      mirroring=NoMirror
      resourceName=../../../Engine/Content/Slate/Common/ScrollBoxShadowTop.png
      resourceObject=null
      tiling=NoTile
      imageSize=<16 8>
      margin=<ltrb 0.5 1 0.5 0>
      outlineSettings {
       bUseBrushTransparency=false
       roundingType=HalfHeightRadius
       width=0
       color {colorUseRule=UseColor_Specified, specifiedColor=<rgba 0 0 0 0>}
       cornerRadii=<0 0 0 0>
      }
      tintColor {colorUseRule=UseColor_Specified, specifiedColor=<rgba 1 1 1 1>}
      uVRegion {bIsValid=false, max=<0 0>, min=<0 0>}
     }
     verticalScrolledContentPadding=<ltrb 0 0 0 1>
    }
   }
  }
 ]
}
