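# Editor utility widget that forms the shadow tab of the render settings tool, derived from EUW_renderSettingsMaster, for tuning shadow console variables.
# Collapsible sections (global, CSM, inset, contact, distance field and more) hold reusable check box, combo box and spin box widgets, each bound to a cvar.
# Fill Widget Arrays hand-lists controls into typed arrays.
# Construct calls it, then refreshUI with resetDefault; tab selection refreshes without reset.
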EditorUtilityWidgetBlueprint /Game/ArtTools/RenderSettings/EUW_shadowSettings
graphs [
 EventGraph {nodes [
   Event Construct()
   > Fill Widget Arrays!(execute, self:EUW Render Settings Master Object Reference, panel:Scroll Box Object Reference) -> (then)
   > refreshUI!(execute, self:EUW Render Settings Master Object Reference, resetDefault:bool=true) -> (then)
   Event onTabSelected() {is=title K2Node_Event}
   > PrintString(InString=Shadow Tab Selected)
   > refreshUI(resetDefault=false)
   n1 = For Each Loop(Array:Array of EUW Cvar Check Box Object References=Get checkBoxes() -> (checkBoxes:Array of EUW Cvar Check Box Object References)) -> (Array Element:EUW Cvar Check Box Object Reference)
   > PrintString(InString=GetDisplayName(Object=@n1))
   Event PreConstruct()
   Event Tick()
  ]}
 Fill Widget Arrays {nodes [
   Fill Widget Arrays!() -> (then, panel:Scroll Box Object Reference) {is=title K2Node_FunctionEntry}
   n1 = Self-Reference()
   > Clear(TargetArray:Array of EUW Cvar Spin Box Object References=Get spinBoxes(None=@n1) -> (spinBoxes:Array of EUW Cvar Spin Box Object References))
   > Clear(TargetArray:Array of EUW Cvar Check Box Object References=Get checkBoxes(None=@n1))
   > Clear(TargetArray:Array of EUW Cvar Combo Box Object References=Get comboBoxes(None=@n1) -> (comboBoxes:Array of EUW Cvar Combo Box Object References))
   > Set checkBoxes(checkBoxes:Array of EUW Cvar Check Box Object References=Make Array!(
     "[0]":EUW Cvar Check Box Object Reference=Get EUW_checkBox_distanceFieldShadowing() -> (EUW_checkBox_distanceFieldShadowing:EUW Cvar Check Box Object Reference)
     "[1]":EUW Cvar Check Box Object Reference=Get EUW_checkBox_2() -> (EUW_checkBox_2:EUW Cvar Check Box Object Reference)
     "[2]":EUW Cvar Check Box Object Reference=Get EUW_checkBox_3() -> (EUW_checkBox_3:EUW Cvar Check Box Object Reference)
     "[3]":EUW Cvar Check Box Object Reference=Get EUW_checkBox_4() -> (EUW_checkBox_4:EUW Cvar Check Box Object Reference)
     "[4]":EUW Cvar Check Box Object Reference=Get EUW_checkBox_FarShadowCascades() -> (EUW_checkBox_FarShadowCascades:EUW Cvar Check Box Object Reference)
     "[5]":EUW Cvar Check Box Object Reference=Get EUW_checkBox_shadows_1() -> (EUW_checkBox_shadows_1:EUW Cvar Check Box Object Reference)
     "[6]":EUW Cvar Check Box Object Reference=Get EUW_checkBox_volumetricFog() -> (EUW_checkBox_volumetricFog:EUW Cvar Check Box Object Reference)
     "[7]":EUW Cvar Check Box Object Reference=Get EUW_cvar_checkBox() -> (EUW_cvar_checkBox:EUW Cvar Check Box Object Reference)
    ) -> (Array:Array of EUW Cvar Check Box Object References)) -> (Output_Get:Array of EUW Cvar Check Box Object References)
   > Set comboBoxes(comboBoxes:Array of EUW Cvar Combo Box Object References=Make Array!(
     "[0]":EUW Cvar Combo Box Object Reference=Get EUW_comboBox_DFShadowQuality() -> (EUW_comboBox_DFShadowQuality:EUW Cvar Combo Box Object Reference)
     "[1]":EUW Cvar Combo Box Object Reference=Get EUW_comboBox_1() -> (EUW_comboBox_1:EUW Cvar Combo Box Object Reference)
     "[2]":EUW Cvar Combo Box Object Reference=Get EUW_comboBox_2() -> (EUW_comboBox_2:EUW Cvar Combo Box Object Reference)
     "[3]":EUW Cvar Combo Box Object Reference=Get EUW_comboBox_142() -> (EUW_comboBox_142:EUW Cvar Combo Box Object Reference)
     "[4]":EUW Cvar Combo Box Object Reference=Get EUW_comboBox_maxResolution() -> (EUW_comboBox_maxResolution:EUW Cvar Combo Box Object Reference)
     "[5]":EUW Cvar Combo Box Object Reference=Get EUW_comboBox_shadowQuality() -> (EUW_comboBox_shadowQuality:EUW Cvar Combo Box Object Reference)
     "[6]":EUW Cvar Combo Box Object Reference=Get EUW_comboBox_volumetricFog_MissCount() -> (EUW_comboBox_volumetricFog_MissCount:EUW Cvar Combo Box Object Reference)
     "[7]":EUW Cvar Combo Box Object Reference
     "[8]":EUW Cvar Combo Box Object Reference
     "[9]":EUW Cvar Combo Box Object Reference
     "[10]":EUW Cvar Combo Box Object Reference
    ) -> (Array:Array of EUW Cvar Combo Box Object References)) -> (Output_Get:Array of EUW Cvar Combo Box Object References)
   > Set spinBoxes(spinBoxes:Array of EUW Cvar Spin Box Object References=Make Array!(
     "[0]":EUW Cvar Spin Box Object Reference=Get EUW_slider_2() -> (EUW_slider_2:EUW Cvar Spin Box Object Reference)
     "[1]":EUW Cvar Spin Box Object Reference=Get EUW_slider_4() -> (EUW_slider_4:EUW Cvar Spin Box Object Reference)
     "[2]":EUW Cvar Spin Box Object Reference=Get EUW_slider_5() -> (EUW_slider_5:EUW Cvar Spin Box Object Reference)
     "[3]":EUW Cvar Spin Box Object Reference=Get EUW_slider_6() -> (EUW_slider_6:EUW Cvar Spin Box Object Reference)
     "[4]":EUW Cvar Spin Box Object Reference=Get EUW_slider_7() -> (EUW_slider_7:EUW Cvar Spin Box Object Reference)
     "[5]":EUW Cvar Spin Box Object Reference=Get EUW_slider_8() -> (EUW_slider_8:EUW Cvar Spin Box Object Reference)
     "[6]":EUW Cvar Spin Box Object Reference=Get EUW_slider_9() -> (EUW_slider_9:EUW Cvar Spin Box Object Reference)
     "[7]":EUW Cvar Spin Box Object Reference=Get EUW_slider_11() -> (EUW_slider_11:EUW Cvar Spin Box Object Reference)
     "[8]":EUW Cvar Spin Box Object Reference=Get EUW_slider_15() -> (EUW_slider_15:EUW Cvar Spin Box Object Reference)
     "[9]":EUW Cvar Spin Box Object Reference=Get EUW_slider_16() -> (EUW_slider_16:EUW Cvar Spin Box Object Reference)
     "[10]":EUW Cvar Spin Box Object Reference=Get EUW_slider_17() -> (EUW_slider_17:EUW Cvar Spin Box Object Reference)
     "[11]":EUW Cvar Spin Box Object Reference=Get EUW_slider_18() -> (EUW_slider_18:EUW Cvar Spin Box Object Reference)
     "[12]":EUW Cvar Spin Box Object Reference=Get EUW_slider_21() -> (EUW_slider_21:EUW Cvar Spin Box Object Reference)
     "[13]":EUW Cvar Spin Box Object Reference=Get EUW_slider_CSM_MaxCascades() -> (EUW_slider_CSM_MaxCascades:EUW Cvar Spin Box Object Reference)
     "[14]":EUW Cvar Spin Box Object Reference=Get EUW_slider_CSMDepthBias() -> (EUW_slider_CSMDepthBias:EUW Cvar Spin Box Object Reference)
     "[15]":EUW Cvar Spin Box Object Reference=Get EUW_slider_CSMSlopeScaleDepthBias() -> (EUW_slider_CSMSlopeScaleDepthBias:EUW Cvar Spin Box Object Reference)
     "[16]":EUW Cvar Spin Box Object Reference=Get EUW_slider_MAxNumFarShadowCascades() -> (EUW_slider_MAxNumFarShadowCascades:EUW Cvar Spin Box Object Reference)
     "[17]":EUW Cvar Spin Box Object Reference=Get EUW_slider_MaxSlopeScaleDepthBias() -> (EUW_slider_MaxSlopeScaleDepthBias:EUW Cvar Spin Box Object Reference)
     "[18]":EUW Cvar Spin Box Object Reference=Get EUW_slider_PlayerInsetShadowsMaX() -> (EUW_slider_PlayerInsetShadowsMaX:EUW Cvar Spin Box Object Reference)
     "[19]":EUW Cvar Spin Box Object Reference=Get EUW_slider_PlayerInsetShadowsMaX_1() -> (EUW_slider_PlayerInsetShadowsMaX_1:EUW Cvar Spin Box Object Reference)
     "[20]":EUW Cvar Spin Box Object Reference=Get EUW_slider_PlayerInsetShadowsMaX_2() -> (EUW_slider_PlayerInsetShadowsMaX_2:EUW Cvar Spin Box Object Reference)
     "[21]":EUW Cvar Spin Box Object Reference=Get EUW_slider_volumetricFog_GridPixelSize() -> (EUW_slider_volumetricFog_GridPixelSize:EUW Cvar Spin Box Object Reference)
    ) -> (Array:Array of EUW Cvar Spin Box Object References)) -> (Output_Get:Array of EUW Cvar Spin Box Object References)
   n2 = For Each Loop(Array:Array of Widget Object References=GetAllChildren(self=Get ScrollBox_296() -> (ScrollBox_296:Scroll Box Object Reference))) -> (Array Element:Widget Object Reference)
   > n3 = Cast To ExpandableArea!(execute, Object=@n2) -> (then, CastFailed, AsExpandable Area:Expandable Area Object Reference) {is=title K2Node_DynamicCast}
   > n4 = GetCanvasPanelContainer!(execute:Exec, self:Object Reference=@n3, ContainerEnum:EFortActorIndicatorContainerWidget Enum=Top) -> (then:Exec, ReturnValue:Canvas Panel Object Reference) {is=title K2Node_Message}
   > PrintString(InString=GetDisplayName(Object=@n4))
   GetDisplayName(Object=@n2)
   Parent: Fill Widget Arrays!(execute, panel:Scroll Box Object Reference=Get ScrollBox_296()) -> (then) {is=title K2Node_CallParentFunction}
  ]}
]
variables []
widgets {
 info {
  namedSlotCount=18
  parentClass=&/EUW_renderSettingsMaster.EUW_renderSettingsMaster_C
  rootWidgetClass=&CanvasPanel
  widgetCount=76
 }
 widgets [
  CanvasPanel_0 {widgetClassPath=&CanvasPanel, widget=CanvasPanel_0}
  ScrollBox_296 {
   widgetClassPath=&ScrollBox
   bIsVariable=true
   parent=CanvasPanel_0
   widget=ScrollBox_296
   properties {
    alwaysShowScrollbar=true
    consumeMouseWheel=Always
    widgetBarStyle {
     thickness=8
     draggedThumbImage {
      bIsDynamicallyLoaded=false
      drawAs=RoundedBox
      imageType=NoImage
      mirroring=NoMirror
      resourceName=None
      resourceObject=null
      tiling=NoTile
      imageSize=<0 0>
      margin=<ltrb 0 0 0 0>
      outlineSettings {
       bUseBrushTransparency=false
       roundingType=FixedRadius
       width=0
       color {colorUseRule=UseColor_Specified, specifiedColor=<rgba 0 0 0 0>}
       cornerRadii=<4 4 4 4>
      }
      tintColor {colorUseRule=UseColor_Specified, specifiedColor=<rgba 0.2159 0.2159 0.2159 1>}
      uVRegion {bIsValid=false, max=<0 0>, min=<0 0>}
     }
     horizontalBackgroundImage {
      bIsDynamicallyLoaded=false
      drawAs=NoDrawType
      imageType=NoImage
      mirroring=NoMirror
      resourceName=None
      resourceObject=null
      tiling=NoTile
      imageSize=<0 0>
      margin=<ltrb 0 0 0 0>
      outlineSettings {
       bUseBrushTransparency=false
       roundingType=HalfHeightRadius
       width=0
       color {colorUseRule=UseColor_Specified, specifiedColor=<rgba 0 0 0 0>}
       cornerRadii=<0 0 0 0>
      }
      tintColor {colorUseRule=UseColor_Specified, specifiedColor=<rgba 1 1 1 1>}
      uVRegion {bIsValid=false, max=<0 0>, min=<0 0>}
     }
     horizontalBottomSlotImage {
      bIsDynamicallyLoaded=false
      drawAs=NoDrawType
      imageType=NoImage
      mirroring=NoMirror
      resourceName=None
      resourceObject=null
      tiling=NoTile
      imageSize=<0 0>
      margin=<ltrb 0 0 0 0>
      outlineSettings {
       bUseBrushTransparency=false
       roundingType=HalfHeightRadius
       width=0
       color {colorUseRule=UseColor_Specified, specifiedColor=<rgba 0 0 0 0>}
       cornerRadii=<0 0 0 0>
      }
      tintColor {colorUseRule=UseColor_Specified, specifiedColor=<rgba 1 1 1 1>}
      uVRegion {bIsValid=false, max=<0 0>, min=<0 0>}
     }
     horizontalTopSlotImage {
      bIsDynamicallyLoaded=false
      drawAs=NoDrawType
      imageType=NoImage
      mirroring=NoMirror
      resourceName=None
      resourceObject=null
      tiling=NoTile
      imageSize=<0 0>
      margin=<ltrb 0 0 0 0>
      outlineSettings {
       bUseBrushTransparency=false
       roundingType=HalfHeightRadius
       width=0
       color {colorUseRule=UseColor_Specified, specifiedColor=<rgba 0 0 0 0>}
       cornerRadii=<0 0 0 0>
      }
      tintColor {colorUseRule=UseColor_Specified, specifiedColor=<rgba 1 1 1 1>}
      uVRegion {bIsValid=false, max=<0 0>, min=<0 0>}
     }
     hoveredThumbImage {
      bIsDynamicallyLoaded=false
      drawAs=RoundedBox
      imageType=NoImage
      mirroring=NoMirror
      resourceName=None
      resourceObject=null
      tiling=NoTile
      imageSize=<0 0>
      margin=<ltrb 0 0 0 0>
      outlineSettings {
       bUseBrushTransparency=false
       roundingType=FixedRadius
       width=0
       color {colorUseRule=UseColor_Specified, specifiedColor=<rgba 0 0 0 0>}
       cornerRadii=<4 4 4 4>
      }
      tintColor {colorUseRule=UseColor_Specified, specifiedColor=<rgba 0.2159 0.2159 0.2159 1>}
      uVRegion {bIsValid=false, max=<0 0>, min=<0 0>}
     }
     normalThumbImage {
      bIsDynamicallyLoaded=false
      drawAs=RoundedBox
      imageType=NoImage
      mirroring=NoMirror
      resourceName=None
      resourceObject=null
      tiling=NoTile
      imageSize=<0 0>
      margin=<ltrb 0 0 0 0>
      outlineSettings {
       bUseBrushTransparency=false
       roundingType=FixedRadius
       width=0
       color {colorUseRule=UseColor_Specified, specifiedColor=<rgba 0 0 0 0>}
       cornerRadii=<4 4 4 4>
      }
      tintColor {colorUseRule=UseColor_Specified, specifiedColor=<rgba 0.0953 0.0953 0.0953 1>}
      uVRegion {bIsValid=false, max=<0 0>, min=<0 0>}
     }
     verticalBackgroundImage {
      bIsDynamicallyLoaded=false
      drawAs=NoDrawType
      imageType=NoImage
      mirroring=NoMirror
      resourceName=None
      resourceObject=null
      tiling=NoTile
      imageSize=<0 0>
      margin=<ltrb 0 0 0 0>
      outlineSettings {
       bUseBrushTransparency=false
       roundingType=HalfHeightRadius
       width=0
       color {colorUseRule=UseColor_Specified, specifiedColor=<rgba 0 0 0 0>}
       cornerRadii=<0 0 0 0>
      }
      tintColor {colorUseRule=UseColor_Specified, specifiedColor=<rgba 1 1 1 1>}
      uVRegion {bIsValid=false, max=<0 0>, min=<0 0>}
     }
     verticalBottomSlotImage {
      bIsDynamicallyLoaded=false
      drawAs=NoDrawType
      imageType=NoImage
      mirroring=NoMirror
      resourceName=None
      resourceObject=null
      tiling=NoTile
      imageSize=<0 0>
      margin=<ltrb 0 0 0 0>
      outlineSettings {
       bUseBrushTransparency=false
       roundingType=HalfHeightRadius
       width=0
       color {colorUseRule=UseColor_Specified, specifiedColor=<rgba 0 0 0 0>}
       cornerRadii=<0 0 0 0>
      }
      tintColor {colorUseRule=UseColor_Specified, specifiedColor=<rgba 1 1 1 1>}
      uVRegion {bIsValid=false, max=<0 0>, min=<0 0>}
     }
     verticalTopSlotImage {
      bIsDynamicallyLoaded=false
      drawAs=NoDrawType
      imageType=NoImage
      mirroring=NoMirror
      resourceName=None
      resourceObject=null
      tiling=NoTile
      imageSize=<0 0>
      margin=<ltrb 0 0 0 0>
      outlineSettings {
       bUseBrushTransparency=false
       roundingType=HalfHeightRadius
       width=0
       color {colorUseRule=UseColor_Specified, specifiedColor=<rgba 0 0 0 0>}
       cornerRadii=<0 0 0 0>
      }
      tintColor {colorUseRule=UseColor_Specified, specifiedColor=<rgba 1 1 1 1>}
      uVRegion {bIsValid=false, max=<0 0>, min=<0 0>}
     }
    }
    widgetStyle {
     barThickness=8
     bottomShadowBrush {
      bIsDynamicallyLoaded=false
      drawAs=Box
      imageType=FullColor
      mirroring=NoMirror
      resourceName=../../../Engine/Content/Slate/Common/ScrollBoxShadowBottom.png
      resourceObject=null
      tiling=NoTile
      imageSize=<16 8>
      margin=<ltrb 0.5 0 0.5 1>
      outlineSettings {
       bUseBrushTransparency=false
       roundingType=HalfHeightRadius
       width=0
       color {colorUseRule=UseColor_Specified, specifiedColor=<rgba 0 0 0 0>}
       cornerRadii=<0 0 0 0>
      }
      tintColor {colorUseRule=UseColor_Specified, specifiedColor=<rgba 1 1 1 1>}
      uVRegion {bIsValid=false, max=<0 0>, min=<0 0>}
     }
     horizontalScrolledContentPadding=<ltrb 0 0 1 0>
     leftShadowBrush {
      bIsDynamicallyLoaded=false
      drawAs=Box
      imageType=FullColor
      mirroring=NoMirror
      resourceName=../../../Engine/Content/Slate/Common/ScrollBoxShadowLeft.png
      resourceObject=null
      tiling=NoTile
      imageSize=<8 16>
      margin=<ltrb 1 0.5 0 0.5>
      outlineSettings {
       bUseBrushTransparency=false
       roundingType=HalfHeightRadius
       width=0
       color {colorUseRule=UseColor_Specified, specifiedColor=<rgba 0 0 0 0>}
       cornerRadii=<0 0 0 0>
      }
      tintColor {colorUseRule=UseColor_Specified, specifiedColor=<rgba 1 1 1 1>}
      uVRegion {bIsValid=false, max=<0 0>, min=<0 0>}
     }
     rightShadowBrush {
      bIsDynamicallyLoaded=false
      drawAs=Box
      imageType=FullColor
      mirroring=NoMirror
      resourceName=../../../Engine/Content/Slate/Common/ScrollBoxShadowRight.png
      resourceObject=null
      tiling=NoTile
      imageSize=<8 16>
      margin=<ltrb 0 0.5 1 0.5>
      outlineSettings {
       bUseBrushTransparency=false
       roundingType=HalfHeightRadius
       width=0
       color {colorUseRule=UseColor_Specified, specifiedColor=<rgba 0 0 0 0>}
       cornerRadii=<0 0 0 0>
      }
      tintColor {colorUseRule=UseColor_Specified, specifiedColor=<rgba 1 1 1 1>}
      uVRegion {bIsValid=false, max=<0 0>, min=<0 0>}
     }
     topShadowBrush {
      bIsDynamicallyLoaded=false
      drawAs=Box
      imageType=FullColor
      mirroring=NoMirror
      resourceName=../../../Engine/Content/Slate/Common/ScrollBoxShadowTop.png
      resourceObject=null
      tiling=NoTile
      imageSize=<16 8>
      margin=<ltrb 0.5 1 0.5 0>
      outlineSettings {
       bUseBrushTransparency=false
       roundingType=HalfHeightRadius
       width=0
       color {colorUseRule=UseColor_Specified, specifiedColor=<rgba 0 0 0 0>}
       cornerRadii=<0 0 0 0>
      }
      tintColor {colorUseRule=UseColor_Specified, specifiedColor=<rgba 1 1 1 1>}
      uVRegion {bIsValid=false, max=<0 0>, min=<0 0>}
     }
     verticalScrolledContentPadding=<ltrb 0 0 0 1>
    }
   }
   slot {class=CanvasPanelSlot, properties {layoutData {alignment=<0 0>, anchors {maximum=<1 1>, minimum=<0 0>}, offsets=<ltrb 0 0 0 0>}}}
  }
  ExpandableArea_76 {
   widgetClassPath=/Script/UMG.ExpandableArea
   bIsVariable=true
   parent=ScrollBox_296
   widget=ExpandableArea_76
   properties {
    bIsExpanded=true
    areaPadding=<ltrb 10 10 10 10>
    borderBrush {
     bIsDynamicallyLoaded=false
     drawAs=RoundedBox
     imageType=NoImage
     mirroring=NoMirror
     resourceName=None
     resourceObject=null
     tiling=NoTile
     imageSize=<0 0>
     margin=<ltrb 0 0 0 0>
     outlineSettings {
      bUseBrushTransparency=false
      roundingType=FixedRadius
      width=0
      color {colorUseRule=UseColor_Specified, specifiedColor=<rgba 0 0 0 0>}
      cornerRadii=<4 4 4 4>
     }
     tintColor {colorUseRule=UseColor_Specified, specifiedColor=<rgba 0.0176 0.0176 0.0176 1>}
     uVRegion {bIsValid=false, max=<0 0>, min=<0 0>}
    }
    style {
     rolloutAnimationSeconds=0.1
     collapsedImage {
      bIsDynamicallyLoaded=false
      drawAs=Image
      imageType=Vector
      mirroring=NoMirror
      resourceName=../../../Engine/Content/Slate/Starship/Common/chevron-right.svg
      resourceObject=null
      tiling=NoTile
      imageSize=<16 16>
      margin=<ltrb 0 0 0 0>
      outlineSettings {
       bUseBrushTransparency=false
       roundingType=HalfHeightRadius
       width=0
       color {colorUseRule=UseColor_Specified, specifiedColor=<rgba 0 0 0 0>}
       cornerRadii=<0 0 0 0>
      }
      tintColor {colorUseRule=UseColor_Specified, specifiedColor=<rgba 0.5271 0.5271 0.5271 1>}
      uVRegion {bIsValid=false, max=<0 0>, min=<0 0>}
     }
     expandedImage {
      bIsDynamicallyLoaded=false
      drawAs=Image
      imageType=Vector
      mirroring=NoMirror
      resourceName=../../../Engine/Content/Slate/Starship/Common/chevron-down.svg
      resourceObject=null
      tiling=NoTile
      imageSize=<16 16>
      margin=<ltrb 0 0 0 0>
      outlineSettings {
       bUseBrushTransparency=false
       roundingType=HalfHeightRadius
       width=0
       color {colorUseRule=UseColor_Specified, specifiedColor=<rgba 0 0 0 0>}
       cornerRadii=<0 0 0 0>
      }
      tintColor {colorUseRule=UseColor_Specified, specifiedColor=<rgba 0.5271 0.5271 0.5271 1>}
      uVRegion {bIsValid=false, max=<0 0>, min=<0 0>}
     }
    }
   }
   slot {class=ScrollBoxSlot}
  }
  label_6 {
   widgetClassPath=&TextBlock
   bIsVariable=true
   namedSlotHost=ExpandableArea_76
   widget=label_6
   properties {
    text=Global Shadow Settings
    font {
     bForceMonospaced=false
     bMaterialIsStencil=false
     fontMaterial=null
     fontObject=&Roboto
     letterSpacing=0
     monospacedWidth=1
     size=11
     skewAmount=0
     typefaceFontName=Bold
     outlineSettings {
      bApplyOutlineToDropShadows=false
      bMiteredCorners=false
      bSeparateFillAlpha=false
      outlineMaterial=null
      outlineSize=0
      outlineColor=<rgba 0 0 0 1>
     }
    }
   }
  }
  VerticalBox_124 {widgetClassPath=&VerticalBox, namedSlotHost=ExpandableArea_76, widget=VerticalBox_124}
  EUW_comboBox_shadowQuality {
   widgetClassPath=&/EUW_cvar_comboBox
   bIsVariable=true
   parent=VerticalBox_124
   widget=EUW_comboBox_shadowQuality
   properties {
    cvar=r.ShadowQuality
    labelText=Shadow Quality
    labelToolTip="Defines the shadow method which allows to adjust for quality or performance.\r\n 0:off, 1:low(unfiltered), 2:low .. 5:max (default)"
    options [0,1,2,3,4,5]
   }
   slot {class=VerticalBoxSlot}
  }
  EUW_comboBox_maxResolution {
   widgetClassPath=&EUW_cvar_comboBox
   bIsVariable=true
   parent=VerticalBox_124
   widget=EUW_comboBox_maxResolution
   properties {
    cvar=r.Shadow.MaxResolution
    labelText=Shadow Max Resolution
    labelToolTip="Max square dimensions (in texels) allowed for rendering shadow depths. Range 4 to hardware limit. Higher = better quality shadows but at a performance cost."
    option=1024
    options [128,256,512,1024,2048,4096,8192]
   }
   slot {class=VerticalBoxSlot}
  }
  EUW_slider_2 {
   widgetClassPath=&/EUW_cvar_spinBox
   bIsVariable=true
   parent=VerticalBox_124
   widget=EUW_slider_2
   properties {
    cvar=r.Shadow.DistanceScale
    labelText=Shadow Distance Scale
    labelToolTip="Scalability option to trade shadow distance versus performance for directional lights (clamped within a reasonable range).\r\n<1: shorter distance\r\n 1: normal (default)\r\n>1: larger distance"
    max=1
    step=0.001
   }
   slot {class=VerticalBoxSlot}
  }
  EUW_slider_MaxSlopeScaleDepthBias {
   widgetClassPath=&EUW_cvar_spinBox
   bIsVariable=true
   parent=VerticalBox_124
   widget=EUW_slider_MaxSlopeScaleDepthBias
   properties {
    cvar=r.Shadow.ShadowMaxSlopeScaleDepthBias
    labelText=Shadow Max Slope Scale Depth Bias
    labelToolTip="Max Slope depth bias used for shadows for all lights\r\nHigher values give better self-shadowing, but increase self-shadowing artifacts"
    step=0.01
   }
   slot {class=VerticalBoxSlot}
  }
  EUW_checkBox_shadows_1 {
   widgetClassPath=&/EUW_cvar_checkBox
   bIsVariable=true
   parent=VerticalBox_124
   widget=EUW_checkBox_shadows_1
   properties {
    cvar=ShowFlag.LightFunctions
    labelText=LightFunctions
    labelToolTip="Useful to run a build many time with the same showflags (when put in consolevariables.ini like \"showflag.abc=0\")\r\n 0: force the showflag to be OFF\r\n 1: force the showflag to be ON\r\n 2: do not override this showflag (default)"
   }
   slot {class=VerticalBoxSlot}
  }
  Spacer_8 {widgetClassPath=&Spacer, parent=ScrollBox_296, widget=Spacer_8, properties {size=<1 20>}, slot {class=ScrollBoxSlot}}
  ExpandableArea_CSM {
   widgetClassPath=&ExpandableArea
   bIsVariable=true
   parent=ScrollBox_296
   widget=ExpandableArea_CSM
   properties {
    bIsExpanded=true
    areaPadding=<ltrb 10 10 10 10>
    borderBrush {
     bIsDynamicallyLoaded=false
     drawAs=RoundedBox
     imageType=NoImage
     mirroring=NoMirror
     resourceName=None
     resourceObject=null
     tiling=NoTile
     imageSize=<0 0>
     margin=<ltrb 0 0 0 0>
     outlineSettings {
      bUseBrushTransparency=false
      roundingType=FixedRadius
      width=0
      color {colorUseRule=UseColor_Specified, specifiedColor=<rgba 0 0 0 0>}
      cornerRadii=<4 4 4 4>
     }
     tintColor {colorUseRule=UseColor_Specified, specifiedColor=<rgba 0.0176 0.0176 0.0176 1>}
     uVRegion {bIsValid=false, max=<0 0>, min=<0 0>}
    }
    style {
     rolloutAnimationSeconds=0.1
     collapsedImage {
      bIsDynamicallyLoaded=false
      drawAs=Image
      imageType=Vector
      mirroring=NoMirror
      resourceName=../../../Engine/Content/Slate/Starship/Common/chevron-right.svg
      resourceObject=null
      tiling=NoTile
      imageSize=<16 16>
      margin=<ltrb 0 0 0 0>
      outlineSettings {
       bUseBrushTransparency=false
       roundingType=HalfHeightRadius
       width=0
       color {colorUseRule=UseColor_Specified, specifiedColor=<rgba 0 0 0 0>}
       cornerRadii=<0 0 0 0>
      }
      tintColor {colorUseRule=UseColor_Specified, specifiedColor=<rgba 0.5271 0.5271 0.5271 1>}
      uVRegion {bIsValid=false, max=<0 0>, min=<0 0>}
     }
     expandedImage {
      bIsDynamicallyLoaded=false
      drawAs=Image
      imageType=Vector
      mirroring=NoMirror
      resourceName=../../../Engine/Content/Slate/Starship/Common/chevron-down.svg
      resourceObject=null
      tiling=NoTile
      imageSize=<16 16>
      margin=<ltrb 0 0 0 0>
      outlineSettings {
       bUseBrushTransparency=false
       roundingType=HalfHeightRadius
       width=0
       color {colorUseRule=UseColor_Specified, specifiedColor=<rgba 0 0 0 0>}
       cornerRadii=<0 0 0 0>
      }
      tintColor {colorUseRule=UseColor_Specified, specifiedColor=<rgba 0.5271 0.5271 0.5271 1>}
      uVRegion {bIsValid=false, max=<0 0>, min=<0 0>}
     }
    }
   }
   slot {class=ScrollBoxSlot}
  }
  label {
   widgetClassPath=&TextBlock
   bIsVariable=true
   namedSlotHost=ExpandableArea_CSM
   widget=label
   properties {
    text=Cascading Shadow Maps
    font {
     bForceMonospaced=false
     bMaterialIsStencil=false
     fontMaterial=null
     fontObject=&Roboto
     letterSpacing=0
     monospacedWidth=1
     size=10
     skewAmount=0
     typefaceFontName=Bold
     outlineSettings {
      bApplyOutlineToDropShadows=false
      bMiteredCorners=false
      bSeparateFillAlpha=false
      outlineMaterial=null
      outlineSize=0
      outlineColor=<rgba 0 0 0 1>
     }
    }
   }
  }
  VerticalBox_205 {widgetClassPath=&VerticalBox, namedSlotHost=ExpandableArea_CSM, widget=VerticalBox_205}
  EUW_slider_CSM_MaxCascades {
   widgetClassPath=&EUW_cvar_spinBox
   bIsVariable=true
   parent=VerticalBox_205
   widget=EUW_slider_CSM_MaxCascades
   properties {
    cvar=r.Shadow.CSM.MaxCascades
    labelText=CSM Max Cascades
    labelToolTip=The maximum number of cascades with which to render dynamic directional light shadows.
    max=20
   }
   slot {class=VerticalBoxSlot}
  }
  EUW_slider_16 {
   widgetClassPath=&EUW_cvar_spinBox
   bIsVariable=true
   parent=VerticalBox_205
   widget=EUW_slider_16
   properties {
    cvar=r.Shadow.CSM.TransitionScale
    labelText=CSM Transition Scale
    labelToolTip="Allows to scale the cascaded shadow map transition region. Clamped within 0..2.\r\n0: no transition (fastest)\r\n1: as specific in the light settings (default)\r\n2: 2x larger than what was specified in the light"
    max=2
    step=0.01
   }
   slot {class=VerticalBoxSlot}
  }
  EUW_comboBox_1 {
   widgetClassPath=&EUW_cvar_comboBox
   bIsVariable=true
   parent=VerticalBox_205
   widget=EUW_comboBox_1
   properties {
    cvar=r.Shadow.MaxCSMResolution
    labelText=MaxCSMResolution
    labelToolTip="Max square dimensions (in texels) allowed for rendering Cascaded Shadow depths. Range 4 to hardware limit. Higher = better quality shadows but at a performance cost."
    option=1024
    options [128,256,512,1024,2048,4096,8192]
   }
   slot {class=VerticalBoxSlot}
  }
  EUW_slider_CSMDepthBias {
   widgetClassPath=&EUW_cvar_spinBox
   bIsVariable=true
   parent=VerticalBox_205
   widget=EUW_slider_CSMDepthBias
   properties {cvar=r.Shadow.CSMDepthBias, labelText=CSMDepthBias, labelToolTip="", max=20, step=0.01}
   slot {class=VerticalBoxSlot}
  }
  EUW_slider_CSMSlopeScaleDepthBias {
   widgetClassPath=&EUW_cvar_spinBox
   bIsVariable=true
   parent=VerticalBox_205
   widget=EUW_slider_CSMSlopeScaleDepthBias
   properties {cvar=r.Shadow.CSMSlopeScaleDepthBias, labelText=CSMSlopeScaleDepthBias, labelToolTip="", max=20, step=0.01}
   slot {class=VerticalBoxSlot}
  }
  Spacer_6 {widgetClassPath=&Spacer, parent=ScrollBox_296, widget=Spacer_6, properties {size=<1 20>}, slot {class=ScrollBoxSlot}}
  ExpandableArea_Inset {
   widgetClassPath=&ExpandableArea
   bIsVariable=true
   parent=ScrollBox_296
   widget=ExpandableArea_Inset
   properties {
    bIsExpanded=true
    areaPadding=<ltrb 10 10 10 10>
    borderBrush {
     bIsDynamicallyLoaded=false
     drawAs=RoundedBox
     imageType=NoImage
     mirroring=NoMirror
     resourceName=None
     resourceObject=null
     tiling=NoTile
     imageSize=<0 0>
     margin=<ltrb 0 0 0 0>
     outlineSettings {
      bUseBrushTransparency=false
      roundingType=FixedRadius
      width=0
      color {colorUseRule=UseColor_Specified, specifiedColor=<rgba 0 0 0 0>}
      cornerRadii=<4 4 4 4>
     }
     tintColor {colorUseRule=UseColor_Specified, specifiedColor=<rgba 0.0176 0.0176 0.0176 1>}
     uVRegion {bIsValid=false, max=<0 0>, min=<0 0>}
    }
    style {
     rolloutAnimationSeconds=0.1
     collapsedImage {
      bIsDynamicallyLoaded=false
      drawAs=Image
      imageType=Vector
      mirroring=NoMirror
      resourceName=../../../Engine/Content/Slate/Starship/Common/chevron-right.svg
      resourceObject=null
      tiling=NoTile
      imageSize=<16 16>
      margin=<ltrb 0 0 0 0>
      outlineSettings {
       bUseBrushTransparency=false
       roundingType=HalfHeightRadius
       width=0
       color {colorUseRule=UseColor_Specified, specifiedColor=<rgba 0 0 0 0>}
       cornerRadii=<0 0 0 0>
      }
      tintColor {colorUseRule=UseColor_Specified, specifiedColor=<rgba 0.5271 0.5271 0.5271 1>}
      uVRegion {bIsValid=false, max=<0 0>, min=<0 0>}
     }
     expandedImage {
      bIsDynamicallyLoaded=false
      drawAs=Image
      imageType=Vector
      mirroring=NoMirror
      resourceName=../../../Engine/Content/Slate/Starship/Common/chevron-down.svg
      resourceObject=null
      tiling=NoTile
      imageSize=<16 16>
      margin=<ltrb 0 0 0 0>
      outlineSettings {
       bUseBrushTransparency=false
       roundingType=HalfHeightRadius
       width=0
       color {colorUseRule=UseColor_Specified, specifiedColor=<rgba 0 0 0 0>}
       cornerRadii=<0 0 0 0>
      }
      tintColor {colorUseRule=UseColor_Specified, specifiedColor=<rgba 0.5271 0.5271 0.5271 1>}
      uVRegion {bIsValid=false, max=<0 0>, min=<0 0>}
     }
    }
   }
   slot {class=ScrollBoxSlot}
  }
  label_1 {
   widgetClassPath=&TextBlock
   bIsVariable=true
   namedSlotHost=ExpandableArea_Inset
   widget=label_1
   properties {
    text=Inset Shadows
    font {
     bForceMonospaced=false
     bMaterialIsStencil=false
     fontMaterial=null
     fontObject=&Roboto
     letterSpacing=0
     monospacedWidth=1
     size=11
     skewAmount=0
     typefaceFontName=Bold
     outlineSettings {
      bApplyOutlineToDropShadows=false
      bMiteredCorners=false
      bSeparateFillAlpha=false
      outlineMaterial=null
      outlineSize=0
      outlineColor=<rgba 0 0 0 1>
     }
    }
   }
  }
  VerticalBox_60 {widgetClassPath=&VerticalBox, namedSlotHost=ExpandableArea_Inset, widget=VerticalBox_60}
  EUW_slider_PlayerInsetShadowsMaX {
   widgetClassPath=&EUW_cvar_spinBox
   bIsVariable=true
   parent=VerticalBox_60
   widget=EUW_slider_PlayerInsetShadowsMaX
   properties {
    cvar=Fort.Scalability.PlayerInsetShadowsMax
    labelText=PlayerInsetShadowsMax
    labelToolTip="Max number of players that will have inset shadows enabled at a time.\r\n 0 to disable"
    max=15
   }
   slot {class=VerticalBoxSlot}
  }
  EUW_slider_11 {
   widgetClassPath=&EUW_cvar_spinBox
   bIsVariable=true
   parent=VerticalBox_60
   widget=EUW_slider_11
   properties {
    cvar=Fort.Scalability.PlayerInsetShadowsLOD
    labelText=PlayerInsetShadowsLOD
    labelToolTip="LODValue <= PlayerInsetShadowsLOD will enable inset shadows.\r\n -1 to disable"
    max=1
    min=-1
    step=0.001
   }
   slot {class=VerticalBoxSlot}
  }
  EUW_slider_PlayerInsetShadowsMaX_2 {
   widgetClassPath=&EUW_cvar_spinBox
   bIsVariable=true
   parent=VerticalBox_60
   widget=EUW_slider_PlayerInsetShadowsMaX_2
   properties {
    cvar=r.Shadow.PerObjectDirectionalDepthBias
    labelText=Per Object Directional Depth Bias
    labelToolTip="Constant depth bias used by per-object shadows from directional lights\r\nLower values give better shadow contact, but increase self-shadowing artifacts"
    max=20
    step=0.1
   }
   slot {class=VerticalBoxSlot}
  }
  EUW_slider_PlayerInsetShadowsMaX_1 {
   widgetClassPath=&EUW_cvar_spinBox
   bIsVariable=true
   parent=VerticalBox_60
   widget=EUW_slider_PlayerInsetShadowsMaX_1
   properties {
    cvar=r.Shadow.PerObjectDirectionalSlopeDepthBias
    labelText=Per Object Directional Slope Depth Bias
    labelToolTip=Max number of players that will have inset shadows enabled at a time.
    max=15
   }
   slot {class=VerticalBoxSlot}
  }
  Spacer_5 {widgetClassPath=&Spacer, parent=ScrollBox_296, widget=Spacer_5, properties {size=<1 20>}, slot {class=ScrollBoxSlot}}
  ExpandableArea_114 {
   widgetClassPath=&ExpandableArea
   bIsVariable=true
   parent=ScrollBox_296
   widget=ExpandableArea_114
   properties {
    bIsExpanded=true
    areaPadding=<ltrb 10 10 10 10>
    borderBrush {
     bIsDynamicallyLoaded=false
     drawAs=RoundedBox
     imageType=NoImage
     mirroring=NoMirror
     resourceName=None
     resourceObject=null
     tiling=NoTile
     imageSize=<0 0>
     margin=<ltrb 0 0 0 0>
     outlineSettings {
      bUseBrushTransparency=false
      roundingType=FixedRadius
      width=0
      color {colorUseRule=UseColor_Specified, specifiedColor=<rgba 0 0 0 0>}
      cornerRadii=<4 4 4 4>
     }
     tintColor {colorUseRule=UseColor_Specified, specifiedColor=<rgba 0.0176 0.0176 0.0176 1>}
     uVRegion {bIsValid=false, max=<0 0>, min=<0 0>}
    }
    style {
     rolloutAnimationSeconds=0.1
     collapsedImage {
      bIsDynamicallyLoaded=false
      drawAs=Image
      imageType=Vector
      mirroring=NoMirror
      resourceName=../../../Engine/Content/Slate/Starship/Common/chevron-right.svg
      resourceObject=null
      tiling=NoTile
      imageSize=<16 16>
      margin=<ltrb 0 0 0 0>
      outlineSettings {
       bUseBrushTransparency=false
       roundingType=HalfHeightRadius
       width=0
       color {colorUseRule=UseColor_Specified, specifiedColor=<rgba 0 0 0 0>}
       cornerRadii=<0 0 0 0>
      }
      tintColor {colorUseRule=UseColor_Specified, specifiedColor=<rgba 0.5271 0.5271 0.5271 1>}
      uVRegion {bIsValid=false, max=<0 0>, min=<0 0>}
     }
     expandedImage {
      bIsDynamicallyLoaded=false
      drawAs=Image
      imageType=Vector
      mirroring=NoMirror
      resourceName=../../../Engine/Content/Slate/Starship/Common/chevron-down.svg
      resourceObject=null
      tiling=NoTile
      imageSize=<16 16>
      margin=<ltrb 0 0 0 0>
      outlineSettings {
       bUseBrushTransparency=false
       roundingType=HalfHeightRadius
       width=0
       color {colorUseRule=UseColor_Specified, specifiedColor=<rgba 0 0 0 0>}
       cornerRadii=<0 0 0 0>
      }
      tintColor {colorUseRule=UseColor_Specified, specifiedColor=<rgba 0.5271 0.5271 0.5271 1>}
      uVRegion {bIsValid=false, max=<0 0>, min=<0 0>}
     }
    }
   }
   slot {class=ScrollBoxSlot}
  }
  label_8 {
   widgetClassPath=&TextBlock
   bIsVariable=true
   namedSlotHost=ExpandableArea_114
   widget=label_8
   properties {
    text=Contact Shadows
    font {
     bForceMonospaced=false
     bMaterialIsStencil=false
     fontMaterial=null
     fontObject=&Roboto
     letterSpacing=0
     monospacedWidth=1
     size=10
     skewAmount=0
     typefaceFontName=Bold
     outlineSettings {
      bApplyOutlineToDropShadows=false
      bMiteredCorners=false
      bSeparateFillAlpha=false
      outlineMaterial=null
      outlineSize=0
      outlineColor=<rgba 0 0 0 1>
     }
    }
   }
  }
  VerticalBox_71 {widgetClassPath=&VerticalBox, namedSlotHost=ExpandableArea_114, widget=VerticalBox_71}
  EUW_cvar_checkBox {
   widgetClassPath=&EUW_cvar_checkBox
   bIsVariable=true
   parent=VerticalBox_71
   widget=EUW_cvar_checkBox
   properties {cvar=r.contactShadows, labelText=contactShadows, labelToolTip="0: disabled.\r\n 1: enabled."}
   slot {class=VerticalBoxSlot}
  }
  Spacer_186 {widgetClassPath=&Spacer, parent=ScrollBox_296, widget=Spacer_186, properties {size=<1 20>}, slot {class=ScrollBoxSlot}}
  ExpandableArea_364 {
   widgetClassPath=&ExpandableArea
   bIsVariable=true
   parent=ScrollBox_296
   widget=ExpandableArea_364
   properties {
    bIsExpanded=true
    areaPadding=<ltrb 10 10 10 10>
    borderBrush {
     bIsDynamicallyLoaded=false
     drawAs=RoundedBox
     imageType=NoImage
     mirroring=NoMirror
     resourceName=None
     resourceObject=null
     tiling=NoTile
     imageSize=<0 0>
     margin=<ltrb 0 0 0 0>
     outlineSettings {
      bUseBrushTransparency=false
      roundingType=FixedRadius
      width=0
      color {colorUseRule=UseColor_Specified, specifiedColor=<rgba 0 0 0 0>}
      cornerRadii=<4 4 4 4>
     }
     tintColor {colorUseRule=UseColor_Specified, specifiedColor=<rgba 0.0176 0.0176 0.0176 1>}
     uVRegion {bIsValid=false, max=<0 0>, min=<0 0>}
    }
    style {
     rolloutAnimationSeconds=0.1
     collapsedImage {
      bIsDynamicallyLoaded=false
      drawAs=Image
      imageType=Vector
      mirroring=NoMirror
      resourceName=../../../Engine/Content/Slate/Starship/Common/chevron-right.svg
      resourceObject=null
      tiling=NoTile
      imageSize=<16 16>
      margin=<ltrb 0 0 0 0>
      outlineSettings {
       bUseBrushTransparency=false
       roundingType=HalfHeightRadius
       width=0
       color {colorUseRule=UseColor_Specified, specifiedColor=<rgba 0 0 0 0>}
       cornerRadii=<0 0 0 0>
      }
      tintColor {colorUseRule=UseColor_Specified, specifiedColor=<rgba 0.5271 0.5271 0.5271 1>}
      uVRegion {bIsValid=false, max=<0 0>, min=<0 0>}
     }
     expandedImage {
      bIsDynamicallyLoaded=false
      drawAs=Image
      imageType=Vector
      mirroring=NoMirror
      resourceName=../../../Engine/Content/Slate/Starship/Common/chevron-down.svg
      resourceObject=null
      tiling=NoTile
      imageSize=<16 16>
      margin=<ltrb 0 0 0 0>
      outlineSettings {
       bUseBrushTransparency=false
       roundingType=HalfHeightRadius
       width=0
       color {colorUseRule=UseColor_Specified, specifiedColor=<rgba 0 0 0 0>}
       cornerRadii=<0 0 0 0>
      }
      tintColor {colorUseRule=UseColor_Specified, specifiedColor=<rgba 0.5271 0.5271 0.5271 1>}
      uVRegion {bIsValid=false, max=<0 0>, min=<0 0>}
     }
    }
   }
   slot {class=ScrollBoxSlot}
  }
  label_2 {
   widgetClassPath=&TextBlock
   bIsVariable=true
   namedSlotHost=ExpandableArea_364
   widget=label_2
   properties {
    text=Distance Field
    font {
     bForceMonospaced=false
     bMaterialIsStencil=false
     fontMaterial=null
     fontObject=&Roboto
     letterSpacing=0
     monospacedWidth=1
     size=10
     skewAmount=0
     typefaceFontName=Bold
     outlineSettings {
      bApplyOutlineToDropShadows=false
      bMiteredCorners=false
      bSeparateFillAlpha=false
      outlineMaterial=null
      outlineSize=0
      outlineColor=<rgba 0 0 0 1>
     }
    }
   }
  }
  VerticalBox_309 {widgetClassPath=&VerticalBox, namedSlotHost=ExpandableArea_364, widget=VerticalBox_309}
  EUW_checkBox_distanceFieldShadowing {
   widgetClassPath=&EUW_cvar_checkBox
   bIsVariable=true
   parent=VerticalBox_309
   widget=EUW_checkBox_distanceFieldShadowing
   properties {cvar=r.DistanceFieldShadowing, labelText=DistanceFieldShadowing, labelToolTip=Whether the distance field shadowing feature is allowed.}
   slot {class=VerticalBoxSlot}
  }
  EUW_comboBox_DFShadowQuality {
   widgetClassPath=&EUW_cvar_comboBox
   bIsVariable=true
   parent=VerticalBox_309
   widget=EUW_comboBox_DFShadowQuality
   properties {
    cvar=r.DFShadowQuality
    labelText=DFShadowQuality
    labelToolTip="Defines the distance field shadow method which allows to adjust for quality or performance.\r\n 0:off, 1:low (20 steps, no SSS), 2:medium (32 steps, no SSS), 3:high (64 steps, SSS, default)"
    option=1024
    options [0,1,2,3]
   }
   slot {class=VerticalBoxSlot}
  }
  Spacer_4 {widgetClassPath=&Spacer, parent=ScrollBox_296, widget=Spacer_4, properties {size=<1 20>}, slot {class=ScrollBoxSlot}}
  ExpandableArea_465 {
   widgetClassPath=&ExpandableArea
   bIsVariable=true
   parent=ScrollBox_296
   widget=ExpandableArea_465
   properties {
    bIsExpanded=true
    areaPadding=<ltrb 10 10 10 10>
    borderBrush {
     bIsDynamicallyLoaded=false
     drawAs=RoundedBox
     imageType=NoImage
     mirroring=NoMirror
     resourceName=None
     resourceObject=null
     tiling=NoTile
     imageSize=<0 0>
     margin=<ltrb 0 0 0 0>
     outlineSettings {
      bUseBrushTransparency=false
      roundingType=FixedRadius
      width=0
      color {colorUseRule=UseColor_Specified, specifiedColor=<rgba 0 0 0 0>}
      cornerRadii=<4 4 4 4>
     }
     tintColor {colorUseRule=UseColor_Specified, specifiedColor=<rgba 0.0176 0.0176 0.0176 1>}
     uVRegion {bIsValid=false, max=<0 0>, min=<0 0>}
    }
    style {
     rolloutAnimationSeconds=0.1
     collapsedImage {
      bIsDynamicallyLoaded=false
      drawAs=Image
      imageType=Vector
      mirroring=NoMirror
      resourceName=../../../Engine/Content/Slate/Starship/Common/chevron-right.svg
      resourceObject=null
      tiling=NoTile
      imageSize=<16 16>
      margin=<ltrb 0 0 0 0>
      outlineSettings {
       bUseBrushTransparency=false
       roundingType=HalfHeightRadius
       width=0
       color {colorUseRule=UseColor_Specified, specifiedColor=<rgba 0 0 0 0>}
       cornerRadii=<0 0 0 0>
      }
      tintColor {colorUseRule=UseColor_Specified, specifiedColor=<rgba 0.5271 0.5271 0.5271 1>}
      uVRegion {bIsValid=false, max=<0 0>, min=<0 0>}
     }
     expandedImage {
      bIsDynamicallyLoaded=false
      drawAs=Image
      imageType=Vector
      mirroring=NoMirror
      resourceName=../../../Engine/Content/Slate/Starship/Common/chevron-down.svg
      resourceObject=null
      tiling=NoTile
      imageSize=<16 16>
      margin=<ltrb 0 0 0 0>
      outlineSettings {
       bUseBrushTransparency=false
       roundingType=HalfHeightRadius
       width=0
       color {colorUseRule=UseColor_Specified, specifiedColor=<rgba 0 0 0 0>}
       cornerRadii=<0 0 0 0>
      }
      tintColor {colorUseRule=UseColor_Specified, specifiedColor=<rgba 0.5271 0.5271 0.5271 1>}
      uVRegion {bIsValid=false, max=<0 0>, min=<0 0>}
     }
    }
   }
   slot {class=ScrollBoxSlot}
  }
  label_4 {
   widgetClassPath=&TextBlock
   bIsVariable=true
   namedSlotHost=ExpandableArea_465
   widget=label_4
   properties {
    text=Far Shadows
    font {
     bForceMonospaced=false
     bMaterialIsStencil=false
     fontMaterial=null
     fontObject=&Roboto
     letterSpacing=0
     monospacedWidth=1
     size=10
     skewAmount=0
     typefaceFontName=Bold
     outlineSettings {
      bApplyOutlineToDropShadows=false
      bMiteredCorners=false
      bSeparateFillAlpha=false
      outlineMaterial=null
      outlineSize=0
      outlineColor=<rgba 0 0 0 1>
     }
    }
   }
  }
  VerticalBox_369 {widgetClassPath=&VerticalBox, namedSlotHost=ExpandableArea_465, widget=VerticalBox_369}
  EUW_checkBox_FarShadowCascades {
   widgetClassPath=&EUW_cvar_checkBox
   bIsVariable=true
   parent=VerticalBox_369
   widget=EUW_checkBox_FarShadowCascades
   properties {
    hasCvar=false
    labelText=Far Shadow Cascades
    labelToolTip=Far Shadow Cascade Togglre
    overrideSpinBox=&self:WidgetTree.EUW_slider_MAxNumFarShadowCascades
   }
   slot {class=VerticalBoxSlot}
  }
  EUW_slider_MAxNumFarShadowCascades {
   widgetClassPath=&EUW_cvar_spinBox
   bIsVariable=true
   parent=VerticalBox_369
   widget=EUW_slider_MAxNumFarShadowCascades
   properties {
    cvar=r.Shadow.MaxNumFarShadowCascades
    labelText=MaxNumFarShadowCascades
    labelToolTip=Max number of far shadow cascades that can be cast from a directional light
   }
   slot {class=VerticalBoxSlot}
  }
  Spacer_3 {widgetClassPath=&Spacer, parent=ScrollBox_296, widget=Spacer_3, properties {size=<1 20>}, slot {class=ScrollBoxSlot}}
  ExpandableArea_546 {
   widgetClassPath=&ExpandableArea
   bIsVariable=true
   parent=ScrollBox_296
   widget=ExpandableArea_546
   properties {
    bIsExpanded=true
    areaPadding=<ltrb 10 10 10 10>
    borderBrush {
     bIsDynamicallyLoaded=false
     drawAs=RoundedBox
     imageType=NoImage
     mirroring=NoMirror
     resourceName=None
     resourceObject=null
     tiling=NoTile
     imageSize=<0 0>
     margin=<ltrb 0 0 0 0>
     outlineSettings {
      bUseBrushTransparency=false
      roundingType=FixedRadius
      width=0
      color {colorUseRule=UseColor_Specified, specifiedColor=<rgba 0 0 0 0>}
      cornerRadii=<4 4 4 4>
     }
     tintColor {colorUseRule=UseColor_Specified, specifiedColor=<rgba 0.0176 0.0176 0.0176 1>}
     uVRegion {bIsValid=false, max=<0 0>, min=<0 0>}
    }
    style {
     rolloutAnimationSeconds=0.1
     collapsedImage {
      bIsDynamicallyLoaded=false
      drawAs=Image
      imageType=Vector
      mirroring=NoMirror
      resourceName=../../../Engine/Content/Slate/Starship/Common/chevron-right.svg
      resourceObject=null
      tiling=NoTile
      imageSize=<16 16>
      margin=<ltrb 0 0 0 0>
      outlineSettings {
       bUseBrushTransparency=false
       roundingType=HalfHeightRadius
       width=0
       color {colorUseRule=UseColor_Specified, specifiedColor=<rgba 0 0 0 0>}
       cornerRadii=<0 0 0 0>
      }
      tintColor {colorUseRule=UseColor_Specified, specifiedColor=<rgba 0.5271 0.5271 0.5271 1>}
      uVRegion {bIsValid=false, max=<0 0>, min=<0 0>}
     }
     expandedImage {
      bIsDynamicallyLoaded=false
      drawAs=Image
      imageType=Vector
      mirroring=NoMirror
      resourceName=../../../Engine/Content/Slate/Starship/Common/chevron-down.svg
      resourceObject=null
      tiling=NoTile
      imageSize=<16 16>
      margin=<ltrb 0 0 0 0>
      outlineSettings {
       bUseBrushTransparency=false
       roundingType=HalfHeightRadius
       width=0
       color {colorUseRule=UseColor_Specified, specifiedColor=<rgba 0 0 0 0>}
       cornerRadii=<0 0 0 0>
      }
      tintColor {colorUseRule=UseColor_Specified, specifiedColor=<rgba 0.5271 0.5271 0.5271 1>}
      uVRegion {bIsValid=false, max=<0 0>, min=<0 0>}
     }
    }
   }
   slot {class=ScrollBoxSlot}
  }
  label_5 {
   widgetClassPath=&TextBlock
   bIsVariable=true
   namedSlotHost=ExpandableArea_546
   widget=label_5
   properties {
    text=HeightField Shadows
    font {
     bForceMonospaced=false
     bMaterialIsStencil=false
     fontMaterial=null
     fontObject=&Roboto
     letterSpacing=0
     monospacedWidth=1
     size=10
     skewAmount=0
     typefaceFontName=Bold
     outlineSettings {
      bApplyOutlineToDropShadows=false
      bMiteredCorners=false
      bSeparateFillAlpha=false
      outlineMaterial=null
      outlineSize=0
      outlineColor=<rgba 0 0 0 1>
     }
    }
   }
  }
  VerticalBox_428 {widgetClassPath=&VerticalBox, namedSlotHost=ExpandableArea_546, widget=VerticalBox_428}
  EUW_checkBox_2 {
   widgetClassPath=&EUW_cvar_checkBox
   bIsVariable=true
   parent=VerticalBox_428
   widget=EUW_checkBox_2
   properties {cvar=r.HeightFieldShadowing, labelText=HeightFieldShadowing, labelToolTip=Whether the height field shadowing feature is allowed.}
   slot {class=VerticalBoxSlot}
  }
  EUW_comboBox_2 {
   widgetClassPath=&EUW_cvar_comboBox
   bIsVariable=true
   parent=VerticalBox_428
   widget=EUW_comboBox_2
   properties {
    cvar=r.HFShadowQuality
    labelText=HFShadowQuality
    labelToolTip="Defines the height field shadow method which allows to adjust for quality or performance.\r\n 0:off, 1:low (8 steps), 2:medium (16 steps, default), 3:high (32 steps, hole aware)"
    option=1024
    options [0,1,2,3]
   }
   slot {class=VerticalBoxSlot}
  }
  Spacer_2 {widgetClassPath=&Spacer, parent=ScrollBox_296, widget=Spacer_2, properties {size=<1 20>}, slot {class=ScrollBoxSlot}}
  ExpandableArea_654 {
   widgetClassPath=&ExpandableArea
   bIsVariable=true
   parent=ScrollBox_296
   widget=ExpandableArea_654
   properties {
    bIsExpanded=true
    areaPadding=<ltrb 10 10 10 10>
    borderBrush {
     bIsDynamicallyLoaded=false
     drawAs=RoundedBox
     imageType=NoImage
     mirroring=NoMirror
     resourceName=None
     resourceObject=null
     tiling=NoTile
     imageSize=<0 0>
     margin=<ltrb 0 0 0 0>
     outlineSettings {
      bUseBrushTransparency=false
      roundingType=FixedRadius
      width=0
      color {colorUseRule=UseColor_Specified, specifiedColor=<rgba 0 0 0 0>}
      cornerRadii=<4 4 4 4>
     }
     tintColor {colorUseRule=UseColor_Specified, specifiedColor=<rgba 0.0176 0.0176 0.0176 1>}
     uVRegion {bIsValid=false, max=<0 0>, min=<0 0>}
    }
    style {
     rolloutAnimationSeconds=0.1
     collapsedImage {
      bIsDynamicallyLoaded=false
      drawAs=Image
      imageType=Vector
      mirroring=NoMirror
      resourceName=../../../Engine/Content/Slate/Starship/Common/chevron-right.svg
      resourceObject=null
      tiling=NoTile
      imageSize=<16 16>
      margin=<ltrb 0 0 0 0>
      outlineSettings {
       bUseBrushTransparency=false
       roundingType=HalfHeightRadius
       width=0
       color {colorUseRule=UseColor_Specified, specifiedColor=<rgba 0 0 0 0>}
       cornerRadii=<0 0 0 0>
      }
      tintColor {colorUseRule=UseColor_Specified, specifiedColor=<rgba 0.5271 0.5271 0.5271 1>}
      uVRegion {bIsValid=false, max=<0 0>, min=<0 0>}
     }
     expandedImage {
      bIsDynamicallyLoaded=false
      drawAs=Image
      imageType=Vector
      mirroring=NoMirror
      resourceName=../../../Engine/Content/Slate/Starship/Common/chevron-down.svg
      resourceObject=null
      tiling=NoTile
      imageSize=<16 16>
      margin=<ltrb 0 0 0 0>
      outlineSettings {
       bUseBrushTransparency=false
       roundingType=HalfHeightRadius
       width=0
       color {colorUseRule=UseColor_Specified, specifiedColor=<rgba 0 0 0 0>}
       cornerRadii=<0 0 0 0>
      }
      tintColor {colorUseRule=UseColor_Specified, specifiedColor=<rgba 0.5271 0.5271 0.5271 1>}
      uVRegion {bIsValid=false, max=<0 0>, min=<0 0>}
     }
    }
   }
   slot {class=ScrollBoxSlot}
  }
  label_3 {
   widgetClassPath=&TextBlock
   bIsVariable=true
   namedSlotHost=ExpandableArea_654
   widget=label_3
   properties {
    text=Distance Field AO
    font {
     bForceMonospaced=false
     bMaterialIsStencil=false
     fontMaterial=null
     fontObject=&Roboto
     letterSpacing=0
     monospacedWidth=1
     size=10
     skewAmount=0
     typefaceFontName=Bold
     outlineSettings {
      bApplyOutlineToDropShadows=false
      bMiteredCorners=false
      bSeparateFillAlpha=false
      outlineMaterial=null
      outlineSize=0
      outlineColor=<rgba 0 0 0 1>
     }
    }
   }
  }
  VerticalBox_487 {widgetClassPath=&VerticalBox, namedSlotHost=ExpandableArea_654, widget=VerticalBox_487}
  EUW_checkBox_3 {
   widgetClassPath=&EUW_cvar_checkBox
   bIsVariable=true
   parent=VerticalBox_487
   widget=EUW_checkBox_3
   properties {
    cvar=r.distanceFieldAO
    labelText=DistanceFieldAO
    labelToolTip="Whether the distance field AO feature is allowed, which is used to implement shadows of Movable sky lights from static meshes."
   }
   slot {class=VerticalBoxSlot}
  }
  EUW_comboBox_142 {
   widgetClassPath=&EUW_cvar_comboBox
   bIsVariable=true
   parent=VerticalBox_487
   widget=EUW_comboBox_142
   properties {
    cvar=r.AOQuality
    labelText=AOQuality
    labelToolTip="r.AOQuality : Defines the distance field AO method which allows to adjust for quality or performance.\r\n 0:off, 1:medium, 2:high (default)"
    options [0,1,2]
   }
   slot {class=VerticalBoxSlot}
  }
  EUW_slider_4 {
   widgetClassPath=&EUW_cvar_spinBox
   bIsVariable=true
   parent=VerticalBox_487
   widget=EUW_slider_4
   properties {
    cvar=r.SkylightIntensityMultiplier
    labelText=SkylightIntensityMultiplier
    labelToolTip="Intensity scale on Stationary and Movable skylights.  This is useful to control overall lighting contrast in dynamically lit games with scalability levels which disable Ambient Occlusion.  For example, if medium quality disables SSAO and DFAO, reduce skylight intensity."
    max=1
    step=0.001
   }
   slot {class=VerticalBoxSlot}
  }
  Spacer_1 {widgetClassPath=&Spacer, parent=ScrollBox_296, widget=Spacer_1, properties {size=<1 20>}, slot {class=ScrollBoxSlot}}
  ExpandableArea_Misc {
   widgetClassPath=&ExpandableArea
   bIsVariable=true
   parent=ScrollBox_296
   widget=ExpandableArea_Misc
   properties {
    areaPadding=<ltrb 10 10 10 10>
    borderBrush {
     bIsDynamicallyLoaded=false
     drawAs=RoundedBox
     imageType=NoImage
     mirroring=NoMirror
     resourceName=None
     resourceObject=null
     tiling=NoTile
     imageSize=<0 0>
     margin=<ltrb 0 0 0 0>
     outlineSettings {
      bUseBrushTransparency=false
      roundingType=FixedRadius
      width=0
      color {colorUseRule=UseColor_Specified, specifiedColor=<rgba 0 0 0 0>}
      cornerRadii=<4 4 4 4>
     }
     tintColor {colorUseRule=UseColor_Specified, specifiedColor=<rgba 0.0176 0.0176 0.0176 1>}
     uVRegion {bIsValid=false, max=<0 0>, min=<0 0>}
    }
    style {
     rolloutAnimationSeconds=0.1
     collapsedImage {
      bIsDynamicallyLoaded=false
      drawAs=Image
      imageType=Vector
      mirroring=NoMirror
      resourceName=../../../Engine/Content/Slate/Starship/Common/chevron-right.svg
      resourceObject=null
      tiling=NoTile
      imageSize=<15 15>
      margin=<ltrb 0 0 0 0>
      outlineSettings {
       bUseBrushTransparency=false
       roundingType=HalfHeightRadius
       width=0
       color {colorUseRule=UseColor_Specified, specifiedColor=<rgba 0 0 0 0>}
       cornerRadii=<0 0 0 0>
      }
      tintColor {colorUseRule=UseColor_Specified, specifiedColor=<rgba 0.5271 0.5271 0.5271 1>}
      uVRegion {bIsValid=false, max=<0 0>, min=<0 0>}
     }
     expandedImage {
      bIsDynamicallyLoaded=false
      drawAs=Image
      imageType=Vector
      mirroring=NoMirror
      resourceName=../../../Engine/Content/Slate/Starship/Common/chevron-down.svg
      resourceObject=null
      tiling=NoTile
      imageSize=<16 16>
      margin=<ltrb 0 0 0 0>
      outlineSettings {
       bUseBrushTransparency=false
       roundingType=HalfHeightRadius
       width=0
       color {colorUseRule=UseColor_Specified, specifiedColor=<rgba 0 0 0 0>}
       cornerRadii=<0 0 0 0>
      }
      tintColor {colorUseRule=UseColor_Specified, specifiedColor=<rgba 0.5271 0.5271 0.5271 1>}
      uVRegion {bIsValid=false, max=<0 0>, min=<0 0>}
     }
    }
   }
   slot {class=ScrollBoxSlot}
  }
  label_7 {
   widgetClassPath=&TextBlock
   bIsVariable=true
   namedSlotHost=ExpandableArea_Misc
   widget=label_7
   properties {
    text=Misc
    font {
     bForceMonospaced=false
     bMaterialIsStencil=false
     fontMaterial=null
     fontObject=&Roboto
     letterSpacing=0
     monospacedWidth=1
     size=10
     skewAmount=0
     typefaceFontName=Bold
     outlineSettings {
      bApplyOutlineToDropShadows=false
      bMiteredCorners=false
      bSeparateFillAlpha=false
      outlineMaterial=null
      outlineSize=0
      outlineColor=<rgba 0 0 0 1>
     }
    }
   }
  }
  VerticalBox_557 {widgetClassPath=&VerticalBox, namedSlotHost=ExpandableArea_Misc, widget=VerticalBox_557}
  EUW_checkBox_volumetricFog {
   widgetClassPath=&EUW_cvar_checkBox
   bIsVariable=true
   parent=VerticalBox_557
   widget=EUW_checkBox_volumetricFog
   properties {cvar=r.VolumetricFog, labelText=VolumetricFog, labelToolTip=Whether to allow the volumetric fog feature.}
   slot {class=VerticalBoxSlot}
  }
  EUW_comboBox_volumetricFog_MissCount {
   widgetClassPath=&EUW_cvar_comboBox
   bIsVariable=true
   parent=VerticalBox_557
   widget=EUW_comboBox_volumetricFog_MissCount
   properties {
    cvar=r.VolumetricFog.HistoryMissSupersampleCount
    labelText=VolumetricFog.MissCount
    labelToolTip="Number of lighting samples to compute for voxels whose history value is not available.\r\nThis reduces noise when panning or on camera cuts, but introduces a variable cost to volumetric fog computation.  Valid range [1, 16]."
    options [1,2,3,4,5,6,7,8,9,10,11,12,13,14,15,16]
   }
   slot {class=VerticalBoxSlot}
  }
  EUW_slider_volumetricFog_GridPixelSize {
   widgetClassPath=&EUW_cvar_spinBox
   bIsVariable=true
   parent=VerticalBox_557
   widget=EUW_slider_volumetricFog_GridPixelSize
   properties {
    cvar=r.VolumetricFog.GridPixelSize
    labelText=VolumetricFog.GridPixelSize
    labelToolTip="XY Size of a cell in the voxel grid, in pixels"
    max=128
   }
   slot {class=VerticalBoxSlot}
  }
  EUW_slider_21 {
   widgetClassPath=&EUW_cvar_spinBox
   bIsVariable=true
   parent=VerticalBox_557
   widget=EUW_slider_21
   properties {cvar=r.VolumetricFog.GridSizeZ, labelText=VolumetricFog.GridSizeZ, labelToolTip=How many Volumetric Fog cells to use in z., max=256}
   slot {class=VerticalBoxSlot}
  }
  Spacer_83 {widgetClassPath=&Spacer, parent=VerticalBox_557, widget=Spacer_83, properties {size=<1 20>}, slot {class=VerticalBoxSlot}}
  EUW_slider_8 {
   widgetClassPath=&EUW_cvar_spinBox
   bIsVariable=true
   parent=VerticalBox_557
   widget=EUW_slider_8
   properties {
    cvar=Fort.TODDirectionalLightUpdateRate
    labelText=TODDirectionalLightUpdateRate
    labelToolTip="When >0, the directional light in the ToD manager is updated at the specified interval"
    max=1
    step=0.001
   }
   slot {class=VerticalBoxSlot}
  }
  EUW_slider_9 {
   widgetClassPath=&EUW_cvar_spinBox
   bIsVariable=true
   parent=VerticalBox_557
   widget=EUW_slider_9
   properties {
    cvar=Fort.TODDirectionalLightInterpSpeed
    labelText=TODDirectionalLightInterpSpeed
    labelToolTip="When TODDirectionalLightUpdateRate >0, this is the speed of interpolation to the target rotation. If 0, the interp speed is instant."
    step=0.001
   }
   slot {class=VerticalBoxSlot}
  }
  EUW_slider_15 {
   widgetClassPath=&EUW_cvar_spinBox
   bIsVariable=true
   parent=VerticalBox_557
   widget=EUW_slider_15
   properties {
    cvar=r.Shadow.PreShadowResolutionFactor
    labelText=PreShadowResolutionFactor
    labelToolTip=Mulitplier for preshadow resolution
    max=1
    step=0.01
   }
   slot {class=VerticalBoxSlot}
  }
  EUW_slider_17 {
   widgetClassPath=&EUW_cvar_spinBox
   bIsVariable=true
   parent=VerticalBox_557
   widget=EUW_slider_17
   properties {
    cvar=r.Shadow.RadiusThreshold
    labelText=ShadowRadiusThreshold
    labelToolTip="Cull shadow casters if they are too small, value is the minimal screen space bounding sphere radius."
    max=1
    step=0.01
   }
   slot {class=VerticalBoxSlot}
  }
  EUW_slider_18 {
   widgetClassPath=&EUW_cvar_spinBox
   bIsVariable=true
   parent=VerticalBox_557
   widget=EUW_slider_18
   properties {
    cvar=r.LightFunctionQuality
    labelText=LightFunctionQuality
    labelToolTip="Defines the light function quality which allows to adjust for quality or performance.\r\n<=0: off (fastest)\r\n  1: low quality (e.g. half res with blurring, not yet implemented)\r\n  2: normal quality (default)\r\n  3: high quality (e.g. super-sampled or colored, not yet implemented)"
    max=3
   }
   slot {class=VerticalBoxSlot}
  }
  EUW_slider_6 {
   widgetClassPath=&EUW_cvar_spinBox
   bIsVariable=true
   parent=VerticalBox_557
   widget=EUW_slider_6
   properties {
    cvar=r.LightMaxDrawDistanceScale
    labelText=LightMaxDrawDistanceScale
    labelToolTip="Scale applied to the MaxDrawDistance of lights.  Useful for fading out local lights more aggressively on some platforms."
    max=1
    step=0.001
   }
   slot {class=VerticalBoxSlot}
  }
  Spacer {widgetClassPath=&Spacer, parent=VerticalBox_557, widget=Spacer, properties {size=<1 20>}, slot {class=VerticalBoxSlot}}
  EUW_checkBox_4 {
   widgetClassPath=&EUW_cvar_checkBox
   bIsVariable=true
   parent=VerticalBox_557
   widget=EUW_checkBox_4
   properties {
    cvar=r.CapsuleShadows
    labelText=CapsuleShadows
    labelToolTip=Whether to allow capsule shadowing on skinned components with bCastCapsuleDirectShadow or bCastCapsuleIndirectShadow enabled.
   }
   slot {class=VerticalBoxSlot}
  }
  EUW_slider_5 {
   widgetClassPath=&EUW_cvar_spinBox
   bIsVariable=true
   parent=VerticalBox_557
   widget=EUW_slider_5
   properties {
    cvar=r.Shadow.MaxNumSpotShadowCacheUpdatesPerFrame
    labelText=MaxSpotShadowCacheUpdates
    labelToolTip=Maximum number of spot light shadow cache updates allowed per frame.Only affect updates caused by resolution change. -1 means no limit.
    max=4
    min=-1
   }
   slot {class=VerticalBoxSlot}
  }
  EUW_slider_7 {
   widgetClassPath=&EUW_cvar_spinBox
   bIsVariable=true
   parent=VerticalBox_557
   widget=EUW_slider_7
   properties {
    cvar=r.Shadow.MaxNumPointShadowCacheUpdatesPerFrame
    labelText=MaxPointShadowCacheUpdates
    labelToolTip=aximum number of point light shadow cache updates allowed per frame.Only affect updates caused by resolution change. -1 means no limit.
    max=4
    min=-1
   }
   slot {class=VerticalBoxSlot}
  }
 ]
}
